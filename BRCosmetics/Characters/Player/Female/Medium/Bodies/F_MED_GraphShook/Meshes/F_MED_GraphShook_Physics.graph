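# Physics asset for the GraphShook female medium player body, defining bodies and constraints for its coat, belts, zipper and cloth-volume bones.
# Coat panel chains hang from spine_04 through MSTR bones with mostly tight angular limits, cross-linked to neighboring panels by free, limited or locked constraints.
# Free-swinging cloth-volume chains attach to the pelvis; belt chains hang from upper arms and coat hem.
# Mostly capsule bodies also cover spine, pelvis, thighs and arms.
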PhysicsAsset /BRCosmetics/Characters/Player/Female/Medium/Bodies/F_MED_GraphShook/Meshes/F_MED_GraphShook_Physics
bodies {
 dyn_Cloth_volume_bk02 [{shapeType=Capsule, length=2.645, radius=9.5081, shapeName=dyn_Cloth_volume_bk02_capsule, center=<0 6.05 0>}]
 dyn_Cloth_volume_fn [{shapeType=Capsule, length=8.7515, radius=7.0561, shapeName=dyn_Cloth_volume_fn_01_capsule, center=<0 3.0079 0>}]
 dyn_Cloth_volume_fn_01 [{shapeType=Capsule, length=11.2137, radius=7.0561, shapeName=dyn_Cloth_volume_fn_01_capsule, center=<-2.3958 4.9924 0>}]
 dyn_Cloth_volume_fn_02 [{shapeType=Capsule, length=8.7616, radius=9.5081, shapeName=dyn_Cloth_volume_fn_02_capsule, center=<0 8.7708 0>}]
 dyn_Cloth_volume_l_A [{shapeType=Capsule, length=5.097, radius=7.0561, shapeName=dyn_Cloth_volume_l_A01_capsule, center=<0 3.7076 0>}]
 dyn_Cloth_volume_l_A01 [{shapeType=Capsule, length=5.097, radius=7.0561, shapeName=dyn_Cloth_volume_l_A01_capsule, center=<-2.3958 7.3805 0>}]
 dyn_Cloth_volume_l_A02 [{shapeType=Capsule, length=2.645, radius=9.5081, shapeName=dyn_Cloth_volume_l_A02_capsule, center=<0 9.5696 0>}]
 dyn_Cloth_volume_l_B [{shapeType=Capsule, length=6.8009, radius=5.3522, shapeName=dyn_Cloth_volume_l_B01_capsule, center=<0 3.8923 0>}]
 dyn_Cloth_volume_l_B01 [{shapeType=Capsule, length=6.8009, radius=5.3522, shapeName=dyn_Cloth_volume_l_B01_capsule, center=<0 7.0329 0>}]
 dyn_Cloth_volume_l_B02 [{shapeType=Capsule, length=4.3489, radius=7.8042, shapeName=dyn_Cloth_volume_l_B02_capsule, center=<0 11.8375 0>}]
 dyn_Cloth_volume_l_C [{shapeType=Capsule, length=6.8009, radius=5.3522, shapeName=dyn_Cloth_volume_l_B01_capsule, center=<0 3.8923 0>}]
 dyn_Cloth_volume_l_C01 [{shapeType=Capsule, length=6.8009, radius=5.3522, shapeName=dyn_Cloth_volume_l_B01_capsule, center=<0 7.0329 0>}]
 dyn_Cloth_volume_l_C02 [{shapeType=Capsule, length=4.3489, radius=7.8042, shapeName=dyn_Cloth_volume_l_C02_capsule, center=<0 6.05 0>}]
 dyn_Cloth_volume_r_A [{shapeType=Capsule, length=5.097, radius=7.0561, shapeName=dyn_Cloth_volume_r_A01_capsule, center=<0 9.4524 0>}]
 dyn_Cloth_volume_r_A01 [{shapeType=Capsule, length=5.097, radius=7.0561, shapeName=dyn_Cloth_volume_r_A01_capsule, center=<-2.3958 6.7544 0>}]
 dyn_Cloth_volume_r_A02 [{shapeType=Capsule, length=2.645, radius=9.5081, shapeName=dyn_Cloth_volume_r_A02_capsule, center=<0 7.7963 0>}]
 dyn_Cloth_volume_r_B [{shapeType=Capsule, length=6.8009, radius=5.3522, shapeName=dyn_Cloth_volume_r_B01_capsule, center=<0 7.3691 0>}]
 dyn_Cloth_volume_r_B01 [{shapeType=Capsule, length=6.8009, radius=5.3522, shapeName=dyn_Cloth_volume_r_B01_capsule, center=<0 6.4417 0>}]
 dyn_Cloth_volume_r_B02 [{shapeType=Capsule, length=4.3489, radius=7.8042, shapeName=dyn_Cloth_volume_r_B02_capsule, center=<0 11.3777 0>}]
 dyn_Cloth_volume_r_C [{shapeType=Capsule, length=6.8009, radius=5.3522, shapeName=dyn_Cloth_volume_r_B01_capsule, center=<0 3.8923 0>}]
 dyn_Cloth_volume_r_C01 [{shapeType=Capsule, length=6.8009, radius=5.3522, shapeName=dyn_Cloth_volume_r_B01_capsule, center=<0 1.6849 0>}]
 dyn_Cloth_volume_r_C02 [{shapeType=Capsule, length=4.3489, radius=7.8042, shapeName=dyn_Cloth_volume_r_C02_capsule, center=<0 3.4961 0>}]
 dyn_coat_01_L [{
   shapeType=Capsule
   length=1.0645
   radius=4.9355
   shapeName=dyn_coat_01_L_capsule
   center=<3.1804 0.0071 0.1575>
   rotation=<rot 0 0.0001 92.5648>
  }]
 dyn_coat_01_R [{
   shapeType=Capsule
   length=1.0645
   radius=4.9355
   shapeName=dyn_coat_01_R_capsule
   center=<3.1804 0.007 0.1575>
   rotation=<rot 0 -0.0001 92.5648>
  }]
 dyn_coat_01_fL [{
   shapeType=Capsule
   length=2.7988
   radius=4.4355
   shapeName=dyn_coat_01_fL_capsule
   center=<3.1804 0.0071 0.1575>
   rotation=<rot 0.0001 0.0001 92.5648>
  }]
 dyn_coat_01_fR [{
   shapeType=Capsule
   length=2.7988
   radius=4.4355
   shapeName=dyn_coat_01_fL_capsule
   center=<3.1804 0.0071 0.1575>
   rotation=<rot 0.0001 0.0001 92.5648>
  }]
 dyn_coat_02_L [{
   shapeType=Capsule
   length=0.8592
   radius=5.5963
   shapeName=dyn_coat_02_R_capsule
   center=<4.8929 0.1576 3.5177>
   rotation=<rot 0 -0.0001 92.5648>
  }]
 dyn_coat_02_R [{
   shapeType=Capsule
   length=0.404
   radius=6.0516
   shapeName=dyn_coat_02_R_capsule
   center=<5.2184 0.213 4.7543>
   rotation=<rot -6.1922 -0.2786 92.5799>
  }]
 dyn_coat_02_fL [{
   shapeType=Capsule
   length=0
   radius=6.9005
   shapeName=dyn_coat_02_fL_capsule
   center=<4.7606 0.1502 3.3532>
   rotation=<rot 0.0001 0.0001 92.5648>
  }]
 dyn_coat_02_fR [{
   shapeType=Capsule
   length=0
   radius=6.9005
   shapeName=dyn_coat_02_fL_capsule
   center=<4.7606 0.1502 3.3532>
   rotation=<rot 0.0001 0.0001 92.5648>
  }]
 dyn_coat_03_L [{
   shapeType=Capsule
   length=3.1844
   radius=6.5209
   shapeName=dyn_coat_02_R_capsule
   center=<5.1804 0.1057 2.3607>
   rotation=<rot 0 -0.0001 92.5648>
  }]
 dyn_coat_03_R [{
   shapeType=Capsule
   length=4.1251
   radius=6.0418
   shapeName=dyn_coat_02_R_capsule
   center=<5.1804 0.2928 6.5361>
   rotation=<rot 0 -0.0001 92.5648>
  }]
 dyn_coat_03_belt01_L [{
   shapeType=Capsule
   length=1.9741
   radius=4.0437
   shapeName=dyn_coat_02_L_capsule
   center=<5.1804 0.0071 0.1575>
   rotation=<rot 2.5619 -87.2774 90.1218>
  }]
 dyn_coat_03_belt01_R [{
   shapeType=Capsule
   length=6.6125
   radius=3.0928
   shapeName=dyn_coat_02_R_capsule
   center=<3.9046 0.017 0.1571>
   rotation=<rot 2.5647 -90.449 89.9799>
  }]
 dyn_coat_03_belt02_L [{
   shapeType=Capsule
   length=1.9741
   radius=4.0437
   shapeName=dyn_coat_02_L_capsule
   center=<5.1804 0.0071 0.1575>
   rotation=<rot 2.5619 -87.2774 90.1218>
  }]
 dyn_coat_03_belt02_R [{
   shapeType=Capsule
   length=7.0395
   radius=2.6658
   shapeName=dyn_coat_02_R_capsule
   center=<6.2223 -0.0011 0.1579>
   rotation=<rot 2.5647 -90.449 89.9799>
  }]
 dyn_coat_03_belt03_R [{
   shapeType=Capsule
   length=7.0395
   radius=2.6658
   shapeName=dyn_coat_02_R_capsule
   center=<5.1804 0.007 0.1575>
   rotation=<rot 2.5647 -90.449 89.9799>
  }]
 dyn_coat_03_fL [{
   shapeType=Capsule
   length=2.5748
   radius=7.0324
   shapeName=dyn_coat_02_fL_capsule
   center=<6.7606 0.1502 3.3532>
   rotation=<rot 0.0001 0.0001 92.5648>
  }]
 dyn_coat_03_fR [{
   shapeType=Capsule
   length=2.5748
   radius=7.0324
   shapeName=dyn_coat_02_fL_capsule
   center=<6.7606 0.2518 5.6216>
   rotation=<rot 0.0001 0.0001 92.5648>
  }]
 dyn_coat_MSTR_L [{shapeType=Capsule, length=2.5, radius=3.5, shapeName=dyn_coat_MSTR_L_capsule, center=<3.1804 0.0071 0.1575>, rotation=<rot 0 0.0001 92.5648>}]
 dyn_coat_MSTR_R [{shapeType=Capsule, length=2.5, radius=3.5, shapeName=dyn_coat_MSTR_R_capsule, center=<3.1804 0.007 0.1575>, rotation=<rot 0 -0.0001 92.5648>}]
 dyn_coat_MSTR_fL [{
   shapeType=Capsule
   length=2.5
   radius=3.5
   shapeName=dyn_coat_MSTR_fL_capsule
   center=<3.1804 0.0071 0.1575>
   rotation=<rot 0.0001 0.0001 92.5648>
  }]
 dyn_coat_MSTR_fR [{
   shapeType=Capsule
   length=2.5
   radius=3.5
   shapeName=dyn_coat_MSTR_fR_capsule
   center=<3.1804 0.007 0.1575>
   rotation=<rot 0.0001 -0.0001 92.5648>
  }]
 dyn_coat_bk_01 [{
   shapeType=Capsule
   length=5.2874
   radius=4.9355
   shapeName=dyn_coat_bk_01_capsule
   center=<3.1804 0.0276 0.6166>
   rotation=<rot -0.0001 0 92.5648>
  }]
 dyn_coat_bk_02 [{
   shapeType=Capsule
   length=4.3458
   radius=4.6088
   shapeName=dyn_coat_bk_02_capsule
   center=<3.1804 0.1109 2.4754>
   rotation=<rot -0.0001 0 92.5648>
  }]
 dyn_coat_bk_03 [{
   shapeType=Capsule
   length=5.7311
   radius=6.4732
   shapeName=dyn_coat_bk_02_capsule
   center=<5.8874 0.1614 3.6041>
   rotation=<rot -0.0001 0 92.5648>
  }]
 dyn_coat_bk_MSTR [{
   shapeType=Capsule
   length=2.5
   radius=3.5
   shapeName=dyn_coat_bk_MSTR_capsule
   center=<3.1804 0.0071 0.1575>
   rotation=<rot -0.0001 0 92.5648>
  }]
 dyn_coat_fr_01_L [{
   shapeType=Capsule
   length=2.4067
   radius=5.1855
   shapeName=dyn_coat_fr_01_L_capsule
   center=<4.0605 0.1058 2.3614>
   rotation=<rot 0.0001 0.0001 92.5648>
  }]
 dyn_coat_fr_01_R [{
   shapeType=Capsule
   length=2.0395
   radius=5.5527
   shapeName=dyn_coat_fr_01_L_capsule
   center=<4.0605 0.1058 2.3614>
   rotation=<rot 0.0001 0.0001 92.5648>
  }]
 dyn_coat_fr_02_L [{
   shapeType=Capsule
   length=0
   radius=7.2119
   shapeName=dyn_coat_fr_02_L_capsule
   center=<5.7697 0.2602 5.8077>
   rotation=<rot 0.0001 0.0001 92.5648>
  }]
 dyn_coat_fr_02_R [{shapeType=Box, extentX=13.3487, extentY=17.9911, extentZ=10.0247, shapeName=dyn_coat_fr_02_R_box, center=<6.8997 0 3.9504>}]
 dyn_coat_fr_03_L [{
   shapeType=Capsule
   length=2.5072
   radius=7.696
   shapeName=dyn_coat_fr_02_L_capsule
   center=<6.0605 0.3273 7.3072>
   rotation=<rot 0.0001 0.0001 92.5648>
  }]
 dyn_coat_fr_03_R [{shapeType=Box, extentX=13.3487, extentY=9.5449, extentZ=10.0247, shapeName=dyn_coat_fr_03_R_box, center=<6.8997 0 3.9504>}]
 dyn_coat_fr_03_zipper01_L [{
   shapeType=Capsule
   length=6.7361
   radius=3.2021
   shapeName=dyn_coat_fr_03_zipper01_L_capsule
   center=<5.2508 -0.1108 -0.0121>
   rotation=<rot 0 89.244 -89.4794>
  }]
 dyn_coat_fr_MSTR_L [{
   shapeType=Capsule
   length=2.5
   radius=3.5
   shapeName=dyn_coat_fr_MSTR_L_capsule
   center=<3.1804 0.0071 0.1575>
   rotation=<rot 0.0001 0.0001 92.5648>
  }]
 dyn_coat_fr_MSTR_R [{
   shapeType=Capsule
   length=2.5
   radius=3.5
   shapeName=dyn_coat_fr_MSTR_R_capsule
   center=<3.1804 0.0071 0.1575>
   rotation=<rot 0.0001 -0.0001 92.5648>
  }]
 dyn_coat_mid_01_L [{
   shapeType=Capsule
   length=2.1951
   radius=4.9355
   shapeName=dyn_coat_mid_01_L_capsule
   center=<3.1804 0.0071 0.1575>
   rotation=<rot -0.0001 0 92.5648>
  }]
 dyn_coat_mid_01_R [{
   shapeType=Capsule
   length=4.083
   radius=4.9355
   shapeName=dyn_coat_mid_01_R_capsule
   center=<3.1804 0.0148 0.3308>
   rotation=<rot -0.0001 0 92.5648>
  }]
 dyn_coat_mid_02_L [{
   shapeType=Capsule
   length=1.6075
   radius=4.9355
   shapeName=dyn_coat_mid_02_L_capsule
   center=<8.1914 0.0997 2.2257>
   rotation=<rot -0.0001 0 92.5648>
  }]
 dyn_coat_mid_02_R [{
   shapeType=Capsule
   length=2.0748
   radius=4.9355
   shapeName=dyn_coat_mid_02_R_capsule
   center=<3.1804 0.0916 2.0454>
   rotation=<rot -0.0001 0 92.5648>
  }]
 dyn_coat_mid_03_L [{
   shapeType=Capsule
   length=3.224
   radius=6.5687
   shapeName=dyn_coat_mid_02_L_capsule
   center=<8.1914 0.1424 3.1795>
   rotation=<rot -0.0001 0 92.5648>
  }]
 dyn_coat_mid_03_R [{
   shapeType=Capsule
   length=2.6876
   radius=6.3524
   shapeName=dyn_coat_mid_02_R_capsule
   center=<5.1804 0.166 3.7064>
   rotation=<rot -0.0001 0 92.5648>
  }]
 dyn_coat_mid_MSTR_L [{
   shapeType=Capsule
   length=2.5
   radius=3.5
   shapeName=dyn_coat_mid_MSTR_L_capsule
   center=<3.1804 0.0071 0.1575>
   rotation=<rot -0.0001 0 92.5648>
  }]
 dyn_coat_mid_MSTR_R [{
   shapeType=Capsule
   length=2.5
   radius=3.5
   shapeName=dyn_coat_mid_MSTR_R_capsule
   center=<3.1804 0.0071 0.1575>
   rotation=<rot -0.0001 0 92.5648>
  }]
 lowerarm_l [{
   shapeType=Capsule
   length=22.2739
   radius=6.2323
   shapeName=lowerarm_l_capsule
   center=<16.9405 -0.2721 -2.469>
   rotation=<rot 0 -93.8953 -80.7933>
  }]
 pelvis [{shapeType=Capsule, length=3.0639, radius=11.2452, shapeName=spine_03_capsule, center=<-2.1753 -2.6971 0>}]
 spine_01 [{shapeType=Capsule, length=12.2492, radius=17.0457, center=<-0.9355 3.5472 -1.6956>}]
 spine_02 [{shapeType=Capsule, length=5.6263, radius=11.1667, shapeName=spine_03_capsule, center=<-2.1753 1.3029 0>}]
 spine_03 [{shapeType=Capsule, length=0.3434, radius=8.4924, shapeName=spine_03_capsule, center=<-2.1753 1.3029 0>}]
 spine_04 [{shapeType=Capsule, length=1.5362, radius=15.1723, shapeName=spine_03_capsule, center=<-11.4818 -3.0473 0>}]
 spine_05 [{shapeType=Capsule, length=12.0283, radius=14.702, center=<-6.8982 -0.9185 -0.9617>, rotation=<rot 0.0251 20.5244 -90.0669>}]
 thigh_l [{
   shapeType=Capsule
   length=33.8099
   radius=9.3332
   shapeName=thigh_l_capsule
   center=<-22.6021 0.6914 0.7242>
   rotation=<rot 0 91.2606 -88.8026>
  }]
 thigh_r [{
   shapeType=Capsule
   length=33.8099
   radius=9.3332
   shapeName=thigh_l_capsule
   center=<21.2279 1.6559 -1.4255>
   rotation=<rot 0 91.2606 -88.8026>
  }]
 thigh_twist_01_l [{
   shapeType=Capsule
   length=26.6477
   radius=16.4954
   shapeName=thigh_l_capsule
   center=<15.57 1.5314 0.8649>
   rotation=<rot 0 91.2606 -88.8026>
  }]
 upperarm_l [{shapeType=Capsule, length=20.279, radius=5.5238, center=<14.2552 0.3624 -0.2842>, rotation=<rot 84.7907 -179.9717 -179.9993>}]
 upperarm_r [{shapeType=Capsule, length=20.4239, radius=5.3789, center=<-13.5449 -0.9214 1.8203>, rotation=<rot 75.1196 172.4014 -179.9717>}]
 upperarm_twist_01_belt01_l [{shapeType=Capsule, length=4.3909, radius=1.8423, shapeName=dyn_coat_02_L_capsule, center=<0 0 -4.3386>}]
 upperarm_twist_01_belt02_l [{shapeType=Capsule, length=4.6728, radius=2.0191, shapeName=dyn_coat_02_L_capsule, center=<0 0 -4.3386>}]
 upperarm_twist_01_belt03_l [{shapeType=Capsule, length=4.5853, radius=2.6486, shapeName=dyn_coat_02_L_capsule, center=<0 0 -4.3386>}]
 upperarm_twist_01_belt04_l [{shapeType=Capsule, length=4.4963, radius=2.7376, shapeName=dyn_coat_02_L_capsule, center=<0 0 -4.3386>}]
 upperarm_twist_01_belt_r [{
   shapeType=Capsule
   length=2.5428
   radius=1.9429
   shapeName=upperarm_twist_01_belt_r_capsule
   center=<0.0324 1.075 -2.7405>
   rotation=<rot -0.3344 -90 2.7319>
  }]
 upperarm_twist_01_l [{
   shapeType=Capsule
   length=0
   radius=2.441
   shapeName=upperarm_twist_01_l_capsule
   center=<4.8505 -1.2271 -1.7672>
   rotation=<rot 0 -95.6639 -74.8152>
  }]
 upperarm_twist_01_r [{
   shapeType=Capsule
   length=0
   radius=1.1387
   shapeName=upperarm_twist_01_r_capsule
   center=<-4.9725 0.8905 1.6705>
   rotation=<rot 0 -91.6848 -86.6532>
  }]
 upperarm_twist_02_belt_r [{
   shapeType=Capsule
   length=4.8605
   radius=2.8972
   shapeName=upperarm_twist_02_belt_r_capsule
   center=<-0.0281 -1.5898 -5.8369>
   rotation=<rot 15.7062 -90 -0.1708>
  }]
}
constraints [
 {bone1Name=upperarm_l, bone2Name=spine_05, swing1LimitDegrees=45, swing2LimitDegrees=45, twistLimitDegrees=45}
 {bone1Name=upperarm_r, bone2Name=spine_05, swing1LimitDegrees=45, swing2LimitDegrees=45, twistLimitDegrees=45}
 {
  bone1Name=spine_05
  bone2Name=spine_04
  swing1LimitDegrees=45
  swing1Motion=Free
  swing2LimitDegrees=45
  swing2Motion=Free
  twistLimitDegrees=45
  twistMotion=Free
 }
 {bone1Name=dyn_coat_mid_01_L, bone2Name=dyn_coat_mid_MSTR_L, swing1LimitDegrees=5, swing2LimitDegrees=7, twistLimitDegrees=1}
 {bone1Name=dyn_coat_mid_02_L, bone2Name=dyn_coat_mid_01_L, swing1LimitDegrees=1, swing2LimitDegrees=7, twistLimitDegrees=1}
 {bone1Name=dyn_coat_mid_01_R, bone2Name=dyn_coat_mid_MSTR_R, swing1LimitDegrees=2, swing2LimitDegrees=15, twistLimitDegrees=1}
 {bone1Name=dyn_coat_mid_02_R, bone2Name=dyn_coat_mid_01_R, swing1LimitDegrees=1, swing2LimitDegrees=7, twistLimitDegrees=1}
 {bone1Name=dyn_coat_fr_02_R, bone2Name=dyn_coat_fr_01_R, swing1LimitDegrees=5, swing2LimitDegrees=40, twistLimitDegrees=1}
 {bone1Name=dyn_coat_fr_01_R, bone2Name=dyn_coat_fr_MSTR_R, swing1LimitDegrees=2, swing2LimitDegrees=16, twistLimitDegrees=1}
 {bone1Name=dyn_coat_fr_02_L, bone2Name=dyn_coat_fr_01_L, swing1LimitDegrees=5, swing2LimitDegrees=20, twistLimitDegrees=1}
 {bone1Name=dyn_coat_fr_01_L, bone2Name=dyn_coat_fr_MSTR_L, swing1LimitDegrees=2, swing2LimitDegrees=16, twistLimitDegrees=1}
 {bone1Name=dyn_coat_bk_02, bone2Name=dyn_coat_bk_01, swing1LimitDegrees=2, swing2LimitDegrees=3, twistLimitDegrees=1}
 {bone1Name=dyn_coat_bk_01, bone2Name=dyn_coat_bk_MSTR, swing1LimitDegrees=2, swing2LimitDegrees=5, twistLimitDegrees=1}
 {bone1Name=dyn_coat_02_fR, bone2Name=dyn_coat_01_fR, swing1LimitDegrees=0.5, swing2LimitDegrees=3, twistLimitDegrees=1}
 {bone1Name=dyn_coat_01_fR, bone2Name=dyn_coat_MSTR_fR, swing1LimitDegrees=0.5, swing2LimitDegrees=3, twistLimitDegrees=1}
 {bone1Name=dyn_coat_02_fL, bone2Name=dyn_coat_01_fL, swing1LimitDegrees=1, swing2LimitDegrees=10, twistLimitDegrees=1}
 {bone1Name=dyn_coat_01_fL, bone2Name=dyn_coat_MSTR_fL, swing1LimitDegrees=1, swing2LimitDegrees=5, twistLimitDegrees=1}
 {bone1Name=dyn_coat_02_R, bone2Name=dyn_coat_01_R, swing1LimitDegrees=1, swing2LimitDegrees=1, twistLimitDegrees=1}
 {bone1Name=dyn_coat_01_R, bone2Name=dyn_coat_MSTR_R, swing1LimitDegrees=1, swing2LimitDegrees=1, twistLimitDegrees=1}
 {bone1Name=dyn_coat_02_L, bone2Name=dyn_coat_01_L, swing1LimitDegrees=1, swing2LimitDegrees=7, twistLimitDegrees=1}
 {bone1Name=dyn_coat_01_L, bone2Name=dyn_coat_MSTR_L, swing1LimitDegrees=1, swing2LimitDegrees=1, twistLimitDegrees=1}
 {
  bone1Name=dyn_coat_MSTR_R
  bone2Name=spine_04
  swing1LimitDegrees=45
  swing1Motion=Free
  swing2LimitDegrees=45
  swing2Motion=Free
  twistLimitDegrees=45
  twistMotion=Free
 }
 {
  bone1Name=dyn_coat_MSTR_L
  bone2Name=spine_04
  swing1LimitDegrees=45
  swing1Motion=Free
  swing2LimitDegrees=45
  swing2Motion=Free
  twistLimitDegrees=45
  twistMotion=Free
 }
 {
  bone1Name=dyn_coat_MSTR_fL
  bone2Name=spine_04
  swing1LimitDegrees=45
  swing1Motion=Free
  swing2LimitDegrees=45
  swing2Motion=Free
  twistLimitDegrees=45
  twistMotion=Free
 }
 {
  bone1Name=dyn_coat_MSTR_fR
  bone2Name=spine_04
  swing1LimitDegrees=45
  swing1Motion=Free
  swing2LimitDegrees=45
  swing2Motion=Free
  twistLimitDegrees=45
  twistMotion=Free
 }
 {
  bone1Name=dyn_coat_bk_MSTR
  bone2Name=spine_04
  swing1LimitDegrees=45
  swing1Motion=Free
  swing2LimitDegrees=45
  swing2Motion=Free
  twistLimitDegrees=45
  twistMotion=Free
 }
 {
  bone1Name=dyn_coat_fr_MSTR_L
  bone2Name=spine_04
  swing1LimitDegrees=45
  swing1Motion=Free
  swing2LimitDegrees=45
  swing2Motion=Free
  twistLimitDegrees=45
  twistMotion=Free
 }
 {
  bone1Name=dyn_coat_fr_MSTR_R
  bone2Name=spine_04
  swing1LimitDegrees=45
  swing1Motion=Free
  swing2LimitDegrees=45
  swing2Motion=Free
  twistLimitDegrees=45
  twistMotion=Free
 }
 {
  bone1Name=dyn_coat_mid_MSTR_L
  bone2Name=spine_04
  swing1LimitDegrees=45
  swing1Motion=Free
  swing2LimitDegrees=45
  swing2Motion=Free
  twistLimitDegrees=45
  twistMotion=Free
 }
 {
  bone1Name=dyn_coat_mid_MSTR_R
  bone2Name=spine_04
  swing1LimitDegrees=45
  swing1Motion=Free
  swing2LimitDegrees=45
  swing2Motion=Free
  twistLimitDegrees=45
  twistMotion=Free
 }
 {
  bone1Name=dyn_coat_mid_02_L
  bone2Name=dyn_coat_bk_02
  swing1LimitDegrees=45
  swing1Motion=Free
  swing2LimitDegrees=45
  swing2Motion=Free
  twistLimitDegrees=45
  twistMotion=Free
 }
 {
  bone1Name=dyn_coat_mid_02_R
  bone2Name=dyn_coat_bk_02
  swing1LimitDegrees=45
  swing1Motion=Free
  swing2LimitDegrees=45
  swing2Motion=Free
  twistLimitDegrees=45
  twistMotion=Free
 }
 {
  bone1Name=spine_04
  bone2Name=spine_03
  swing1LimitDegrees=45
  swing1Motion=Free
  swing2LimitDegrees=45
  swing2Motion=Free
  twistLimitDegrees=45
  twistMotion=Free
 }
 {bone1Name=dyn_coat_03_L, bone2Name=dyn_coat_02_L, swing1LimitDegrees=1, swing2LimitDegrees=15, twistLimitDegrees=1}
 {bone1Name=dyn_coat_03_R, bone2Name=dyn_coat_02_R, swing1LimitDegrees=0.2, swing2LimitDegrees=4, twistLimitDegrees=0.2}
 {bone1Name=dyn_coat_03_fL, bone2Name=dyn_coat_02_fL, swing1LimitDegrees=5, swing2LimitDegrees=20, twistLimitDegrees=1}
 {bone1Name=dyn_coat_03_fR, bone2Name=dyn_coat_02_fR, swing1LimitDegrees=5, swing2LimitDegrees=20, twistLimitDegrees=1}
 {bone1Name=dyn_coat_bk_03, bone2Name=dyn_coat_bk_02, swing1LimitDegrees=2, swing2LimitDegrees=5, twistLimitDegrees=1}
 {bone1Name=dyn_coat_fr_03_L, bone2Name=dyn_coat_fr_02_L, swing1LimitDegrees=5, swing2LimitDegrees=20, twistLimitDegrees=1}
 {bone1Name=dyn_coat_fr_03_R, bone2Name=dyn_coat_fr_02_R, swing1LimitDegrees=5, swing2LimitDegrees=40, twistLimitDegrees=1}
 {bone1Name=dyn_coat_mid_03_L, bone2Name=dyn_coat_mid_02_L, swing1LimitDegrees=1, swing2LimitDegrees=13, twistLimitDegrees=1}
 {bone1Name=dyn_coat_mid_03_R, bone2Name=dyn_coat_mid_02_R, swing1LimitDegrees=1, swing2LimitDegrees=13, twistLimitDegrees=1}
 {
  bone1Name=spine_03
  bone2Name=spine_02
  swing1LimitDegrees=45
  swing1Motion=Free
  swing2LimitDegrees=45
  swing2Motion=Free
  twistLimitDegrees=45
  twistMotion=Free
 }
 {
  bone1Name=spine_02
  bone2Name=spine_01
  swing1LimitDegrees=45
  swing1Motion=Free
  swing2LimitDegrees=45
  swing2Motion=Free
  twistLimitDegrees=45
  twistMotion=Free
 }
 {bone1Name=dyn_coat_fr_03_zipper01_L, bone2Name=dyn_coat_fr_03_L, swing1LimitDegrees=5, swing2LimitDegrees=20, twistLimitDegrees=1}
 {bone1Name=dyn_coat_03_belt03_R, bone2Name=dyn_coat_03_belt02_R, swing1LimitDegrees=45, swing2LimitDegrees=45, twistLimitDegrees=3}
 {bone1Name=dyn_coat_03_belt02_R, bone2Name=dyn_coat_03_belt01_R, swing1LimitDegrees=45, swing2LimitDegrees=45, twistLimitDegrees=3}
 {bone1Name=dyn_coat_03_belt01_R, bone2Name=dyn_coat_03_R, swing1LimitDegrees=45, swing2LimitDegrees=45, twistLimitDegrees=3}
 {bone1Name=dyn_coat_03_belt02_L, bone2Name=dyn_coat_03_belt01_L, swing1LimitDegrees=10, swing2LimitDegrees=25, twistLimitDegrees=3}
 {bone1Name=dyn_coat_03_belt01_L, bone2Name=dyn_coat_03_L, swing1LimitDegrees=10, swing2LimitDegrees=25, twistLimitDegrees=3}
 {
  bone1Name=lowerarm_l
  bone2Name=upperarm_l
  swing1LimitDegrees=45
  swing1Motion=Free
  swing2LimitDegrees=45
  swing2Motion=Free
  twistLimitDegrees=45
  twistMotion=Free
 }
 {
  bone1Name=upperarm_twist_01_l
  bone2Name=upperarm_l
  swing1LimitDegrees=45
  swing1Motion=Free
  swing2LimitDegrees=45
  swing2Motion=Free
  twistLimitDegrees=45
  twistMotion=Free
 }
 {
  bone1Name=upperarm_twist_01_r
  bone2Name=upperarm_r
  swing1LimitDegrees=45
  swing1Motion=Free
  swing2LimitDegrees=45
  swing2Motion=Free
  twistLimitDegrees=45
  twistMotion=Free
 }
 {
  bone1Name=dyn_coat_mid_03_L
  bone2Name=dyn_coat_bk_03
  swing1LimitDegrees=45
  swing1Motion=Free
  swing2LimitDegrees=45
  swing2Motion=Free
  twistLimitDegrees=45
  twistMotion=Free
 }
 {
  bone1Name=dyn_coat_mid_03_R
  bone2Name=dyn_coat_bk_03
  swing1LimitDegrees=45
  swing1Motion=Free
  swing2LimitDegrees=45
  swing2Motion=Free
  twistLimitDegrees=45
  twistMotion=Free
 }
 {
  bone1Name=dyn_Cloth_volume_r_C02
  bone2Name=dyn_Cloth_volume_r_C01
  swing1LimitDegrees=45
  swing1Motion=Free
  swing2LimitDegrees=45
  swing2Motion=Free
  twistLimitDegrees=45
  twistMotion=Free
 }
 {
  bone1Name=dyn_Cloth_volume_r_B02
  bone2Name=dyn_Cloth_volume_r_B01
  swing1LimitDegrees=45
  swing1Motion=Free
  swing2LimitDegrees=45
  swing2Motion=Free
  twistLimitDegrees=45
  twistMotion=Free
 }
 {
  bone1Name=dyn_Cloth_volume_r_B01
  bone2Name=dyn_Cloth_volume_r_B
  swing1LimitDegrees=45
  swing1Motion=Free
  swing2LimitDegrees=45
  swing2Motion=Free
  twistLimitDegrees=45
  twistMotion=Free
 }
 {
  bone1Name=dyn_Cloth_volume_r_A02
  bone2Name=dyn_Cloth_volume_r_A01
  swing1LimitDegrees=45
  swing1Motion=Free
  swing2LimitDegrees=45
  swing2Motion=Free
  twistLimitDegrees=45
  twistMotion=Free
 }
 {
  bone1Name=dyn_Cloth_volume_r_A01
  bone2Name=dyn_Cloth_volume_r_A
  swing1LimitDegrees=45
  swing1Motion=Free
  swing2LimitDegrees=45
  swing2Motion=Free
  twistLimitDegrees=45
  twistMotion=Free
 }
 {
  bone1Name=dyn_Cloth_volume_l_C02
  bone2Name=dyn_Cloth_volume_l_C01
  swing1LimitDegrees=45
  swing1Motion=Free
  swing2LimitDegrees=45
  swing2Motion=Free
  twistLimitDegrees=45
  twistMotion=Free
 }
 {
  bone1Name=dyn_Cloth_volume_l_B02
  bone2Name=dyn_Cloth_volume_l_B01
  swing1LimitDegrees=45
  swing1Motion=Free
  swing2LimitDegrees=45
  swing2Motion=Free
  twistLimitDegrees=45
  twistMotion=Free
 }
 {
  bone1Name=dyn_Cloth_volume_l_B01
  bone2Name=dyn_Cloth_volume_l_B
  swing1LimitDegrees=45
  swing1Motion=Free
  swing2LimitDegrees=45
  swing2Motion=Free
  twistLimitDegrees=45
  twistMotion=Free
 }
 {
  bone1Name=dyn_Cloth_volume_l_A02
  bone2Name=dyn_Cloth_volume_l_A01
  swing1LimitDegrees=45
  swing1Motion=Free
  swing2LimitDegrees=45
  swing2Motion=Free
  twistLimitDegrees=45
  twistMotion=Free
 }
 {
  bone1Name=dyn_Cloth_volume_l_A01
  bone2Name=dyn_Cloth_volume_l_A
  swing1LimitDegrees=45
  swing1Motion=Free
  swing2LimitDegrees=45
  swing2Motion=Free
  twistLimitDegrees=45
  twistMotion=Free
 }
 {
  bone1Name=dyn_Cloth_volume_fn_02
  bone2Name=dyn_Cloth_volume_fn_01
  swing1LimitDegrees=45
  swing1Motion=Free
  swing2LimitDegrees=45
  swing2Motion=Free
  twistLimitDegrees=45
  twistMotion=Free
 }
 {
  bone1Name=dyn_Cloth_volume_fn_01
  bone2Name=dyn_Cloth_volume_fn
  swing1LimitDegrees=45
  swing1Motion=Free
  swing2LimitDegrees=45
  swing2Motion=Free
  twistLimitDegrees=45
  twistMotion=Free
 }
 {
  bone1Name=dyn_Cloth_volume_bk02
  bone2Name=pelvis
  swing1LimitDegrees=45
  swing1Motion=Free
  swing2LimitDegrees=45
  swing2Motion=Free
  twistLimitDegrees=45
  twistMotion=Free
 }
 {
  bone1Name=dyn_Cloth_volume_fn
  bone2Name=pelvis
  swing1LimitDegrees=45
  swing1Motion=Free
  swing2LimitDegrees=45
  swing2Motion=Free
  twistLimitDegrees=45
  twistMotion=Free
 }
 {
  bone1Name=dyn_Cloth_volume_l_A
  bone2Name=pelvis
  swing1LimitDegrees=45
  swing1Motion=Free
  swing2LimitDegrees=45
  swing2Motion=Free
  twistLimitDegrees=45
  twistMotion=Free
 }
 {
  bone1Name=dyn_Cloth_volume_r_A
  bone2Name=pelvis
  swing1LimitDegrees=45
  swing1Motion=Free
  swing2LimitDegrees=45
  swing2Motion=Free
  twistLimitDegrees=45
  twistMotion=Free
 }
 {
  bone1Name=dyn_Cloth_volume_l_B
  bone2Name=pelvis
  swing1LimitDegrees=45
  swing1Motion=Free
  swing2LimitDegrees=45
  swing2Motion=Free
  twistLimitDegrees=45
  twistMotion=Free
 }
 {
  bone1Name=dyn_Cloth_volume_r_B
  bone2Name=pelvis
  swing1LimitDegrees=45
  swing1Motion=Free
  swing2LimitDegrees=45
  swing2Motion=Free
  twistLimitDegrees=45
  twistMotion=Free
 }
 {
  bone1Name=dyn_coat_mid_03_R
  bone2Name=dyn_coat_03_R
  swing1LimitDegrees=45
  swing1Motion=Free
  swing2LimitDegrees=45
  swing2Motion=Free
  twistLimitDegrees=45
  twistMotion=Free
 }
 {
  bone1Name=dyn_coat_mid_03_L
  bone2Name=dyn_coat_03_L
  swing1LimitDegrees=45
  swing1Motion=Free
  swing2LimitDegrees=45
  swing2Motion=Free
  twistLimitDegrees=45
  twistMotion=Free
 }
 {
  bone1Name=dyn_coat_02_R
  bone2Name=dyn_coat_mid_02_R
  swing1LimitDegrees=45
  swing1Motion=Free
  swing2LimitDegrees=45
  swing2Motion=Free
  twistLimitDegrees=45
  twistMotion=Free
 }
 {
  bone1Name=dyn_coat_02_L
  bone2Name=dyn_coat_mid_02_L
  swing1LimitDegrees=45
  swing1Motion=Free
  swing2LimitDegrees=45
  swing2Motion=Free
  twistLimitDegrees=45
  twistMotion=Free
 }
 {
  bone1Name=dyn_Cloth_volume_l_C01
  bone2Name=dyn_Cloth_volume_l_C
  swing1LimitDegrees=45
  swing1Motion=Free
  swing2LimitDegrees=45
  swing2Motion=Free
  twistLimitDegrees=45
  twistMotion=Free
 }
 {
  bone1Name=dyn_Cloth_volume_r_C01
  bone2Name=dyn_Cloth_volume_r_C
  swing1LimitDegrees=45
  swing1Motion=Free
  swing2LimitDegrees=45
  swing2Motion=Free
  twistLimitDegrees=45
  twistMotion=Free
 }
 {
  bone1Name=dyn_Cloth_volume_r_C
  bone2Name=pelvis
  swing1LimitDegrees=45
  swing1Motion=Free
  swing2LimitDegrees=45
  swing2Motion=Free
  twistLimitDegrees=45
  twistMotion=Free
 }
 {
  bone1Name=dyn_Cloth_volume_l_C
  bone2Name=pelvis
  swing1LimitDegrees=45
  swing1Motion=Free
  swing2LimitDegrees=45
  swing2Motion=Free
  twistLimitDegrees=45
  twistMotion=Free
 }
 {
  bone1Name=dyn_coat_fr_03_L
  bone2Name=dyn_coat_03_fL
  swing1LimitDegrees=45
  swing1Motion=Locked
  swing2LimitDegrees=45
  swing2Motion=Locked
  twistLimitDegrees=10
 }
 {
  bone1Name=dyn_coat_fr_02_L
  bone2Name=dyn_coat_02_fL
  swing1LimitDegrees=45
  swing1Motion=Locked
  swing2LimitDegrees=45
  swing2Motion=Locked
  twistLimitDegrees=45
  twistMotion=Locked
 }
 {bone1Name=dyn_coat_03_fL, bone2Name=dyn_coat_03_L, swing1LimitDegrees=45, swing2LimitDegrees=45, twistLimitDegrees=45}
 {
  bone1Name=thigh_r
  bone2Name=pelvis
  swing1LimitDegrees=45
  swing1Motion=Free
  swing2LimitDegrees=45
  swing2Motion=Free
  twistLimitDegrees=45
  twistMotion=Free
 }
 {
  bone1Name=thigh_l
  bone2Name=pelvis
  swing1LimitDegrees=45
  swing1Motion=Free
  swing2LimitDegrees=45
  swing2Motion=Free
  twistLimitDegrees=45
  twistMotion=Free
 }
 {
  bone1Name=upperarm_twist_01_belt04_l
  bone2Name=upperarm_twist_01_belt03_l
  swing1LimitDegrees=30
  swing2LimitDegrees=70
  twistLimitDegrees=70
 }
 {
  bone1Name=upperarm_twist_01_belt03_l
  bone2Name=upperarm_twist_01_belt02_l
  swing1LimitDegrees=30
  swing2LimitDegrees=70
  twistLimitDegrees=70
 }
 {
  bone1Name=upperarm_twist_01_belt02_l
  bone2Name=upperarm_twist_01_belt01_l
  swing1LimitDegrees=30
  swing2LimitDegrees=70
  twistLimitDegrees=70
 }
 {bone1Name=upperarm_twist_01_belt01_l, bone2Name=upperarm_twist_01_l, swing1LimitDegrees=30, swing2LimitDegrees=70, twistLimitDegrees=70}
 {bone1Name=upperarm_twist_01_belt_r, bone2Name=upperarm_twist_01_r, swing1LimitDegrees=10, swing2LimitDegrees=50, twistLimitDegrees=20}
 {bone1Name=upperarm_twist_02_belt_r, bone2Name=upperarm_twist_01_belt_r, swing1LimitDegrees=25, swing2LimitDegrees=55, twistLimitDegrees=65}
 {
  bone1Name=spine_01
  bone2Name=pelvis
  swing1LimitDegrees=45
  swing1Motion=Free
  swing2LimitDegrees=45
  swing2Motion=Free
  twistLimitDegrees=45
  twistMotion=Free
 }
 {
  bone1Name=thigh_twist_01_l
  bone2Name=thigh_l
  swing1LimitDegrees=45
  swing1Motion=Free
  swing2LimitDegrees=45
  swing2Motion=Free
  twistLimitDegrees=45
  twistMotion=Free
 }
 {bone1Name=dyn_coat_03_fR, bone2Name=dyn_coat_03_R, swing1LimitDegrees=45, swing2LimitDegrees=45, twistLimitDegrees=45}
 {
  bone1Name=dyn_coat_fr_03_R
  bone2Name=dyn_coat_03_fR
  swing1LimitDegrees=45
  swing1Motion=Locked
  swing2LimitDegrees=45
  swing2Motion=Locked
  twistLimitDegrees=25
 }
 {
  bone1Name=dyn_coat_fr_02_R
  bone2Name=dyn_coat_02_fR
  swing1LimitDegrees=45
  swing1Motion=Free
  swing2LimitDegrees=45
  swing2Motion=Free
  twistLimitDegrees=45
  twistMotion=Free
 }
]
