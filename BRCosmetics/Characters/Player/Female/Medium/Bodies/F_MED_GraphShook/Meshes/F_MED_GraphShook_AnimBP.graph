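# Post-process-style AnimBlueprint for the F_MED_GraphShook female body cosmetic, adding secondary motion to coat, cloth and chest bones.
# It copies the parent mesh pose into a Control Rig fed by sliding flags, then runs long per-bone AnimDynamics chains.
# A RigidBody node and further coat dynamics turn SkirtCachePose into SkirtDynPose.
# When bPbBlendEnabled is true, additive chest offsets and stacked angular and linear spring sims are layered on top.
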AnimBlueprint /BRCosmetics/Characters/Player/Female/Medium/Bodies/F_MED_GraphShook/Meshes/F_MED_GraphShook_AnimBP
target_skeleton=&/F_MED_GraphShook_Skeleton
graphs [
 AnimGraph {nodes [
   n1 = "Transform (Modify) Bone - Bone: dyn_chest_l"(ComponentPose=Local To Component(LocalPose=Use cached pose 'SkirtDynPose'()), Alpha=1, Translation="0,0,1", Rotation="0,0,0", Scale="1,1,1") {translationMode=BMM_Additive, translationSpace=BCS_BoneSpace, boneToModify {boneName=dyn_chest_l}, translation {z=1}}
   n2 = "Transform (Modify) Bone - Bone: dyn_chest_r"(ComponentPose=@n1, Alpha=1, Translation="0,0,1", Rotation="0,0,0", Scale="1,1,1") {translationMode=BMM_Additive, translationSpace=BCS_BoneSpace, boneToModify {boneName=dyn_chest_r}, translation {z=1}}
   n3 = Get Sim Space Settings() -> (Sim Space Settings:Anim Phys Sim Space Settings Structure)
   n4 = Anim Dynamics - Bone: dyn_chest_r!(ComponentPose=@n2, Alpha=1, AngularSpringConstant:float=5000, SimSpaceSettings:Anim Phys Sim Space Settings Structure=@n3) -> (Pose) {
    angularBiasOverride=1
    angularDampingOverride=0.9999
    angularSpringConstant=5000
    bAngularSpring=true
    bGravityOverrideInSimSpace=true
    bOverrideAngularBias=true
    bOverrideAngularDamping=true
    bUseGravityOverride=true
    linearDampingOverride=0.9999
    linearSpringConstant=100
    simulationSpace=RootRelative
    boundBone {boneName=dyn_chest_r}
    componentAppliedLinearAccClamp=<2 2 2>
    gravityOverride {z=-980}
    physicsBodyDefinitions [{
      boundBone {boneName=dyn_chest_r}
      boxExtents=<26 26 26>
      constraintSetup {angularLimitsMax=<0 1 9>, angularLimitsMin=<0 -5 -9>, angularTarget=<1 0 0>}
      localJointOffset=<2 0 0>
     }]
    relativeSpaceBone {boneName=spine_04}
    simSpaceSettings {maxAngularAcceleration=45, maxAngularVelocity=8, simSpaceAngularAlpha=0.25}
   }
   n5 = Anim Dynamics - Bone: dyn_chest_l!(ComponentPose=@n4, Alpha=1, AngularSpringConstant:float=5000, SimSpaceSettings:Anim Phys Sim Space Settings Structure=@n3) -> (Pose) {
    angularBiasOverride=1
    angularDampingOverride=0.9999
    angularSpringConstant=5000
    bAngularSpring=true
    bGravityOverrideInSimSpace=true
    bOverrideAngularBias=true
    bOverrideAngularDamping=true
    bUseGravityOverride=true
    linearDampingOverride=0.9999
    linearSpringConstant=100
    simulationSpace=RootRelative
    boundBone {boneName=dyn_chest_l}
    componentAppliedLinearAccClamp=<2 2 2>
    gravityOverride {z=-980}
    physicsBodyDefinitions [{
      boundBone {boneName=dyn_chest_l}
      boxExtents=<25 25 25>
      constraintSetup {angularLimitsMax=<0 1 9>, angularLimitsMin=<0 -5 -9>, angularTarget=<1 0 0>}
      localJointOffset=<2 0 0>
     }]
    relativeSpaceBone {boneName=spine_04}
    simSpaceSettings {maxAngularAcceleration=45, maxAngularVelocity=8, simSpaceAngularAlpha=0.25}
   }
   n6 = Anim Dynamics - Bone: dyn_chest_r!(ComponentPose=@n5, Alpha=1, GravityOverride:Vector="0,0,-200", LinearSpringConstant:float=180) -> (Pose) {
    angularBiasOverride=1
    angularDampingOverride=0.99
    angularSpringConstant=10000
    bGravityOverrideInSimSpace=true
    bLinearSpring=true
    bOverrideAngularBias=true
    bOverrideLinearDamping=true
    bUseGravityOverride=true
    linearDampingOverride=1
    linearSpringConstant=180
    simulationSpace=RootRelative
    boundBone {boneName=dyn_chest_r}
    componentAppliedLinearAccClamp=<5 5 5>
    gravityOverride {z=-200}
    physicsBodyDefinitions [{
      boundBone {boneName=dyn_chest_r}
      boxExtents=<80 80 80>
      constraintSetup {angularLimitsMax=<0 0 0>, angularLimitsMin=<0 0 0>, angularTarget=<1 0 0>, linearAxesMax=<0 0.5 1>, linearAxesMin=<0 -0.5 -0.5>}
      localJointOffset=<0.2 0 0>
     }]
    relativeSpaceBone {boneName=spine_04}
    simSpaceSettings {maxAngularAcceleration=50, maxAngularVelocity=10, simSpaceAngularAlpha=0.1}
   }
   n7 = Anim Dynamics - Bone: dyn_chest_l!(ComponentPose=@n6, Alpha=1, GravityOverride:Vector="0,0,-200", LinearSpringConstant:float=180) -> (Pose) {
    angularBiasOverride=1
    angularDampingOverride=0.99
    angularSpringConstant=10000
    bGravityOverrideInSimSpace=true
    bLinearSpring=true
    bOverrideAngularBias=true
    bOverrideLinearDamping=true
    bUseGravityOverride=true
    linearDampingOverride=1
    linearSpringConstant=180
    simulationSpace=RootRelative
    boundBone {boneName=dyn_chest_l}
    componentAppliedLinearAccClamp=<5 5 5>
    gravityOverride {z=-200}
    physicsBodyDefinitions [{
      boundBone {boneName=dyn_chest_l}
      boxExtents=<90 90 90>
      constraintSetup {angularLimitsMax=<0 0 0>, angularLimitsMin=<0 0 0>, angularTarget=<1 0 0>, linearAxesMax=<0 0.5 1>, linearAxesMin=<0 -0.5 -0.5>}
      localJointOffset=<0.2 0 0>
     }]
    relativeSpaceBone {boneName=spine_04}
    simSpaceSettings {maxAngularAcceleration=50, maxAngularVelocity=10, simSpaceAngularAlpha=0.1}
   }
   Output Pose(Result=Blend Poses by bool(
     BlendPose_0=Component To Local(ComponentPose=@n7)
     BlendPose_1=Use cached pose 'SkirtDynPose'()
     bActiveValue=Get bPbBlendEnabled() -> (bPbBlendEnabled:bool)
    )) {layerGroup=None}
   n8 = RigidBody!(
    ComponentPose=Local To Component(LocalPose=Use cached pose 'SkirtCachePose'())
    Alpha
    OverrideWorldGravity:Vector="0,0,0"
    ExternalForce
    SimSpaceSettings:Sim Space Settings Structure=(WorldAlpha=1,VelocityScaleZ=0.5,DampingAlpha=0,MaxLinearVelocity=600,MaxAngularVelocity=10,MaxLinearAcceleration=1200,MaxAngularAcceleration=40,ExternalLinearDragV=(X=0.3,Y=0.3,Z=0.3),ExternalLinearVelocity=(X=0,Y=0,Z=0),ExternalAngularVelocity=(X=0,Y=0,Z=0))
   ) -> (Pose) {
    bClampLinearTranslationLimitToRefPose=true
    bOverrideWorldGravity=true
    overridePhysicsAsset=&/F_MED_GraphShook_Physics
    simulationSpace=BaseBoneSpace
    alphaBoolBlend {blendOption=Linear}
    baseBoneRef {boneName=spine_05}
    overrideWorldGravity {z=-980}
    simSpaceSettings {
     dampingAlpha=0
     maxAngularAcceleration=40
     maxAngularVelocity=10
     maxLinearAcceleration=1200
     maxLinearVelocity=600
     velocityScaleZ=0.5
     worldAlpha=1
     externalLinearDragV=<0.3 0.3 0.3>
    }
   }
   n9 = Anim Dynamics - Bone: dyn_coat_mid_01_hat_L!(ComponentPose=@n8, bAlphaBoolEnabled:bool=True, GravityOverride:Vector="0,0,500", ExternalForce="0,0,0") -> (Pose) {
    alphaInputType=Bool
    angularDampingOverride=1
    angularSpringConstant=5000
    bAngularSpring=true
    bOverrideAngularDamping=true
    bUseGravityOverride=true
    gravityScale=0.3
    simulationSpace=BoneRelative
    alphaBoolBlend {blendOption=Linear}
    boundBone {boneName=dyn_coat_mid_01_hat_L}
    componentAppliedLinearAccClamp=<2500 2500 1500>
    componentLinearAccScale=<0.15 0.15 0.15>
    componentLinearVelScale=<0.35 0.35 0.35>
    gravityOverride {z=500}
    physicsBodyDefinitions [{
      collisionType=CustomSphere
      sphereCollisionRadius=2
      boundBone {boneName=dyn_coat_mid_01_hat_L}
      boxExtents=<32 20 20>
      constraintSetup {angularLimitsMax=<0 0 20>, angularLimitsMin=<0 -35 0>, angularTarget=<0 0 0>}
      localJointOffset=<5 0.1 0>
     }]
    relativeSpaceBone {boneName=root}
    simSpaceSettings {maxAngularAcceleration=500, maxAngularVelocity=50}
   }
   n10 = Anim Dynamics - Bone: dyn_coat_mid_01_hat_R!(ComponentPose=@n9, bAlphaBoolEnabled:bool=True, GravityOverride:Vector="0,0,-980", ExternalForce="0,0,0") -> (Pose) {
    alphaInputType=Bool
    angularDampingOverride=1
    angularSpringConstant=5000
    bAngularSpring=true
    bOverrideAngularDamping=true
    bUseGravityOverride=true
    gravityScale=0.3
    simulationSpace=BoneRelative
    alphaBoolBlend {blendOption=Linear}
    boundBone {boneName=dyn_coat_mid_01_hat_R}
    componentAppliedLinearAccClamp=<2500 2500 1500>
    componentLinearAccScale=<0.15 0.15 0.15>
    componentLinearVelScale=<0.35 0.35 0.35>
    gravityOverride {z=-980}
    physicsBodyDefinitions [{
      collisionType=CustomSphere
      sphereCollisionRadius=2
      boundBone {boneName=dyn_coat_mid_01_hat_R}
      boxExtents=<32 20 20>
      constraintSetup {angularLimitsMax=<0 0 10>, angularLimitsMin=<0 -35 -20>, angularTarget=<0 0 0>}
      localJointOffset=<5 0.1 0>
     }]
    relativeSpaceBone {boneName=root}
    simSpaceSettings {maxAngularAcceleration=500, maxAngularVelocity=50}
   }
   n11 = Anim Dynamics - Bone: dyn_coat_bk_01_hat!(ComponentPose=@n10, bAlphaBoolEnabled:bool=True, GravityOverride:Vector="0,0,500", ExternalForce="0,0,0") -> (Pose) {
    alphaInputType=Bool
    angularDampingOverride=1
    angularSpringConstant=5000
    bAngularSpring=true
    bOverrideAngularDamping=true
    bUseGravityOverride=true
    gravityScale=0.3
    simulationSpace=BoneRelative
    alphaBoolBlend {blendOption=Linear}
    boundBone {boneName=dyn_coat_bk_01_hat}
    componentAppliedLinearAccClamp=<2500 2500 1500>
    componentLinearAccScale=<0.15 0.15 0.15>
    componentLinearVelScale=<0.35 0.35 0.35>
    gravityOverride {z=500}
    physicsBodyDefinitions [{
      collisionType=CustomSphere
      sphereCollisionRadius=2
      boundBone {boneName=dyn_coat_bk_01_hat}
      boxExtents=<32 20 20>
      constraintSetup {angularLimitsMax=<0 10 10>, angularLimitsMin=<0 -10 -20>, angularTarget=<0 0 0>}
      localJointOffset=<5 0.1 1>
     }]
    relativeSpaceBone {boneName=root}
    simSpaceSettings {maxAngularAcceleration=500, maxAngularVelocity=50}
   }
   n12 = Anim Dynamics - Bone: dyn_coat_mid_02_hat_L!(ComponentPose=@n11, bAlphaBoolEnabled:bool=True, GravityOverride:Vector="0,0,500", ExternalForce="0,0,0") -> (Pose) {
    alphaInputType=Bool
    angularDampingOverride=1
    angularSpringConstant=5000
    bAngularSpring=true
    bOverrideAngularDamping=true
    bUseGravityOverride=true
    gravityScale=0.3
    simulationSpace=BoneRelative
    alphaBoolBlend {blendOption=Linear}
    boundBone {boneName=dyn_coat_mid_02_hat_L}
    componentAppliedLinearAccClamp=<2500 2500 1500>
    componentLinearAccScale=<0.15 0.15 0.15>
    componentLinearVelScale=<0.35 0.35 0.35>
    gravityOverride {z=500}
    physicsBodyDefinitions [{
      collisionType=CustomSphere
      sphereCollisionRadius=2
      boundBone {boneName=dyn_coat_mid_02_hat_L}
      boxExtents=<18 12 12>
      constraintSetup {angularLimitsMax=<0 0 10>, angularLimitsMin=<0 -10 0>, angularTarget=<0 0 0>}
      localJointOffset=<5 -0.1 0>
     }]
    relativeSpaceBone {boneName=root}
    simSpaceSettings {maxAngularAcceleration=500, maxAngularVelocity=50}
   }
   n13 = Anim Dynamics - Bone: dyn_coat_mid_02_hat_R!(ComponentPose=@n12, bAlphaBoolEnabled:bool=True, GravityOverride:Vector="0,0,-980", ExternalForce="0,0,0") -> (Pose) {
    alphaInputType=Bool
    angularDampingOverride=1
    angularSpringConstant=5000
    bAngularSpring=true
    bOverrideAngularDamping=true
    bUseGravityOverride=true
    gravityScale=0.3
    simulationSpace=BoneRelative
    alphaBoolBlend {blendOption=Linear}
    boundBone {boneName=dyn_coat_mid_02_hat_R}
    componentAppliedLinearAccClamp=<2500 2500 1500>
    componentLinearAccScale=<0.15 0.15 0.15>
    componentLinearVelScale=<0.35 0.35 0.35>
    gravityOverride {z=-980}
    physicsBodyDefinitions [{
      collisionType=CustomSphere
      sphereCollisionRadius=2
      boundBone {boneName=dyn_coat_mid_02_hat_R}
      boxExtents=<18 12 12>
      constraintSetup {angularLimitsMax=<0 0 10>, angularLimitsMin=<0 -10 -20>, angularTarget=<0 0 0>}
      localJointOffset=<5 -0.1 0>
     }]
    relativeSpaceBone {boneName=root}
    simSpaceSettings {maxAngularAcceleration=500, maxAngularVelocity=50}
   }
   n14 = Anim Dynamics - Bone: dyn_coat_bk_02_hat!(ComponentPose=@n13, bAlphaBoolEnabled:bool=True, GravityOverride:Vector="0,0,500", ExternalForce="0,0,0") -> (Pose) {
    alphaInputType=Bool
    angularDampingOverride=1
    angularSpringConstant=5000
    bAngularSpring=true
    bOverrideAngularDamping=true
    bUseGravityOverride=true
    gravityScale=0.3
    simulationSpace=BoneRelative
    alphaBoolBlend {blendOption=Linear}
    boundBone {boneName=dyn_coat_bk_02_hat}
    componentAppliedLinearAccClamp=<2500 2500 1500>
    componentLinearAccScale=<0.15 0.15 0.15>
    componentLinearVelScale=<0.35 0.35 0.35>
    gravityOverride {z=500}
    physicsBodyDefinitions [{
      collisionType=CustomSphere
      sphereCollisionRadius=2
      boundBone {boneName=dyn_coat_bk_02_hat}
      boxExtents=<18 12 12>
      constraintSetup {angularLimitsMax=<0 10 0>, angularLimitsMin=<0 -10 -10>, angularTarget=<0 0 0>}
      localJointOffset=<5 -0.1 0>
     }]
    relativeSpaceBone {boneName=root}
    simSpaceSettings {maxAngularAcceleration=500, maxAngularVelocity=50}
   }
   Save cached pose 'SkirtDynPose'(Pose=Component To Local(ComponentPose=@n14)) {cacheName=SkirtDynPose}
   n15 = Control Rig!(
    Source=Copy Pose From Mesh() {bCopyCurves=true, bUseAttachedParent=true}
    Alpha
    isSliding:bool=Get bIsSliding()
    isBaseballSliding:bool=Get bIsBaseballSliding()
   ) -> (Pose) {
    alphaBoolBlend {blendOption=Linear}
    controlRigAssetReference {blueprintRigClass=&/F_MED_GraphShook_CtrlRig.F_MED_GraphShook_CtrlRig_C}
   }
   Save cached pose 'control rig_Cache'(Pose=@n15) {cacheName=control rig_Cache}
   n16 = Anim Dynamics - Bone: dyn_coat_fr_01_L!(
    ComponentPose=Local To Component(LocalPose=Use cached pose 'control rig_Cache'())
    Alpha=1
    GravityOverride:Vector="0,0,500"
    ExternalForce="0,0,0"
   ) -> (Pose) {
    angularDampingOverride=0.99
    angularSpringConstant=5000
    bAngularSpring=true
    bOverrideAngularDamping=true
    bUseGravityOverride=true
    gravityScale=0.3
    simulationSpace=BoneRelative
    alphaBoolBlend {blendOption=Linear}
    boundBone {boneName=dyn_coat_fr_01_L}
    componentAppliedLinearAccClamp=<2500 2500 1500>
    componentLinearAccScale=<0.3 0.3 0.3>
    componentLinearVelScale=<0.35 0.35 0.35>
    gravityOverride {z=500}
    physicsBodyDefinitions [{
      collisionType=CustomSphere
      sphereCollisionRadius=2
      boundBone {boneName=dyn_coat_fr_01_L}
      boxExtents=<45 45 45>
      constraintSetup {twistAxis=AxisZ, angularLimitsMax=<0 0.5 3>, angularLimitsMin=<0 -20 -3>, angularTarget=<0 0 0>}
      localJointOffset=<10 0.2 -0.1>
     }]
    relativeSpaceBone {boneName=root}
    simSpaceSettings {maxAngularAcceleration=500, maxAngularVelocity=50}
   }
   n17 = Anim Dynamics - Bone: dyn_coat_fr_01_R!(ComponentPose=@n16, Alpha=1, GravityOverride:Vector="0,0,-980", ExternalForce="0,0,0") -> (Pose) {
    angularDampingOverride=0.99
    angularSpringConstant=5000
    bAngularSpring=true
    bOverrideAngularDamping=true
    bUseGravityOverride=true
    gravityScale=0.3
    simulationSpace=BoneRelative
    alphaBoolBlend {blendOption=Linear}
    boundBone {boneName=dyn_coat_fr_01_R}
    componentAppliedLinearAccClamp=<2500 2500 1500>
    componentLinearAccScale=<0.3 0.3 0.3>
    componentLinearVelScale=<0.35 0.35 0.35>
    gravityOverride {z=-980}
    physicsBodyDefinitions [{
      collisionType=CustomSphere
      sphereCollisionRadius=2
      boundBone {boneName=dyn_coat_fr_01_R}
      boxExtents=<45 45 45>
      constraintSetup {twistAxis=AxisZ, angularLimitsMax=<0 0.5 3>, angularLimitsMin=<0 -20 -3>, angularTarget=<0 0 0>}
      localJointOffset=<10 -0.2 -1>
     }]
    relativeSpaceBone {boneName=root}
    simSpaceSettings {maxAngularAcceleration=500, maxAngularVelocity=50}
   }
   n18 = Anim Dynamics - Bone: dyn_coat_01_fL!(ComponentPose=@n17, Alpha=1, GravityOverride:Vector="0,0,500", ExternalForce="0,0,0") -> (Pose) {
    angularDampingOverride=0.99
    angularSpringConstant=5000
    bAngularSpring=true
    bOverrideAngularDamping=true
    bUseGravityOverride=true
    gravityScale=0.3
    simulationSpace=BoneRelative
    alphaBoolBlend {blendOption=Linear}
    boundBone {boneName=dyn_coat_01_fL}
    componentAppliedLinearAccClamp=<2500 2500 1500>
    componentLinearAccScale=<0.3 0.3 0.3>
    componentLinearVelScale=<0.35 0.35 0.35>
    gravityOverride {z=500}
    physicsBodyDefinitions [{
      collisionType=CustomSphere
      sphereCollisionRadius=2
      boundBone {boneName=dyn_coat_01_fL}
      boxExtents=<45 45 45>
      constraintSetup {twistAxis=AxisZ, angularLimitsMax=<0 0.5 3>, angularLimitsMin=<0 -20 -3>, angularTarget=<0 0 0>}
      localJointOffset=<10 0.2 -0.1>
     }]
    relativeSpaceBone {boneName=root}
    simSpaceSettings {maxAngularAcceleration=500, maxAngularVelocity=50}
   }
   n19 = Anim Dynamics - Bone: dyn_coat_01_fR!(ComponentPose=@n18, Alpha=1, GravityOverride:Vector="0,0,-980", ExternalForce="0,0,0") -> (Pose) {
    angularDampingOverride=0.99
    angularSpringConstant=5000
    bAngularSpring=true
    bOverrideAngularDamping=true
    bUseGravityOverride=true
    gravityScale=0.3
    simulationSpace=BoneRelative
    alphaBoolBlend {blendOption=Linear}
    boundBone {boneName=dyn_coat_01_fR}
    componentAppliedLinearAccClamp=<2500 2500 1500>
    componentLinearAccScale=<0.3 0.3 0.3>
    componentLinearVelScale=<0.35 0.35 0.35>
    gravityOverride {z=-980}
    physicsBodyDefinitions [{
      collisionType=CustomSphere
      sphereCollisionRadius=2
      boundBone {boneName=dyn_coat_01_fR}
      boxExtents=<45 45 45>
      constraintSetup {twistAxis=AxisZ, angularLimitsMax=<0 0.5 3>, angularLimitsMin=<0 -20 -3>, angularTarget=<0 0 0>}
      localJointOffset=<10 -0.2 -0.1>
     }]
    relativeSpaceBone {boneName=root}
    simSpaceSettings {maxAngularAcceleration=500, maxAngularVelocity=50}
   }
   n20 = Anim Dynamics - Bone: dyn_coat_01_L!(ComponentPose=@n19, Alpha=1, GravityOverride:Vector="0,0,500", ExternalForce="0,0,0") -> (Pose) {
    angularDampingOverride=0.9999
    angularSpringConstant=5000
    bAngularSpring=true
    bOverrideAngularDamping=true
    bUseGravityOverride=true
    gravityScale=0.3
    simulationSpace=BoneRelative
    alphaBoolBlend {blendOption=Linear}
    boundBone {boneName=dyn_coat_01_L}
    componentAppliedLinearAccClamp=<2500 2500 1500>
    componentLinearAccScale=<0.3 0.3 0.3>
    componentLinearVelScale=<0.35 0.35 0.35>
    gravityOverride {z=500}
    physicsBodyDefinitions [{
      collisionType=CustomSphere
      sphereCollisionRadius=2
      boundBone {boneName=dyn_coat_01_L}
      boxExtents=<20 10 10>
      constraintSetup {twistAxis=AxisZ, angularLimitsMax=<0 4 5>, angularLimitsMin=<0 -10 -5>, angularTarget=<0 0 0>}
      localJointOffset=<2.5 0 0.5>
     }]
    relativeSpaceBone {boneName=root}
    simSpaceSettings {maxAngularAcceleration=500, maxAngularVelocity=50}
   }
   n21 = Anim Dynamics - Bone: dyn_coat_01_R!(ComponentPose=@n20, Alpha=1, GravityOverride:Vector="0,0,-980", ExternalForce="0,0,0") -> (Pose) {
    angularDampingOverride=0.9999
    angularSpringConstant=5000
    bAngularSpring=true
    bOverrideAngularDamping=true
    bUseGravityOverride=true
    gravityScale=0.3
    simulationSpace=BoneRelative
    alphaBoolBlend {blendOption=Linear}
    boundBone {boneName=dyn_coat_01_R}
    componentAppliedLinearAccClamp=<2500 2500 1500>
    componentLinearAccScale=<0.3 0.3 0.3>
    componentLinearVelScale=<0.35 0.35 0.35>
    gravityOverride {z=-980}
    physicsBodyDefinitions [{
      collisionType=CustomSphere
      sphereCollisionRadius=2
      boundBone {boneName=dyn_coat_01_R}
      boxExtents=<20 10 10>
      constraintSetup {twistAxis=AxisZ, angularLimitsMax=<0 4 5>, angularLimitsMin=<0 -10 -5>, angularTarget=<0 0 0>}
      localJointOffset=<2.5 0 0.5>
     }]
    relativeSpaceBone {boneName=root}
    simSpaceSettings {maxAngularAcceleration=500, maxAngularVelocity=50}
   }
   n22 = Anim Dynamics - Bone: dyn_coat_mid_01_L!(ComponentPose=@n21, Alpha=1, GravityOverride:Vector="0,0,500", ExternalForce="0,0,0") -> (Pose) {
    angularDampingOverride=0.9999
    angularSpringConstant=5000
    bAngularSpring=true
    bOverrideAngularDamping=true
    bUseGravityOverride=true
    gravityScale=0.3
    simulationSpace=BoneRelative
    alphaBoolBlend {blendOption=Linear}
    boundBone {boneName=dyn_coat_mid_01_L}
    componentAppliedLinearAccClamp=<2500 2500 1500>
    componentLinearAccScale=<0.3 0.3 0.3>
    componentLinearVelScale=<0.35 0.35 0.35>
    gravityOverride {z=500}
    physicsBodyDefinitions [{
      collisionType=CustomSphere
      sphereCollisionRadius=2
      boundBone {boneName=dyn_coat_mid_01_L}
      boxExtents=<30 20 20>
      constraintSetup {twistAxis=AxisZ, angularLimitsMax=<0 10 6>, angularLimitsMin=<0 -2 -6>, angularTarget=<0 0 0>}
      localJointOffset=<2.5 0 0.5>
     }]
    relativeSpaceBone {boneName=root}
    retargetingSettings {bEnabled=true, rotationComponent=QuaternionAngle, source {rotation {pitch=-10}}}
    simSpaceSettings {maxAngularAcceleration=500, maxAngularVelocity=50}
   }
   n23 = Anim Dynamics - Bone: dyn_coat_mid_01_R!(ComponentPose=@n22, Alpha=1, GravityOverride:Vector="0,0,-980", ExternalForce="0,0,0") -> (Pose) {
    angularDampingOverride=0.9999
    angularSpringConstant=5000
    bAngularSpring=true
    bOverrideAngularDamping=true
    bUseGravityOverride=true
    gravityScale=0.3
    simulationSpace=BoneRelative
    alphaBoolBlend {blendOption=Linear}
    boundBone {boneName=dyn_coat_mid_01_R}
    componentAppliedLinearAccClamp=<2500 2500 1500>
    componentLinearAccScale=<0.3 0.3 0.3>
    componentLinearVelScale=<0.35 0.35 0.35>
    gravityOverride {z=-980}
    physicsBodyDefinitions [{
      collisionType=CustomSphere
      sphereCollisionRadius=2
      boundBone {boneName=dyn_coat_mid_01_R}
      boxExtents=<30 20 20>
      constraintSetup {twistAxis=AxisZ, angularLimitsMax=<0 10 6>, angularLimitsMin=<0 -2 -6>, angularTarget=<0 0 0>}
      localJointOffset=<2.5 0 0.5>
     }]
    relativeSpaceBone {boneName=root}
    retargetingSettings {bEnabled=true, rotationComponent=QuaternionAngle, source {rotation {pitch=-10}}}
    simSpaceSettings {maxAngularAcceleration=500, maxAngularVelocity=50}
   }
   n24 = Anim Dynamics - Bone: dyn_coat_bk_01!(ComponentPose=@n23, Alpha=1, GravityOverride:Vector="0,0,-980", ExternalForce="0,0,0") -> (Pose) {
    angularDampingOverride=0.9999
    angularSpringConstant=5000
    bAngularSpring=true
    bOverrideAngularDamping=true
    bUseGravityOverride=true
    gravityScale=0.3
    simulationSpace=BoneRelative
    alphaBoolBlend {blendOption=Linear}
    boundBone {boneName=dyn_coat_bk_01}
    componentAppliedLinearAccClamp=<2500 2500 1500>
    componentLinearAccScale=<0.3 0.3 0.3>
    componentLinearVelScale=<0.35 0.35 0.35>
    gravityOverride {z=-980}
    physicsBodyDefinitions [{
      collisionType=CustomSphere
      sphereCollisionRadius=2
      boundBone {boneName=dyn_coat_bk_01}
      boxExtents=<30 20 20>
      constraintSetup {twistAxis=AxisZ, angularLimitsMax=<0 10 6>, angularLimitsMin=<0 -2 -6>, angularTarget=<0 0 0>}
      localJointOffset=<2.5 0 0.5>
     }]
    relativeSpaceBone {boneName=root}
    retargetingSettings {bEnabled=true, rotationComponent=QuaternionAngle, source {rotation {pitch=-10}}}
    simSpaceSettings {maxAngularAcceleration=500, maxAngularVelocity=50}
   }
   n25 = Anim Dynamics - Bone: dyn_Cloth_volume_l_A!(ComponentPose=@n24, Alpha=1, GravityOverride:Vector="0,0,500", ExternalForce="0,0,0") -> (Pose) {
    angularDampingOverride=0.99
    angularSpringConstant=5000
    bAngularSpring=true
    bOverrideAngularDamping=true
    bUseGravityOverride=true
    gravityScale=0.3
    simulationSpace=BoneRelative
    alphaBoolBlend {blendOption=Linear}
    boundBone {boneName=dyn_Cloth_volume_l_A}
    componentAppliedLinearAccClamp=<2500 2500 1500>
    componentLinearAccScale=<0.3 0.3 0.3>
    componentLinearVelScale=<0.35 0.35 0.35>
    gravityOverride {z=500}
    physicsBodyDefinitions [{
      collisionType=CustomSphere
      sphereCollisionRadius=2
      boundBone {boneName=dyn_Cloth_volume_l_A}
      boxExtents=<45 45 45>
      constraintSetup {angularLimitsMax=<0 3 15>, angularLimitsMin=<0 -3 0>, angularTarget=<0 0 0>}
      localJointOffset=<-9 0.2 -0.1>
     }]
    relativeSpaceBone {boneName=root}
    simSpaceSettings {maxAngularAcceleration=500, maxAngularVelocity=50}
   }
   n26 = Anim Dynamics - Bone: dyn_Cloth_volume_l_A01!(ComponentPose=@n25, Alpha=1, GravityOverride:Vector="0,0,500", ExternalForce="0,0,0") -> (Pose) {
    angularDampingOverride=0.99
    angularSpringConstant=5000
    bAngularSpring=true
    bOverrideAngularDamping=true
    bUseGravityOverride=true
    gravityScale=0.3
    simulationSpace=BoneRelative
    alphaBoolBlend {blendOption=Linear}
    boundBone {boneName=dyn_Cloth_volume_l_A01}
    componentAppliedLinearAccClamp=<2500 2500 1500>
    componentLinearAccScale=<0.3 0.3 0.3>
    componentLinearVelScale=<0.35 0.35 0.35>
    gravityOverride {z=500}
    physicsBodyDefinitions [{
      collisionType=CustomSphere
      sphereCollisionRadius=2
      boundBone {boneName=dyn_Cloth_volume_l_A01}
      boxExtents=<25 15 15>
      constraintSetup {angularLimitsMax=<0 3 10>, angularLimitsMin=<0 -3 0>, angularTarget=<0 0 0>}
      localJointOffset=<-9 0.2 -0.1>
     }]
    relativeSpaceBone {boneName=root}
    simSpaceSettings {maxAngularAcceleration=500, maxAngularVelocity=50}
   }
   n27 = Anim Dynamics - Bone: dyn_Cloth_volume_l_A02!(ComponentPose=@n26, Alpha=1, GravityOverride:Vector="0,0,500", ExternalForce="0,0,0") -> (Pose) {
    angularDampingOverride=0.99
    angularSpringConstant=5000
    bAngularSpring=true
    bOverrideAngularDamping=true
    bUseGravityOverride=true
    gravityScale=0.3
    simulationSpace=BoneRelative
    alphaBoolBlend {blendOption=Linear}
    boundBone {boneName=dyn_Cloth_volume_l_A02}
    componentAppliedLinearAccClamp=<2500 2500 1500>
    componentLinearAccScale=<0.3 0.3 0.3>
    componentLinearVelScale=<0.35 0.35 0.35>
    gravityOverride {z=500}
    physicsBodyDefinitions [{
      collisionType=CustomSphere
      sphereCollisionRadius=2
      boundBone {boneName=dyn_Cloth_volume_l_A02}
      boxExtents=<12 12 12>
      constraintSetup {angularLimitsMax=<0 3 10>, angularLimitsMin=<0 -3 0>, angularTarget=<0 0 0>}
      localJointOffset=<-9 0.2 -0.1>
     }]
    relativeSpaceBone {boneName=root}
    simSpaceSettings {maxAngularAcceleration=500, maxAngularVelocity=50}
   }
   n28 = Anim Dynamics - Bone: dyn_Cloth_volume_r_A!(ComponentPose=@n27, Alpha=1, GravityOverride:Vector="0,0,-980", ExternalForce="0,0,0") -> (Pose) {
    angularDampingOverride=0.99
    angularSpringConstant=5000
    bAngularSpring=true
    bOverrideAngularDamping=true
    bUseGravityOverride=true
    gravityScale=0.3
    simulationSpace=BoneRelative
    alphaBoolBlend {blendOption=Linear}
    boundBone {boneName=dyn_Cloth_volume_r_A}
    componentAppliedLinearAccClamp=<2500 2500 1500>
    componentLinearAccScale=<0.3 0.3 0.3>
    componentLinearVelScale=<0.35 0.35 0.35>
    gravityOverride {z=-980}
    physicsBodyDefinitions [{
      collisionType=CustomSphere
      sphereCollisionRadius=2
      boundBone {boneName=dyn_Cloth_volume_r_A}
      boxExtents=<45 45 45>
      constraintSetup {angularLimitsMax=<0 3 15>, angularLimitsMin=<0 -3 0>, angularTarget=<0 0 0>}
      localJointOffset=<-9 0.2 -0.1>
     }]
    relativeSpaceBone {boneName=root}
    simSpaceSettings {maxAngularAcceleration=500, maxAngularVelocity=50}
   }
   n29 = Anim Dynamics - Bone: dyn_Cloth_volume_r_A01!(ComponentPose=@n28, Alpha=1, GravityOverride:Vector="0,0,-980", ExternalForce="0,0,0") -> (Pose) {
    angularDampingOverride=0.99
    angularSpringConstant=5000
    bAngularSpring=true
    bOverrideAngularDamping=true
    bUseGravityOverride=true
    gravityScale=0.3
    simulationSpace=BoneRelative
    alphaBoolBlend {blendOption=Linear}
    boundBone {boneName=dyn_Cloth_volume_r_A01}
    componentAppliedLinearAccClamp=<2500 2500 1500>
    componentLinearAccScale=<0.3 0.3 0.3>
    componentLinearVelScale=<0.35 0.35 0.35>
    gravityOverride {z=-980}
    physicsBodyDefinitions [{
      collisionType=CustomSphere
      sphereCollisionRadius=2
      boundBone {boneName=dyn_Cloth_volume_r_A01}
      boxExtents=<25 15 15>
      constraintSetup {angularLimitsMax=<0 3 10>, angularLimitsMin=<0 -3 0>, angularTarget=<0 0 0>}
      localJointOffset=<-9 0.2 -0.1>
     }]
    relativeSpaceBone {boneName=root}
    simSpaceSettings {maxAngularAcceleration=500, maxAngularVelocity=50}
   }
   n30 = Anim Dynamics - Bone: dyn_Cloth_volume_r_A02!(ComponentPose=@n29, Alpha=1, GravityOverride:Vector="0,0,-980", ExternalForce="0,0,0") -> (Pose) {
    angularDampingOverride=0.99
    angularSpringConstant=5000
    bAngularSpring=true
    bOverrideAngularDamping=true
    bUseGravityOverride=true
    gravityScale=0.3
    simulationSpace=BoneRelative
    alphaBoolBlend {blendOption=Linear}
    boundBone {boneName=dyn_Cloth_volume_r_A02}
    componentAppliedLinearAccClamp=<2500 2500 1500>
    componentLinearAccScale=<0.3 0.3 0.3>
    componentLinearVelScale=<0.35 0.35 0.35>
    gravityOverride {z=-980}
    physicsBodyDefinitions [{
      collisionType=CustomSphere
      sphereCollisionRadius=2
      boundBone {boneName=dyn_Cloth_volume_r_A02}
      boxExtents=<12 12 12>
      constraintSetup {angularLimitsMax=<0 3 10>, angularLimitsMin=<0 -3 0>, angularTarget=<0 0 0>}
      localJointOffset=<-9 0.2 -0.1>
     }]
    relativeSpaceBone {boneName=root}
    simSpaceSettings {maxAngularAcceleration=500, maxAngularVelocity=50}
   }
   n31 = Anim Dynamics - Bone: dyn_Cloth_volume_fn!(ComponentPose=@n30, Alpha=1, GravityOverride:Vector="0,0,500", ExternalForce="0,0,0") -> (Pose) {
    angularDampingOverride=0.99
    angularSpringConstant=5000
    bAngularSpring=true
    bOverrideAngularDamping=true
    bUseGravityOverride=true
    gravityScale=0.3
    simulationSpace=BoneRelative
    alphaBoolBlend {blendOption=Linear}
    boundBone {boneName=dyn_Cloth_volume_fn}
    componentAppliedLinearAccClamp=<2500 2500 1500>
    componentLinearAccScale=<0.3 0.3 0.3>
    componentLinearVelScale=<0.35 0.35 0.35>
    gravityOverride {z=500}
    physicsBodyDefinitions [{
      collisionType=CustomSphere
      sphereCollisionRadius=2
      boundBone {boneName=dyn_Cloth_volume_fn}
      boxExtents=<45 45 45>
      constraintSetup {angularLimitsMax=<0 6 15>, angularLimitsMin=<0 -6 0>, angularTarget=<0 0 0>}
      localJointOffset=<-9 0.2 -0.1>
     }]
    relativeSpaceBone {boneName=root}
    simSpaceSettings {maxAngularAcceleration=500, maxAngularVelocity=50}
   }
   n32 = Anim Dynamics - Bone: dyn_Cloth_volume_fn_01!(ComponentPose=@n31, Alpha=1, GravityOverride:Vector="0,0,500", ExternalForce="0,0,0") -> (Pose) {
    angularDampingOverride=0.99
    angularSpringConstant=5000
    bAngularSpring=true
    bOverrideAngularDamping=true
    bUseGravityOverride=true
    gravityScale=0.3
    simulationSpace=BoneRelative
    alphaBoolBlend {blendOption=Linear}
    boundBone {boneName=dyn_Cloth_volume_fn_01}
    componentAppliedLinearAccClamp=<2500 2500 1500>
    componentLinearAccScale=<0.3 0.3 0.3>
    componentLinearVelScale=<0.35 0.35 0.35>
    gravityOverride {z=500}
    physicsBodyDefinitions [{
      collisionType=CustomSphere
      sphereCollisionRadius=2
      boundBone {boneName=dyn_Cloth_volume_fn_01}
      boxExtents=<25 15 15>
      constraintSetup {angularLimitsMax=<0 6 10>, angularLimitsMin=<0 -6 0>, angularTarget=<0 0 0>}
      localJointOffset=<-9 0.2 -0.1>
     }]
    relativeSpaceBone {boneName=root}
    simSpaceSettings {maxAngularAcceleration=500, maxAngularVelocity=50}
   }
   n33 = Anim Dynamics - Bone: dyn_Cloth_volume_fn_02!(ComponentPose=@n32, Alpha=1, GravityOverride:Vector="0,0,500", ExternalForce="0,0,0") -> (Pose) {
    angularDampingOverride=0.99
    angularSpringConstant=5000
    bAngularSpring=true
    bOverrideAngularDamping=true
    bUseGravityOverride=true
    gravityScale=0.3
    simulationSpace=BoneRelative
    alphaBoolBlend {blendOption=Linear}
    boundBone {boneName=dyn_Cloth_volume_fn_02}
    componentAppliedLinearAccClamp=<2500 2500 1500>
    componentLinearAccScale=<0.3 0.3 0.3>
    componentLinearVelScale=<0.35 0.35 0.35>
    gravityOverride {z=500}
    physicsBodyDefinitions [{
      collisionType=CustomSphere
      sphereCollisionRadius=2
      boundBone {boneName=dyn_Cloth_volume_fn_02}
      boxExtents=<12 12 12>
      constraintSetup {angularLimitsMax=<0 6 10>, angularLimitsMin=<0 -6 0>, angularTarget=<0 0 0>}
      localJointOffset=<-9 0.2 -0.1>
     }]
    relativeSpaceBone {boneName=root}
    simSpaceSettings {maxAngularAcceleration=500, maxAngularVelocity=50}
   }
   n34 = Anim Dynamics - Bone: dyn_Cloth_volume_l_B!(ComponentPose=@n33, Alpha=1, GravityOverride:Vector="0,0,500", ExternalForce="0,0,0") -> (Pose) {
    angularDampingOverride=0.99
    angularSpringConstant=5000
    bAngularSpring=true
    bOverrideAngularDamping=true
    bUseGravityOverride=true
    gravityScale=0.3
    simulationSpace=BoneRelative
    alphaBoolBlend {blendOption=Linear}
    boundBone {boneName=dyn_Cloth_volume_l_B}
    componentAppliedLinearAccClamp=<2500 2500 1500>
    componentLinearAccScale=<0.3 0.3 0.3>
    componentLinearVelScale=<0.35 0.35 0.35>
    gravityOverride {z=500}
    physicsBodyDefinitions [{
      collisionType=CustomSphere
      sphereCollisionRadius=2
      boundBone {boneName=dyn_Cloth_volume_l_B}
      boxExtents=<45 45 45>
      constraintSetup {angularLimitsMax=<0 0 15>, angularLimitsMin=<0 0 0>, angularTarget=<0 0 0>}
      localJointOffset=<-9 0.2 -0.1>
     }]
    relativeSpaceBone {boneName=root}
    simSpaceSettings {maxAngularAcceleration=500, maxAngularVelocity=50}
   }
   n35 = Anim Dynamics - Bone: dyn_Cloth_volume_l_B01!(ComponentPose=@n34, Alpha=1, GravityOverride:Vector="0,0,500", ExternalForce="0,0,0") -> (Pose) {
    angularDampingOverride=0.99
    angularSpringConstant=5000
    bAngularSpring=true
    bOverrideAngularDamping=true
    bUseGravityOverride=true
    gravityScale=0.3
    simulationSpace=BoneRelative
    alphaBoolBlend {blendOption=Linear}
    boundBone {boneName=dyn_Cloth_volume_l_B01}
    componentAppliedLinearAccClamp=<2500 2500 1500>
    componentLinearAccScale=<0.3 0.3 0.3>
    componentLinearVelScale=<0.35 0.35 0.35>
    gravityOverride {z=500}
    physicsBodyDefinitions [{
      collisionType=CustomSphere
      sphereCollisionRadius=2
      boundBone {boneName=dyn_Cloth_volume_l_B01}
      boxExtents=<25 15 15>
      constraintSetup {angularLimitsMax=<0 0 10>, angularLimitsMin=<0 0 0>, angularTarget=<0 0 0>}
      localJointOffset=<-9 0.2 -0.1>
     }]
    relativeSpaceBone {boneName=root}
    simSpaceSettings {maxAngularAcceleration=500, maxAngularVelocity=50}
   }
   n36 = Anim Dynamics - Bone: dyn_Cloth_volume_l_B02!(ComponentPose=@n35, Alpha=1, GravityOverride:Vector="0,0,500", ExternalForce="0,0,0") -> (Pose) {
    angularDampingOverride=0.99
    angularSpringConstant=5000
    bAngularSpring=true
    bOverrideAngularDamping=true
    bUseGravityOverride=true
    gravityScale=0.3
    simulationSpace=BoneRelative
    alphaBoolBlend {blendOption=Linear}
    boundBone {boneName=dyn_Cloth_volume_l_B02}
    componentAppliedLinearAccClamp=<2500 2500 1500>
    componentLinearAccScale=<0.3 0.3 0.3>
    componentLinearVelScale=<0.35 0.35 0.35>
    gravityOverride {z=500}
    physicsBodyDefinitions [{
      collisionType=CustomSphere
      sphereCollisionRadius=2
      boundBone {boneName=dyn_Cloth_volume_l_B02}
      boxExtents=<12 6 6>
      constraintSetup {angularLimitsMax=<0 0 10>, angularLimitsMin=<0 0 0>, angularTarget=<0 0 0>}
      localJointOffset=<-9 0.2 -0.1>
     }]
    relativeSpaceBone {boneName=root}
    simSpaceSettings {maxAngularAcceleration=500, maxAngularVelocity=50}
   }
   n37 = Anim Dynamics - Bone: dyn_Cloth_volume_r_B!(ComponentPose=@n36, Alpha=1, GravityOverride:Vector="0,0,-980", ExternalForce="0,0,0") -> (Pose) {
    angularDampingOverride=0.99
    angularSpringConstant=5000
    bAngularSpring=true
    bOverrideAngularDamping=true
    bUseGravityOverride=true
    gravityScale=0.3
    simulationSpace=BoneRelative
    alphaBoolBlend {blendOption=Linear}
    boundBone {boneName=dyn_Cloth_volume_r_B}
    componentAppliedLinearAccClamp=<2500 2500 1500>
    componentLinearAccScale=<0.3 0.3 0.3>
    componentLinearVelScale=<0.35 0.35 0.35>
    gravityOverride {z=-980}
    physicsBodyDefinitions [{
      collisionType=CustomSphere
      sphereCollisionRadius=2
      boundBone {boneName=dyn_Cloth_volume_r_B}
      boxExtents=<45 45 45>
      constraintSetup {angularLimitsMax=<0 0 15>, angularLimitsMin=<0 0 0>, angularTarget=<0 0 0>}
      localJointOffset=<-9 0.2 -0.1>
     }]
    relativeSpaceBone {boneName=root}
    simSpaceSettings {maxAngularAcceleration=500, maxAngularVelocity=50}
   }
   n38 = Anim Dynamics - Bone: dyn_Cloth_volume_r_B01!(ComponentPose=@n37, Alpha=1, GravityOverride:Vector="0,0,-980", ExternalForce="0,0,0") -> (Pose) {
    angularDampingOverride=0.99
    angularSpringConstant=5000
    bAngularSpring=true
    bOverrideAngularDamping=true
    bUseGravityOverride=true
    gravityScale=0.3
    simulationSpace=BoneRelative
    alphaBoolBlend {blendOption=Linear}
    boundBone {boneName=dyn_Cloth_volume_r_B01}
    componentAppliedLinearAccClamp=<2500 2500 1500>
    componentLinearAccScale=<0.3 0.3 0.3>
    componentLinearVelScale=<0.35 0.35 0.35>
    gravityOverride {z=-980}
    physicsBodyDefinitions [{
      collisionType=CustomSphere
      sphereCollisionRadius=2
      boundBone {boneName=dyn_Cloth_volume_r_B01}
      boxExtents=<25 15 15>
      constraintSetup {angularLimitsMax=<0 0 10>, angularLimitsMin=<0 0 0>, angularTarget=<0 0 0>}
      localJointOffset=<-9 0.2 -0.1>
     }]
    relativeSpaceBone {boneName=root}
    simSpaceSettings {maxAngularAcceleration=500, maxAngularVelocity=50}
   }
   n39 = Anim Dynamics - Bone: dyn_Cloth_volume_r_B02!(ComponentPose=@n38, Alpha=1, GravityOverride:Vector="0,0,-980", ExternalForce="0,0,0") -> (Pose) {
    angularDampingOverride=0.99
    angularSpringConstant=5000
    bAngularSpring=true
    bOverrideAngularDamping=true
    bUseGravityOverride=true
    gravityScale=0.3
    simulationSpace=BoneRelative
    alphaBoolBlend {blendOption=Linear}
    boundBone {boneName=dyn_Cloth_volume_r_B02}
    componentAppliedLinearAccClamp=<2500 2500 1500>
    componentLinearAccScale=<0.3 0.3 0.3>
    componentLinearVelScale=<0.35 0.35 0.35>
    gravityOverride {z=-980}
    physicsBodyDefinitions [{
      collisionType=CustomSphere
      sphereCollisionRadius=2
      boundBone {boneName=dyn_Cloth_volume_r_B02}
      boxExtents=<12 6 6>
      constraintSetup {angularLimitsMax=<0 0 10>, angularLimitsMin=<0 0 0>, angularTarget=<0 0 0>}
      localJointOffset=<-9 0.2 -0.1>
     }]
    relativeSpaceBone {boneName=root}
    simSpaceSettings {maxAngularAcceleration=500, maxAngularVelocity=50}
   }
   n40 = Anim Dynamics - Bone: dyn_Cloth_volume_l_C!(ComponentPose=@n39, Alpha=1, GravityOverride:Vector="0,0,500", ExternalForce="0,0,0") -> (Pose) {
    angularDampingOverride=0.9999
    angularSpringConstant=5000
    bAngularSpring=true
    bOverrideAngularDamping=true
    bUseGravityOverride=true
    gravityScale=0.3
    simulationSpace=BoneRelative
    alphaBoolBlend {blendOption=Linear}
    boundBone {boneName=dyn_Cloth_volume_l_C}
    componentAppliedLinearAccClamp=<2500 2500 1500>
    componentLinearAccScale=<0.3 0.3 0.3>
    componentLinearVelScale=<0.35 0.35 0.35>
    gravityOverride {z=500}
    physicsBodyDefinitions [{
      collisionType=CustomSphere
      sphereCollisionRadius=2
      boundBone {boneName=dyn_Cloth_volume_l_C}
      boxExtents=<45 45 45>
      constraintSetup {angularLimitsMax=<0 5 10>, angularLimitsMin=<0 -5 0>, angularTarget=<0 0 0>}
      localJointOffset=<-9 -2 -0.1>
     }]
    relativeSpaceBone {boneName=root}
    simSpaceSettings {maxAngularAcceleration=500, maxAngularVelocity=50}
   }
   n41 = Anim Dynamics - Bone: dyn_Cloth_volume_l_C01!(ComponentPose=@n40, bAlphaBoolEnabled:bool=True, GravityOverride:Vector="0,0,500", ExternalForce="0,0,0") -> (Pose) {
    alphaInputType=Bool
    angularDampingOverride=1
    angularSpringConstant=5000
    bAngularSpring=true
    bOverrideAngularDamping=true
    bUseGravityOverride=true
    gravityScale=0.3
    simulationSpace=BoneRelative
    alphaBoolBlend {blendOption=Linear}
    boundBone {boneName=dyn_Cloth_volume_l_C01}
    componentAppliedLinearAccClamp=<2500 2500 1500>
    componentLinearAccScale=<0.3 0.3 0.3>
    componentLinearVelScale=<0.35 0.35 0.35>
    gravityOverride {z=500}
    physicsBodyDefinitions [{
      collisionType=CustomSphere
      sphereCollisionRadius=2
      boundBone {boneName=dyn_Cloth_volume_l_C01}
      boxExtents=<25 15 15>
      constraintSetup {angularLimitsMax=<0 5 10>, angularLimitsMin=<0 -5 -3>, angularTarget=<0 0 0>}
      localJointOffset=<-9 -2 0>
     }]
    relativeSpaceBone {boneName=root}
    simSpaceSettings {maxAngularAcceleration=500, maxAngularVelocity=50}
   }
   n42 = Anim Dynamics - Bone: dyn_Cloth_volume_l_C02!(ComponentPose=@n41, bAlphaBoolEnabled:bool=True, GravityOverride:Vector="0,0,500", ExternalForce="0,0,0") -> (Pose) {
    alphaInputType=Bool
    angularDampingOverride=1
    angularSpringConstant=5000
    bAngularSpring=true
    bOverrideAngularDamping=true
    bUseGravityOverride=true
    gravityScale=0.3
    simulationSpace=BoneRelative
    alphaBoolBlend {blendOption=Linear}
    boundBone {boneName=dyn_Cloth_volume_l_C02}
    componentAppliedLinearAccClamp=<2500 2500 1500>
    componentLinearAccScale=<0.3 0.3 0.3>
    componentLinearVelScale=<0.35 0.35 0.35>
    gravityOverride {z=500}
    physicsBodyDefinitions [{
      collisionType=CustomSphere
      sphereCollisionRadius=2
      boundBone {boneName=dyn_Cloth_volume_l_C02}
      boxExtents=<12 6 6>
      constraintSetup {angularLimitsMax=<0 5 10>, angularLimitsMin=<0 -5 -7>, angularTarget=<0 0 0>}
      localJointOffset=<-9 -3 0>
     }]
    relativeSpaceBone {boneName=root}
    simSpaceSettings {maxAngularAcceleration=500, maxAngularVelocity=50}
   }
   n43 = Anim Dynamics - Bone: dyn_Cloth_volume_l_C01(ComponentPose=@n42, bAlphaBoolEnabled=True, GravityOverride="0,0,500", ExternalForce="0,0,0") {
    alphaInputType=Bool
    angularDampingOverride=0.99
    angularSpringConstant=5000
    bAngularSpring=true
    bOverrideAngularDamping=true
    bUseGravityOverride=true
    gravityScale=0.3
    simulationSpace=BoneRelative
    alphaBoolBlend {blendOption=Linear}
    boundBone {boneName=dyn_Cloth_volume_l_C01}
    componentAppliedLinearAccClamp=<2500 2500 1500>
    componentLinearAccScale=<0.3 0.3 0.3>
    componentLinearVelScale=<0.35 0.35 0.35>
    gravityOverride {z=500}
    physicsBodyDefinitions [{
      collisionType=CustomSphere
      sphereCollisionRadius=2
      boundBone {boneName=dyn_Cloth_volume_l_C01}
      boxExtents=<25 15 15>
      constraintSetup {angularLimitsMax=<0 5 15>, angularLimitsMin=<0 -5 0>, angularTarget=<0 0 0>}
      localJointOffset=<-9 0.2 -0.1>
     }]
    relativeSpaceBone {boneName=root}
    simSpaceSettings {maxAngularAcceleration=500, maxAngularVelocity=50}
   }
   n44 = Anim Dynamics - Bone: dyn_Cloth_volume_l_C02(ComponentPose=@n43, bAlphaBoolEnabled=True, GravityOverride="0,0,500", ExternalForce="0,0,0") {
    alphaInputType=Bool
    angularDampingOverride=0.99
    angularSpringConstant=5000
    bAngularSpring=true
    bOverrideAngularDamping=true
    bUseGravityOverride=true
    gravityScale=0.3
    simulationSpace=BoneRelative
    alphaBoolBlend {blendOption=Linear}
    boundBone {boneName=dyn_Cloth_volume_l_C02}
    componentAppliedLinearAccClamp=<2500 2500 1500>
    componentLinearAccScale=<0.3 0.3 0.3>
    componentLinearVelScale=<0.35 0.35 0.35>
    gravityOverride {z=500}
    physicsBodyDefinitions [{
      collisionType=CustomSphere
      sphereCollisionRadius=2
      boundBone {boneName=dyn_Cloth_volume_l_C02}
      boxExtents=<12 6 6>
      constraintSetup {angularLimitsMax=<0 5 15>, angularLimitsMin=<0 -5 0>, angularTarget=<0 0 0>}
      localJointOffset=<-9 0.2 -0.1>
     }]
    relativeSpaceBone {boneName=root}
    simSpaceSettings {maxAngularAcceleration=500, maxAngularVelocity=50}
   }
   n45 = Anim Dynamics - Bone: dyn_Cloth_volume_r_C!(ComponentPose=@n44, Alpha=1, GravityOverride:Vector="0,0,-980", ExternalForce="0,0,0") -> (Pose) {
    angularDampingOverride=0.9999
    angularSpringConstant=5000
    bAngularSpring=true
    bOverrideAngularDamping=true
    bUseGravityOverride=true
    gravityScale=0.3
    simulationSpace=BoneRelative
    alphaBoolBlend {blendOption=Linear}
    boundBone {boneName=dyn_Cloth_volume_r_C}
    componentAppliedLinearAccClamp=<2500 2500 1500>
    componentLinearAccScale=<0.3 0.3 0.3>
    componentLinearVelScale=<0.35 0.35 0.35>
    gravityOverride {z=-980}
    physicsBodyDefinitions [{
      collisionType=CustomSphere
      sphereCollisionRadius=2
      boundBone {boneName=dyn_Cloth_volume_r_C}
      boxExtents=<45 45 45>
      constraintSetup {angularLimitsMax=<0 5 10>, angularLimitsMin=<0 -5 0>, angularTarget=<0 0 0>}
      localJointOffset=<-9 -2 -0.1>
     }]
    relativeSpaceBone {boneName=root}
    simSpaceSettings {maxAngularAcceleration=500, maxAngularVelocity=50}
   }
   n46 = Anim Dynamics - Bone: dyn_Cloth_volume_r_C01!(ComponentPose=@n45, bAlphaBoolEnabled:bool=True, GravityOverride:Vector="0,0,-980", ExternalForce="0,0,0") -> (Pose) {
    alphaInputType=Bool
    angularDampingOverride=1
    angularSpringConstant=5000
    bAngularSpring=true
    bOverrideAngularDamping=true
    bUseGravityOverride=true
    gravityScale=0.3
    simulationSpace=BoneRelative
    alphaBoolBlend {blendOption=Linear}
    boundBone {boneName=dyn_Cloth_volume_r_C01}
    componentAppliedLinearAccClamp=<2500 2500 1500>
    componentLinearAccScale=<0.3 0.3 0.3>
    componentLinearVelScale=<0.35 0.35 0.35>
    gravityOverride {z=-980}
    physicsBodyDefinitions [{
      collisionType=CustomSphere
      sphereCollisionRadius=2
      boundBone {boneName=dyn_Cloth_volume_r_C01}
      boxExtents=<25 15 15>
      constraintSetup {angularLimitsMax=<0 5 10>, angularLimitsMin=<0 -5 -3>, angularTarget=<0 0 0>}
      localJointOffset=<-9 -2 0>
     }]
    relativeSpaceBone {boneName=root}
    simSpaceSettings {maxAngularAcceleration=500, maxAngularVelocity=50}
   }
   n47 = Anim Dynamics - Bone: dyn_Cloth_volume_r_C02!(ComponentPose=@n46, bAlphaBoolEnabled:bool=True, GravityOverride:Vector="0,0,-980", ExternalForce="0,0,0") -> (Pose) {
    alphaInputType=Bool
    angularDampingOverride=1
    angularSpringConstant=5000
    bAngularSpring=true
    bOverrideAngularDamping=true
    bUseGravityOverride=true
    gravityScale=0.3
    simulationSpace=BoneRelative
    alphaBoolBlend {blendOption=Linear}
    boundBone {boneName=dyn_Cloth_volume_r_C02}
    componentAppliedLinearAccClamp=<2500 2500 1500>
    componentLinearAccScale=<0.3 0.3 0.3>
    componentLinearVelScale=<0.35 0.35 0.35>
    gravityOverride {z=-980}
    physicsBodyDefinitions [{
      collisionType=CustomSphere
      sphereCollisionRadius=2
      boundBone {boneName=dyn_Cloth_volume_r_C02}
      boxExtents=<12 6 6>
      constraintSetup {angularLimitsMax=<0 5 10>, angularLimitsMin=<0 -5 -7>, angularTarget=<0 0 0>}
      localJointOffset=<-9 -3 0>
     }]
    relativeSpaceBone {boneName=root}
    simSpaceSettings {maxAngularAcceleration=500, maxAngularVelocity=50}
   }
   n48 = Anim Dynamics - Bone: dyn_Cloth_volume_r_C01(ComponentPose=@n47, bAlphaBoolEnabled=True, GravityOverride="0,0,-980", ExternalForce="0,0,0") {
    alphaInputType=Bool
    angularDampingOverride=0.99
    angularSpringConstant=5000
    bAngularSpring=true
    bOverrideAngularDamping=true
    bUseGravityOverride=true
    gravityScale=0.3
    simulationSpace=BoneRelative
    alphaBoolBlend {blendOption=Linear}
    boundBone {boneName=dyn_Cloth_volume_r_C01}
    componentAppliedLinearAccClamp=<2500 2500 1500>
    componentLinearAccScale=<0.3 0.3 0.3>
    componentLinearVelScale=<0.35 0.35 0.35>
    gravityOverride {z=-980}
    physicsBodyDefinitions [{
      collisionType=CustomSphere
      sphereCollisionRadius=2
      boundBone {boneName=dyn_Cloth_volume_r_C01}
      boxExtents=<25 15 15>
      constraintSetup {angularLimitsMax=<0 5 15>, angularLimitsMin=<0 -5 0>, angularTarget=<0 0 0>}
      localJointOffset=<-9 0.2 -0.1>
     }]
    relativeSpaceBone {boneName=root}
    simSpaceSettings {maxAngularAcceleration=500, maxAngularVelocity=50}
   }
   n49 = Anim Dynamics - Bone: dyn_Cloth_volume_r_C02(ComponentPose=@n48, bAlphaBoolEnabled=True, GravityOverride="0,0,-980", ExternalForce="0,0,0") {
    alphaInputType=Bool
    angularDampingOverride=0.99
    angularSpringConstant=5000
    bAngularSpring=true
    bOverrideAngularDamping=true
    bUseGravityOverride=true
    gravityScale=0.3
    simulationSpace=BoneRelative
    alphaBoolBlend {blendOption=Linear}
    boundBone {boneName=dyn_Cloth_volume_r_C02}
    componentAppliedLinearAccClamp=<2500 2500 1500>
    componentLinearAccScale=<0.3 0.3 0.3>
    componentLinearVelScale=<0.35 0.35 0.35>
    gravityOverride {z=-980}
    physicsBodyDefinitions [{
      collisionType=CustomSphere
      sphereCollisionRadius=2
      boundBone {boneName=dyn_Cloth_volume_r_C02}
      boxExtents=<12 6 6>
      constraintSetup {angularLimitsMax=<0 5 15>, angularLimitsMin=<0 -5 0>, angularTarget=<0 0 0>}
      localJointOffset=<-9 0.2 -0.1>
     }]
    relativeSpaceBone {boneName=root}
    simSpaceSettings {maxAngularAcceleration=500, maxAngularVelocity=50}
   }
   n50 = Anim Dynamics - Bone: dyn_Cloth_volume_bk01!(ComponentPose=@n49, Alpha=1, GravityOverride:Vector="0,0,500", ExternalForce="0,0,0") -> (Pose) {
    angularDampingOverride=0.9999
    angularSpringConstant=5000
    bAngularSpring=true
    bOverrideAngularDamping=true
    bUseGravityOverride=true
    gravityScale=0.3
    simulationSpace=BoneRelative
    alphaBoolBlend {blendOption=Linear}
    boundBone {boneName=dyn_Cloth_volume_bk01}
    componentAppliedLinearAccClamp=<2500 2500 1500>
    componentLinearAccScale=<0.3 0.3 0.3>
    componentLinearVelScale=<0.35 0.35 0.35>
    gravityOverride {z=500}
    physicsBodyDefinitions [{
      collisionType=CustomSphere
      sphereCollisionRadius=2
      boundBone {boneName=dyn_Cloth_volume_bk01}
      boxExtents=<45 45 45>
      constraintSetup {angularLimitsMax=<0 5 10>, angularLimitsMin=<0 -5 0>, angularTarget=<0 0 0>}
      localJointOffset=<-7 -2 -0.1>
     }]
    relativeSpaceBone {boneName=root}
    simSpaceSettings {maxAngularAcceleration=500, maxAngularVelocity=50}
   }
   n51 = Anim Dynamics - Bone: dyn_Cloth_volume_bk01!(ComponentPose=@n50, bAlphaBoolEnabled:bool=True, GravityOverride="0,0,500", ExternalForce="0,0,0") -> (Pose) {
    alphaInputType=Bool
    angularDampingOverride=1
    angularSpringConstant=5000
    bAngularSpring=true
    bOverrideAngularDamping=true
    bUseGravityOverride=true
    gravityScale=0.3
    simulationSpace=BoneRelative
    alphaBoolBlend {blendOption=Linear}
    boundBone {boneName=dyn_Cloth_volume_bk01}
    componentAppliedLinearAccClamp=<2500 2500 1500>
    componentLinearAccScale=<0.3 0.3 0.3>
    componentLinearVelScale=<0.35 0.35 0.35>
    gravityOverride {z=500}
    physicsBodyDefinitions [{
      collisionType=CustomSphere
      sphereCollisionRadius=2
      boundBone {boneName=dyn_Cloth_volume_bk01}
      boxExtents=<25 15 15>
      constraintSetup {angularLimitsMax=<0 5 10>, angularLimitsMin=<0 -5 -3>, angularTarget=<0 0 0>}
      localJointOffset=<-9 -2 0>
     }]
    relativeSpaceBone {boneName=root}
    simSpaceSettings {maxAngularAcceleration=500, maxAngularVelocity=50}
   }
   n52 = Anim Dynamics - Bone: dyn_Cloth_volume_bk02!(ComponentPose=@n51, bAlphaBoolEnabled:bool=True, GravityOverride:Vector="0,0,500", ExternalForce="0,0,0") -> (Pose) {
    alphaInputType=Bool
    angularDampingOverride=1
    angularSpringConstant=5000
    bAngularSpring=true
    bOverrideAngularDamping=true
    bUseGravityOverride=true
    gravityScale=0.3
    simulationSpace=BoneRelative
    alphaBoolBlend {blendOption=Linear}
    boundBone {boneName=dyn_Cloth_volume_bk02}
    componentAppliedLinearAccClamp=<2500 2500 1500>
    componentLinearAccScale=<0.3 0.3 0.3>
    componentLinearVelScale=<0.35 0.35 0.35>
    gravityOverride {z=500}
    physicsBodyDefinitions [{
      collisionType=CustomSphere
      sphereCollisionRadius=2
      boundBone {boneName=dyn_Cloth_volume_bk02}
      boxExtents=<12 6 6>
      constraintSetup {angularLimitsMax=<0 5 10>, angularLimitsMin=<0 -5 -7>, angularTarget=<0 0 0>}
      localJointOffset=<-9 -3 0>
     }]
    relativeSpaceBone {boneName=root}
    simSpaceSettings {maxAngularAcceleration=500, maxAngularVelocity=50}
   }
   n53 = Anim Dynamics - Bone: dyn_Cloth_volume_bk01(ComponentPose=@n52, bAlphaBoolEnabled=True, GravityOverride="0,0,500", ExternalForce="0,0,0") {
    alphaInputType=Bool
    angularDampingOverride=0.99
    angularSpringConstant=5000
    bAngularSpring=true
    bOverrideAngularDamping=true
    bUseGravityOverride=true
    gravityScale=0.3
    simulationSpace=BoneRelative
    alphaBoolBlend {blendOption=Linear}
    boundBone {boneName=dyn_Cloth_volume_bk01}
    componentAppliedLinearAccClamp=<2500 2500 1500>
    componentLinearAccScale=<0.3 0.3 0.3>
    componentLinearVelScale=<0.35 0.35 0.35>
    gravityOverride {z=500}
    physicsBodyDefinitions [{
      collisionType=CustomSphere
      sphereCollisionRadius=2
      boundBone {boneName=dyn_Cloth_volume_bk01}
      boxExtents=<25 15 15>
      constraintSetup {angularLimitsMax=<0 5 15>, angularLimitsMin=<0 -5 0>, angularTarget=<0 0 0>}
      localJointOffset=<-9 0.2 -0.1>
     }]
    relativeSpaceBone {boneName=root}
    simSpaceSettings {maxAngularAcceleration=500, maxAngularVelocity=50}
   }
   n54 = Anim Dynamics - Bone: dyn_Cloth_volume_bk02(ComponentPose=@n53, bAlphaBoolEnabled=True, GravityOverride="0,0,500", ExternalForce="0,0,0") {
    alphaInputType=Bool
    angularDampingOverride=0.99
    angularSpringConstant=5000
    bAngularSpring=true
    bOverrideAngularDamping=true
    bUseGravityOverride=true
    gravityScale=0.3
    simulationSpace=BoneRelative
    alphaBoolBlend {blendOption=Linear}
    boundBone {boneName=dyn_Cloth_volume_bk02}
    componentAppliedLinearAccClamp=<2500 2500 1500>
    componentLinearAccScale=<0.3 0.3 0.3>
    componentLinearVelScale=<0.35 0.35 0.35>
    gravityOverride {z=500}
    physicsBodyDefinitions [{
      collisionType=CustomSphere
      sphereCollisionRadius=2
      boundBone {boneName=dyn_Cloth_volume_bk02}
      boxExtents=<12 6 6>
      constraintSetup {angularLimitsMax=<0 5 15>, angularLimitsMin=<0 -5 0>, angularTarget=<0 0 0>}
      localJointOffset=<-9 0.2 -0.1>
     }]
    relativeSpaceBone {boneName=root}
    simSpaceSettings {maxAngularAcceleration=500, maxAngularVelocity=50}
   }
   n55 = "Constraint - dyn_spine_03_belt_r_a01 (PPPPP)"!(
    ComponentPose=@n54
    Alpha=1
    ConstraintWeights_0=0.3768
    ConstraintWeights_1:float=0.2707
    ConstraintWeights_2:float=0.2384
    ConstraintWeights_3:float=0.0698
    ConstraintWeights_4:float=0.0443
   ) -> (Pose) {
    boneToModify {boneName=dyn_spine_03_belt_r_a01}
    constraintSetup [
     {targetBone {boneName=spine_04}}
     {targetBone {boneName=pelvis}}
     {targetBone {boneName=spine_02}}
     {targetBone {boneName=spine_03}}
     {targetBone {boneName=spine_01}}
    ]
    constraintWeights [1,1,1,1,1]
   }
   n56 = "Constraint - dyn_spine_03_belt_r_a02 (PPPPPP)"!(
    ComponentPose=@n55
    Alpha=1
    ConstraintWeights_0=0.4288
    ConstraintWeights_1:float=0.3059
    ConstraintWeights_2:float=0.1331
    ConstraintWeights_3:float=0.1023
    ConstraintWeights_4:float=0.0298
    ConstraintWeights_5:float=0.0002
   ) -> (Pose) {
    boneToModify {boneName=dyn_spine_03_belt_r_a02}
    constraintSetup [
     {targetBone {boneName=pelvis}}
     {targetBone {boneName=spine_02}}
     {targetBone {boneName=spine_04}}
     {targetBone {boneName=spine_01}}
     {targetBone {boneName=spine_03}}
     {targetBone {boneName=dyn_Cloth_volume_fn}}
    ]
    constraintWeights [1,1,1,1,1,1]
   }
   n57 = Control Rig!(
    Source=Component To Local(ComponentPose=@n56)
    Alpha
    isSliding=Get bIsSliding()
    isBaseballSliding=Get bIsBaseballSliding()
    Lobby pose:bool=Get is Lobby pose() -> (is Lobby pose:bool)
    IsPlayingEmote:bool=Get bIsPlayingEmote()
   ) -> (Pose) {
    alphaBoolBlend {blendOption=Linear}
    controlRigAssetReference {blueprintRigClass=&/F_MED_GraphShook_CtrlRig1.F_MED_GraphShook_CtrlRig1_C}
   }
   n58 = Anim Dynamics - Bone: dyn_spine_03_belt_l_02!(ComponentPose=Local To Component(LocalPose=@n57), Alpha=1, GravityOverride:Vector="0,0,500", ExternalForce="0,0,0") -> (Pose) {
    angularDampingOverride=0.999
    angularSpringConstant=2500
    bAngularSpring=true
    bOverrideAngularDamping=true
    bUseGravityOverride=true
    gravityScale=0.3
    simulationSpace=BoneRelative
    alphaBoolBlend {blendOption=Linear}
    boundBone {boneName=dyn_spine_03_belt_l_02}
    componentAppliedLinearAccClamp=<2500 2500 1500>
    componentLinearAccScale=<0.4 0.4 0.3>
    componentLinearVelScale=<0.45 0.45 0.35>
    gravityOverride {z=500}
    physicsBodyDefinitions [{
      collisionType=CustomSphere
      sphereCollisionRadius=2
      boundBone {boneName=dyn_spine_03_belt_l_02}
      boxExtents=<45 45 45>
      constraintSetup {angularLimitsMax=<0 5 10>, angularLimitsMin=<0 -5 0>, angularTarget=<0 0 0>}
      localJointOffset=<-6 0.2 -0.1>
     }]
    relativeSpaceBone {boneName=root}
    simSpaceSettings {maxAngularAcceleration=500, maxAngularVelocity=50}
   }
   n59 = Anim Dynamics - Bone: dyn_spine_03_belt_l_03!(ComponentPose=@n58, Alpha=1, GravityOverride:Vector="0,0,500", ExternalForce="0,0,0") -> (Pose) {
    angularDampingOverride=0.999
    angularSpringConstant=2500
    bAngularSpring=true
    bOverrideAngularDamping=true
    bUseGravityOverride=true
    gravityScale=0.3
    simulationSpace=BoneRelative
    alphaBoolBlend {blendOption=Linear}
    boundBone {boneName=dyn_spine_03_belt_l_03}
    componentAppliedLinearAccClamp=<2500 2500 1500>
    componentLinearAccScale=<0.4 0.4 0.3>
    componentLinearVelScale=<0.45 0.45 0.35>
    gravityOverride {z=500}
    physicsBodyDefinitions [{
      collisionType=CustomSphere
      sphereCollisionRadius=2
      boundBone {boneName=dyn_spine_03_belt_l_03}
      boxExtents=<20 10 10>
      constraintSetup {angularLimitsMax=<0 5 0>, angularLimitsMin=<0 -5 0>, angularTarget=<0 0 0>}
      localJointOffset=<-6 0.2 -0.1>
     }]
    relativeSpaceBone {boneName=root}
    simSpaceSettings {maxAngularAcceleration=500, maxAngularVelocity=50}
   }
   n60 = Anim Dynamics - Bone: dyn_spine_03_belt_l_04!(ComponentPose=@n59, Alpha=1, GravityOverride:Vector="0,0,500", ExternalForce="0,0,0") -> (Pose) {
    angularDampingOverride=0.999
    angularSpringConstant=2500
    bAngularSpring=true
    bOverrideAngularDamping=true
    bUseGravityOverride=true
    gravityScale=0.3
    simulationSpace=BoneRelative
    alphaBoolBlend {blendOption=Linear}
    boundBone {boneName=dyn_spine_03_belt_l_04}
    componentAppliedLinearAccClamp=<2500 2500 1500>
    componentLinearAccScale=<0.4 0.4 0.3>
    componentLinearVelScale=<0.45 0.45 0.35>
    gravityOverride {z=500}
    physicsBodyDefinitions [{
      collisionType=CustomSphere
      sphereCollisionRadius=2
      boundBone {boneName=dyn_spine_03_belt_l_04}
      boxExtents=<10 5 5>
      constraintSetup {angularLimitsMax=<0 5 10>, angularLimitsMin=<0 -5 0>, angularTarget=<0 0 0>}
      localJointOffset=<-6 0.2 -0.1>
     }]
    relativeSpaceBone {boneName=root}
    simSpaceSettings {maxAngularAcceleration=500, maxAngularVelocity=50}
   }
   n61 = Anim Dynamics - Bone: dyn_spine_03_belt_r_02!(ComponentPose=@n60, Alpha=1, GravityOverride:Vector="0,0,-980", ExternalForce="0,0,0") -> (Pose) {
    angularDampingOverride=0.999
    angularSpringConstant=2500
    bAngularSpring=true
    bOverrideAngularDamping=true
    bUseGravityOverride=true
    gravityScale=0.3
    simulationSpace=BoneRelative
    alphaBoolBlend {blendOption=Linear}
    boundBone {boneName=dyn_spine_03_belt_r_02}
    componentAppliedLinearAccClamp=<2500 2500 1500>
    componentLinearAccScale=<0.4 0.4 0.3>
    componentLinearVelScale=<0.45 0.45 0.35>
    gravityOverride {z=-980}
    physicsBodyDefinitions [{
      collisionType=CustomSphere
      sphereCollisionRadius=2
      boundBone {boneName=dyn_spine_03_belt_r_02}
      boxExtents=<45 45 45>
      constraintSetup {angularLimitsMax=<0 5 10>, angularLimitsMin=<0 -5 0>, angularTarget=<0 0 0>}
      localJointOffset=<-4 0.2 -0.1>
     }]
    relativeSpaceBone {boneName=root}
    simSpaceSettings {maxAngularAcceleration=500, maxAngularVelocity=50}
   }
   n62 = Anim Dynamics - Bone: dyn_spine_03_belt_r_03!(ComponentPose=@n61, Alpha=1, GravityOverride:Vector="0,0,-980", ExternalForce="0,0,0") -> (Pose) {
    angularDampingOverride=0.999
    angularSpringConstant=2500
    bAngularSpring=true
    bOverrideAngularDamping=true
    bUseGravityOverride=true
    gravityScale=0.3
    simulationSpace=BoneRelative
    alphaBoolBlend {blendOption=Linear}
    boundBone {boneName=dyn_spine_03_belt_r_03}
    componentAppliedLinearAccClamp=<2500 2500 1500>
    componentLinearAccScale=<0.4 0.4 0.3>
    componentLinearVelScale=<0.45 0.45 0.35>
    gravityOverride {z=-980}
    physicsBodyDefinitions [{
      collisionType=CustomSphere
      sphereCollisionRadius=2
      boundBone {boneName=dyn_spine_03_belt_r_03}
      boxExtents=<20 10 10>
      constraintSetup {angularLimitsMax=<0 5 0>, angularLimitsMin=<0 -5 0>, angularTarget=<0 0 0>}
      localJointOffset=<-4 0.2 -0.1>
     }]
    relativeSpaceBone {boneName=root}
    simSpaceSettings {maxAngularAcceleration=500, maxAngularVelocity=50}
   }
   n63 = Anim Dynamics - Bone: dyn_spine_03_belt_r_04!(ComponentPose=@n62, Alpha=1, GravityOverride:Vector="0,0,-980", ExternalForce="0,0,0") -> (Pose) {
    angularDampingOverride=0.999
    angularSpringConstant=2500
    bAngularSpring=true
    bOverrideAngularDamping=true
    bUseGravityOverride=true
    gravityScale=0.3
    simulationSpace=BoneRelative
    alphaBoolBlend {blendOption=Linear}
    boundBone {boneName=dyn_spine_03_belt_r_04}
    componentAppliedLinearAccClamp=<2500 2500 1500>
    componentLinearAccScale=<0.4 0.4 0.3>
    componentLinearVelScale=<0.45 0.45 0.35>
    gravityOverride {z=-980}
    physicsBodyDefinitions [{
      collisionType=CustomSphere
      sphereCollisionRadius=2
      boundBone {boneName=dyn_spine_03_belt_r_04}
      boxExtents=<10 5 5>
      constraintSetup {angularLimitsMax=<0 5 10>, angularLimitsMin=<0 -5 0>, angularTarget=<0 0 0>}
      localJointOffset=<-4 0.2 -0.1>
     }]
    relativeSpaceBone {boneName=root}
    simSpaceSettings {maxAngularAcceleration=500, maxAngularVelocity=50}
   }
   n64 = Anim Dynamics - Bone: dyn_spine_03_belt_r_a01!(ComponentPose=@n63, Alpha=1, GravityOverride:Vector="0,0,-980", ExternalForce="0,0,0") -> (Pose) {
    angularDampingOverride=0.999
    angularSpringConstant=2500
    bAngularSpring=true
    bOverrideAngularDamping=true
    bUseGravityOverride=true
    gravityScale=0.3
    simulationSpace=BoneRelative
    alphaBoolBlend {blendOption=Linear}
    boundBone {boneName=dyn_spine_03_belt_r_a01}
    componentAppliedLinearAccClamp=<2500 2500 1500>
    componentLinearAccScale=<0.4 0.4 0.3>
    componentLinearVelScale=<0.45 0.45 0.35>
    gravityOverride {z=-980}
    physicsBodyDefinitions [{
      collisionType=CustomSphere
      sphereCollisionRadius=2
      boundBone {boneName=dyn_spine_03_belt_r_a01}
      boxExtents=<20 10 10>
      constraintSetup {angularLimitsMax=<0 15 5>, angularLimitsMin=<0 3 0>, angularTarget=<0 0 0>}
      localJointOffset=<-4 0.2 -0.1>
     }]
    relativeSpaceBone {boneName=root}
    simSpaceSettings {maxAngularAcceleration=500, maxAngularVelocity=50}
   }
   n65 = Anim Dynamics - Bone: dyn_spine_03_belt_r_a02!(ComponentPose=@n64, Alpha=1, GravityOverride:Vector="0,0,-980", ExternalForce="0,0,0") -> (Pose) {
    angularDampingOverride=0.999
    angularSpringConstant=2500
    bAngularSpring=true
    bOverrideAngularDamping=true
    bUseGravityOverride=true
    gravityScale=0.3
    simulationSpace=BoneRelative
    alphaBoolBlend {blendOption=Linear}
    boundBone {boneName=dyn_spine_03_belt_r_a02}
    componentAppliedLinearAccClamp=<2500 2500 1500>
    componentLinearAccScale=<0.4 0.4 0.3>
    componentLinearVelScale=<0.45 0.45 0.35>
    gravityOverride {z=-980}
    physicsBodyDefinitions [{
      collisionType=CustomSphere
      sphereCollisionRadius=2
      boundBone {boneName=dyn_spine_03_belt_r_a02}
      boxExtents=<10 5 5>
      constraintSetup {angularLimitsMax=<0 15 10>, angularLimitsMin=<0 3 0>, angularTarget=<0 0 0>}
      localJointOffset=<-4 0.2 -0.1>
     }]
    relativeSpaceBone {boneName=root}
    simSpaceSettings {maxAngularAcceleration=500, maxAngularVelocity=50}
   }
   n66 = Anim Dynamics - Bone: dyn_spine_03_belt_bk_02!(ComponentPose=@n65, Alpha=1, GravityOverride:Vector="0,0,500", ExternalForce="0,0,0") -> (Pose) {
    angularDampingOverride=0.9999
    angularSpringConstant=2500
    bAngularSpring=true
    bOverrideAngularDamping=true
    bUseGravityOverride=true
    gravityScale=0.3
    simulationSpace=BoneRelative
    alphaBoolBlend {blendOption=Linear}
    boundBone {boneName=dyn_spine_03_belt_bk_02}
    componentAppliedLinearAccClamp=<2500 2500 1500>
    componentLinearAccScale=<0.3 0.3 0.3>
    componentLinearVelScale=<0.35 0.35 0.35>
    gravityOverride {z=500}
    physicsBodyDefinitions [{
      collisionType=CustomSphere
      sphereCollisionRadius=2
      boundBone {boneName=dyn_spine_03_belt_bk_02}
      boxExtents=<45 45 45>
      constraintSetup {angularLimitsMax=<0 2 0>, angularLimitsMin=<0 -2 0>, angularTarget=<0 0 0>}
      localJointOffset=<-5 -2 -0.1>
     }]
    relativeSpaceBone {boneName=root}
    simSpaceSettings {maxAngularAcceleration=500, maxAngularVelocity=50}
   }
   n67 = Anim Dynamics - Bone: dyn_spine_03_belt_bk_03!(ComponentPose=@n66, Alpha=1, GravityOverride:Vector="0,0,500", ExternalForce="0,0,0") -> (Pose) {
    angularDampingOverride=1
    angularSpringConstant=2500
    bAngularSpring=true
    bOverrideAngularDamping=true
    bUseGravityOverride=true
    gravityScale=0.3
    simulationSpace=BoneRelative
    alphaBoolBlend {blendOption=Linear}
    boundBone {boneName=dyn_spine_03_belt_bk_03}
    componentAppliedLinearAccClamp=<2500 2500 1500>
    componentLinearAccScale=<0.3 0.3 0.3>
    componentLinearVelScale=<0.35 0.35 0.35>
    gravityOverride {z=500}
    physicsBodyDefinitions [{
      collisionType=CustomSphere
      sphereCollisionRadius=2
      boundBone {boneName=dyn_spine_03_belt_bk_03}
      boxExtents=<25 15 15>
      constraintSetup {angularLimitsMax=<0 2 10>, angularLimitsMin=<0 -2 0>, angularTarget=<0 0 0>}
      localJointOffset=<-5 -2 0>
     }]
    relativeSpaceBone {boneName=root}
    simSpaceSettings {maxAngularAcceleration=500, maxAngularVelocity=50}
   }
   n68 = Anim Dynamics - Bone: dyn_spine_03_belt_bk_04!(ComponentPose=@n67, Alpha=1, GravityOverride:Vector="0,0,500", ExternalForce="0,0,0") -> (Pose) {
    angularDampingOverride=1
    angularSpringConstant=2500
    bAngularSpring=true
    bOverrideAngularDamping=true
    bUseGravityOverride=true
    gravityScale=0.3
    simulationSpace=BoneRelative
    alphaBoolBlend {blendOption=Linear}
    boundBone {boneName=dyn_spine_03_belt_bk_04}
    componentAppliedLinearAccClamp=<2500 2500 1500>
    componentLinearAccScale=<0.3 0.3 0.3>
    componentLinearVelScale=<0.35 0.35 0.35>
    gravityOverride {z=500}
    physicsBodyDefinitions [{
      collisionType=CustomSphere
      sphereCollisionRadius=2
      boundBone {boneName=dyn_spine_03_belt_bk_04}
      boxExtents=<12 6 6>
      constraintSetup {angularLimitsMax=<0 2 20>, angularLimitsMin=<0 -2 0>, angularTarget=<0 0 0>}
      localJointOffset=<-5 -3 0>
     }]
    relativeSpaceBone {boneName=root}
    simSpaceSettings {maxAngularAcceleration=500, maxAngularVelocity=50}
   }
   Save cached pose 'SkirtCachePose'(Pose=Component To Local(ComponentPose=@n68)) {cacheName=SkirtCachePose}
  ]}
 EventGraph {nodes [
   Event BlueprintUpdateAnimation()
   > n1 = Sequence()
   Set Not bIsSkyDiving(execute=@n1.then_1, Not bIsSkyDiving:bool=NOT Boolean(A=Get bIsSkyDiving())) -> (Output_Get:bool)
   > Set is Lobby pose(is Lobby pose:bool="float > float"(A=GetCurveValue(CurveName=frontend_idle_bypass))) -> (Output_Get:bool)
   n2 = "Gravity Override State Function (Parameter Asset) (Nativized)"(
    execute=@n1.then_0
    CharacterPartAnimInstance=Self-Reference()
    GameTimeInSeconds=GetGameTimeInSeconds()
    Parameters=&/F_MED_GraphShook_Parameters
   )
   > Set Gravity Override Cape(Gravity Override Cape:Vector=@n2) -> (Output_Get:Vector)
   > Set RBAN Settings Out_0(RBAN Settings Out_0:Fort Rigid Body Anim Node Parameters Structure=@n2.OutRigidBodyAnimNodeParameters) -> (Output_Get:Fort Rigid Body Anim Node Parameters Structure)
  ]}
]
variables [
 Gravity Override Cape {type=Vector}
 is Lobby pose {type=boolean}
 Not bIsSkyDiving {type=boolean}
 RBAN Settings Out_0 {type=FortRigidBodyAnimNodeParameters}
 Sim Space Settings {
  type=AnimPhysSimSpaceSettings
  default {maxAngularAcceleration=45, maxAngularVelocity=8, simSpaceAngularAlpha=0.25, externalAngularVelocity=<0 0 0>}
 }
]
