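# Physics asset of collision shapes for the GraphShook female medium body cosmetic, named as the colliders for its cape.
# Capsules cover the upper spine, clavicles and neck, with a sphere on the head.
# More capsules ride on dynamic cloth-volume, coat and hat bones, so those volumes follow the secondary bones.
# Arm, leg and foot bodies carry no shapes, and no constraints link any bodies.
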PhysicsAsset /BRCosmetics/Characters/Player/Female/Medium/Bodies/F_MED_GraphShook/Meshes/F_MED_GraphShook_CapeColliders
bodies {
 calf_l []
 calf_r []
 clavicle_l [{shapeType=Capsule, length=9.0192, radius=9.8871, center=<9.4266 -1.1806 -2.4977>, rotation=<rot 69.3297 1.4476 176.9462>}]
 clavicle_r [{shapeType=Capsule, length=9.0192, radius=9.8871, center=<-9.4266 1.1806 2.4977>, rotation=<rot 69.3297 1.4476 -3.0538>}]
 dyn_Cloth_volume_bk01 [{
   shapeType=Capsule
   length=15.782
   radius=12.4567
   shapeName=dyn_Cloth_volume_bk01_capsule
   center=<-8.6563 4.0364 0.8544>
   rotation=<rot 0 -81.3771 -2.3156>
  }]
 dyn_Cloth_volume_l_B01 [{
   shapeType=Capsule
   length=13.7242
   radius=12.0718
   shapeName=dyn_Cloth_volume_l_B01_capsule
   center=<-6.3856 3.6988 1.155>
   rotation=<rot 0 -89.7624 -89.5935>
  }]
 dyn_Cloth_volume_r_B01 [{
   shapeType=Capsule
   length=13.4743
   radius=12.2179
   shapeName=dyn_Cloth_volume_r_B01_capsule
   center=<-6.1768 4.3634 -1.5957>
   rotation=<rot 0 -90.5706 -85.7593>
  }]
 dyn_coat_bk_01_hat [{
   shapeType=Capsule
   length=6.2023
   radius=10.2869
   shapeName=dyn_coat_bk_02_hat_capsule
   center=<1.1598 0.3657 3.4211>
   rotation=<rot 0 177.1719 -88.8575>
  }]
 dyn_coat_bk_02_hat [{
   shapeType=Capsule
   length=6.2023
   radius=10.2869
   shapeName=dyn_coat_bk_02_hat_capsule
   center=<1.1598 0.3657 3.4211>
   rotation=<rot 0 177.1719 -88.8575>
  }]
 dyn_coat_bk_03 [{
   shapeType=Capsule
   length=15.782
   radius=12.4567
   shapeName=dyn_Cloth_volume_bk01_capsule
   center=<0 -0.061 -1.2782>
   rotation=<rot 0 0 92.7318>
  }]
 dyn_coat_mid_01_hat_L [{
   shapeType=Capsule
   length=6.898
   radius=10.9878
   shapeName=dyn_coat_mid_01_hat_L_capsule
   center=<5.3376 -1.323 2.2095>
   rotation=<rot 0 33.6078 -79.6052>
  }]
 dyn_coat_mid_01_hat_R [{
   shapeType=Capsule
   length=6.6773
   radius=9.8459
   shapeName=dyn_coat_mid_01_hat_R_capsule
   center=<4.7619 1.6134 3.1508>
   rotation=<rot 0 -38.026 -86.2942>
  }]
 dyn_coat_mid_02_hat_L [{
   shapeType=Capsule
   length=3.3696
   radius=7.8682
   shapeName=dyn_coat_mid_02_hat_L_capsule
   center=<2.1303 -2.9657 5.006>
   rotation=<rot 0 -154.0278 -74.805>
  }]
 dyn_coat_mid_02_hat_R [{
   shapeType=Capsule
   length=3.3894
   radius=7.8877
   shapeName=dyn_coat_mid_02_hat_R_capsule
   center=<2.1484 3.108 4.1145>
   rotation=<rot 0 -27.9088 -70.8031>
  }]
 foot_l []
 foot_r []
 head [{shapeType=Sphere, radius=17.3125, center=<11.028 -0.174 0>}]
 lowerarm_l []
 lowerarm_r []
 neck_02 [{shapeType=Capsule, length=13.4375, radius=7.3175, center=<-0.2642 -0.3408 0>, rotation=<rot 90 -171.1193 -175.3978>}]
 spine_04 [{shapeType=Capsule, length=9.4404, radius=14.1325, center=<2.766 0.7765 0.0002>}]
 spine_05 [
  {shapeType=Capsule, length=8.0226, radius=15.695, center=<-2.2795 1.8109 0>}
  {shapeType=Capsule, length=13.9518, radius=13.9425, center=<-0.2916 0.4452 0>}
 ]
 thigh_l []
 thigh_r []
 upperarm_l []
 upperarm_r []
}
constraints []
