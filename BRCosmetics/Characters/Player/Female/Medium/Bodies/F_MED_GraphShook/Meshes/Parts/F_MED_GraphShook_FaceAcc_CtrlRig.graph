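# Control Rig for the GraphShook character's face accessory, procedurally posing hair and ear bones.
# Noise with random timing twitches the ears, damped via randomMotionAlpha by state flags such as skydiving.
# With a pack worn, Hair IK fits a spline to head/spine-constrained controls, side chains follow, and upper-arm pose readers deflect them.
# Otherwise, head-driven spherical pose readers offset back hair, with skydiving flags selecting the right-side behavior.
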ControlRigBlueprint /BRCosmetics/Characters/Player/Female/Medium/Bodies/F_MED_GraphShook/Meshes/Parts/F_MED_GraphShook_FaceAcc_CtrlRig
graphs [
 ContainedGraph {nodes [
   n1 = RigVMFunctionEntryNode()
   n2 = &RigVMFunction_MathBoolOr(A=@n1.A, B=@n1.B)
   RigVMFunctionReturnNode(Result=&RigVMFunction_MathBoolOr(B=@n1.D, A=&RigVMFunction_MathBoolOr(B=@n1.C, A=@n2.Result).Result).Result)
  ]}
 ContainedGraph {nodes [
   n1 = RigVMFunctionEntryNode()
   n2 = &RigVMFunction_MathBoolOr(A=@n1.A, B=@n1.B)
   RigVMFunctionReturnNode(Result=&RigVMFunction_MathBoolOr(B=@n1.D, A=&RigVMFunction_MathBoolOr(B=@n1.C, A=@n2.Result).Result).Result)
  ]}
 ContainedGraph {nodes [
   n1 = RigVMFunctionEntryNode()
   SequenceAggregate = &RigVMFunction_Sequence(ExecuteContext=@n1.ExecuteContext)
   SequenceAggregate_1 = &RigVMFunction_Sequence(ExecuteContext=@SequenceAggregate.B)
   SequenceAggregate_2 = &RigVMFunction_Sequence(ExecuteContext=@SequenceAggregate_1.B)
   RigVMFunctionReturnNode(A=@SequenceAggregate.A, B=@SequenceAggregate_1.A, C=@SequenceAggregate_2.A, D=@SequenceAggregate_2.B)
  ]}
 ContainedGraph {nodes [
   n1 = RigVMFunctionEntryNode()
   SequenceAggregate = &RigVMFunction_Sequence(ExecuteContext=@n1.ExecuteContext)
   SequenceAggregate_1 = &RigVMFunction_Sequence(ExecuteContext=@SequenceAggregate.B)
   SequenceAggregate_2 = &RigVMFunction_Sequence(ExecuteContext=@SequenceAggregate_1.B)
   RigVMFunctionReturnNode(A=@SequenceAggregate.A, B=@SequenceAggregate_1.A, C=@SequenceAggregate_2.A, D=@SequenceAggregate_2.B)
  ]}
 New Function_ContainedGraph {nodes [
   RigVMFunctionEntryNode()
   > SequenceAggregate = RigVMAggregateNode()
   &RigUnit_ParentConstraint(
    AdvancedSettings=(RotationOrderForFilter=XZY)
    Child=(Type=Null,Name="dyn_Hair_Back_01_Null")
    Filter=(TranslationFilter=(bX=true,bY=true,bZ=true),RotationFilter=(bX=false,bY=false,bZ=false),ScaleFilter=(bX=True,bY=True,bZ=True))
    Parents=((Item=(Type=Bone,Name="head"),Weight=0.9),(Item=(Type=Bone,Name="spine_05"),Weight=0.1))
    Weight=1
    ExecutePin=@SequenceAggregate.A
   )
   > &RigUnit_ParentConstraint(
    AdvancedSettings=(RotationOrderForFilter=XZY)
    Child=(Type=Null,Name="dyn_Hair_Back_02_Null")
    Filter=(TranslationFilter=(bX=True,bY=True,bZ=True),RotationFilter=(bX=false,bY=false,bZ=false),ScaleFilter=(bX=True,bY=True,bZ=True))
    Parents=((Item=(Type=Bone,Name="head"),Weight=0.7),(Item=(Type=Bone,Name="neck_02"),Weight=0),(Item=(Type=Bone,Name="spine_05"),Weight=0.3))
    Weight=1
   )
   > &RigUnit_ParentConstraint(
    AdvancedSettings=(RotationOrderForFilter=XZY)
    Child=(Type=Null,Name="dyn_Hair_Back_03_Null")
    Filter=(TranslationFilter=(bX=True,bY=True,bZ=True),RotationFilter=(bX=false,bY=false,bZ=false),ScaleFilter=(bX=True,bY=True,bZ=True))
    Parents=((Item=(Type=Bone,Name="head"),Weight=0.7),(Item=(Type=Bone,Name="neck_02"),Weight=0),(Item=(Type=Bone,Name="spine_05"),Weight=0.3))
    Weight=1
   )
   > &RigUnit_ParentConstraint(
    AdvancedSettings=(RotationOrderForFilter=XZY)
    Child=(Type=Null,Name="dyn_Hair_Back_04_Null")
    Filter=(TranslationFilter=(bX=True,bY=True,bZ=True),RotationFilter=(bX=false,bY=false,bZ=false),ScaleFilter=(bX=True,bY=True,bZ=True))
    Parents=((Item=(Type=Bone,Name="clavicle_r"),Weight=0),(Item=(Type=Bone,Name="spine_04"),Weight=1),(Item=(Type=Bone,Name="clavicle_l"),Weight=0))
    Weight=1
   )
   > &RigUnit_ParentConstraint(
    AdvancedSettings=(RotationOrderForFilter=XZY)
    Child=(Type=Null,Name="dyn_Hair_Back_05_Null")
    Filter=(TranslationFilter=(bX=True,bY=True,bZ=True),RotationFilter=(bX=false,bY=false,bZ=false),ScaleFilter=(bX=True,bY=True,bZ=True))
    Parents=((Item=(Type=Bone,Name="clavicle_l"),Weight=0),(Item=(Type=Bone,Name="spine_03"),Weight=1),(Item=(Type=Bone,Name="clavicle_r"),Weight=0))
    Weight=1
   )
   > &RigUnit_ParentConstraint(
    AdvancedSettings=(RotationOrderForFilter=XZY)
    Child=(Type=Null,Name="dyn_Hair_Back_06_Null")
    Filter=(TranslationFilter=(bX=True,bY=True,bZ=True),RotationFilter=(bX=false,bY=false,bZ=false),ScaleFilter=(bX=True,bY=True,bZ=True))
    Parents=((Item=(Type=Bone,Name="dyn_Hair_Back_06"),Weight=0),(Item=(Type=Bone,Name="spine_03"),Weight=1))
    Weight=1
   )
   RerouteNode_2 = RigVMDispatchNode(Values=((Type=Bone,Name="dyn_Hair_Back_01"),((),Type=Bone,Name="dyn_Hair_Back_02"),((),Type=Bone,Name="dyn_Hair_Back_03"),((),Type=Bone,Name="dyn_Hair_Back_04"),((),Type=Bone,Name="dyn_Hair_Back_05"),((),Type=Bone,Name="dyn_Hair_Back_06"))) {template="DISPATCH_RigVMDispatch_ArrayMake(in Values,out Array)"}
   > SplineIK_2 = RigVMFunctionReferenceNode(
    Primary Axis=(X=0.2,Y=-1,Z=0)
    Samples Per Segment=10
    Secondary Spline Direction=(X=0,Y=0,Z=0)
    Spline Mode=BSpline
    Stretch=false
    Up Axis=(X=0,Y=0,Z=0)
    debug=false
    Bones=@RerouteNode_2.Array
    Controls=RigVMVariableNode() {variable=dyn_Hair_Back_ctrls}.Value
   )
   > TwistBones = RigVMFunctionReferenceNode(
    Axis to Twist=(bX=true,bY=true,bZ=true)
    Bones to Twist=((Type=Bone,Name="dyn_Hair_Back_03"),(Type=Bone,Name="dyn_Hair_Back_04"))
    End=(Type=Bone,Name="spine_05")
    Start=(Type=Bone,Name="spine_05")
   )
   &RigUnit_ParentConstraint(
    AdvancedSettings=(RotationOrderForFilter=XZY)
    Child=(Type=Bone,Name="dyn_Hair_BackL_02")
    Filter=(TranslationFilter=(bX=True,bY=True,bZ=True),RotationFilter=(bX=true,bY=true,bZ=true),ScaleFilter=(bX=True,bY=True,bZ=True))
    Parents=((Item=(Type=Bone,Name="dyn_Hair_Back_02"),Weight=1),(Item=(Type=Bone,Name="spine_05"),Weight=0))
    Weight=1
    ExecutePin=@SequenceAggregate.B
   )
   > &RigUnit_ParentConstraint(
    AdvancedSettings=(RotationOrderForFilter=XZY)
    Child=(Type=Bone,Name="dyn_Hair_BackL_03")
    Filter=(TranslationFilter=(bX=True,bY=True,bZ=True),RotationFilter=(bX=true,bY=true,bZ=true),ScaleFilter=(bX=True,bY=True,bZ=True))
    Parents=((Item=(Type=Bone,Name="dyn_Hair_Back_03"),Weight=1),(Item=(Type=Bone,Name="spine_05"),Weight=0))
    Weight=1
   )
   > &RigUnit_ParentConstraint(
    AdvancedSettings=(RotationOrderForFilter=XZY)
    Child=(Type=Bone,Name="dyn_Hair_BackL_04")
    Filter=(TranslationFilter=(bX=True,bY=True,bZ=True),RotationFilter=(bX=true,bY=true,bZ=true),ScaleFilter=(bX=True,bY=True,bZ=True))
    Parents=((Item=(Type=Bone,Name="dyn_Hair_Back_04"),Weight=1),(Item=(Type=Bone,Name="spine_05"),Weight=0))
    Weight=1
   )
   > &RigUnit_ParentConstraint(
    AdvancedSettings=(RotationOrderForFilter=XZY)
    Child=(Type=Bone,Name="dyn_Hair_BackL_05")
    Filter=(TranslationFilter=(bX=True,bY=True,bZ=True),RotationFilter=(bX=true,bY=true,bZ=true),ScaleFilter=(bX=True,bY=True,bZ=True))
    Parents=((Item=(Type=Bone,Name="dyn_Hair_Back_05"),Weight=1),(Item=(Type=Bone,Name="spine_05"),Weight=0))
    Weight=1
   )
   > n1 = &RigUnit_SphericalPoseReader(
    ActiveRegionScaleFactors=(PositiveWidth=0.6,NegativeWidth=0.6,PositiveHeight=0.6,NegativeHeight=0.6)
    ActiveRegionSize=0.1
    Debug=(bDrawDebug=false,DebugScale=25,DebugSegments=20,DebugThickness=0.25)
    DriverAxis=(X=1,Y=0,Z=0)
    DriverItem=(Type=Bone,Name="upperarm_l")
    FalloffRegionScaleFactors=(PositiveWidth=1,NegativeWidth=0.6,PositiveHeight=0.6,NegativeHeight=0.6)
    FalloffSize=0.55
    FlipHeightScaling=false
    FlipWidthScaling=false
    OptionalParentItem=(Type=Bone,Name="spine_05")
    RotationOffset=(X=-80,Y=-30,Z=-10)
   )
   > &RigUnit_ParentConstraint(
    AdvancedSettings=(RotationOrderForFilter=XZY)
    Child=(Type=Bone,Name="dyn_Hair_BackL_03")
    Filter=(TranslationFilter=(bX=false,bY=false,bZ=false),RotationFilter=(bX=true,bY=true,bZ=true),ScaleFilter=(bX=True,bY=True,bZ=True))
    Parents=((Item=(Type=Bone,Name="dyn_Hair_Back_03"),Weight=0.3),(Item=(Type=Bone,Name="upperarm_l"),Weight=0.7))
    Weight=@n1.OutputParam
   )
   n2 = &RigVMFunction_MathQuaternionFromEuler(Euler=(X=20,Y=-10,Z=0), RotationOrder=YXZ)
   > &RigUnit_ModifyTransforms(
    ItemToModify="((Item=(Type=Bone,Name=\"dyn_Hair_BackL_02\"),Transform=(Rotation=(X=0,Y=0,Z=0,W=1),Translation=(X=2,Y=-1,Z=1),Scale3D=(X=1,Y=1,Z=1))),(Item=(Type=Bone,Name=\"dyn_Hair_BackL_04\"),Transform=(Rotation=(X=0,Y=0,Z=0,W=1),Translation=(X=0,Y=0,Z=0),Scale3D=(X=1,Y=1 ...(truncated)"
    ItemToModify.0.Transform.Rotation=@n2.Result
    ItemToModify.1.Transform.Rotation=@n2.Result
    ItemToModify.2.Transform.Rotation=&RigVMFunction_MathQuaternionFromEuler(Euler=(X=0,Y=-20,Z=25)).Result
    Weight=@n1.OutputParam
   )
   n3 = &RigVMFunction_MathQuaternionFromEuler(Euler=(X=0,Y=0,Z=5))
   > &RigUnit_ModifyTransforms(
    ItemToModify="((Item=(Type=Bone,Name=\"dyn_Hair_Back_03\"),Transform=(Rotation=(X=0,Y=0,Z=0,W=1),Translation=(X=0,Y=0,Z=1),Scale3D=(X=1,Y=1,Z=1))),(Item=(Type=Bone,Name=\"dyn_Hair_Back_04\"),Transform=(Rotation=(X=0,Y=0,Z=0,W=1),Translation=(X=0,Y=0,Z=0),Scale3D=(X=1,Y=1,Z= ...(truncated)"
    ItemToModify.2.Transform.Rotation=&RigVMFunction_MathQuaternionFromEuler(Euler=(X=0,Y=0,Z=2)).Result
    ItemToModify.0.Transform.Rotation=@n3.Result
    ItemToModify.1.Transform.Rotation=@n3.Result
    Weight=@n1.OutputParam
   )
   &RigUnit_ParentConstraint(
    AdvancedSettings=(RotationOrderForFilter=XZY)
    Child=(Type=Bone,Name="dyn_Hair_BackR_01")
    Filter=(TranslationFilter=(bX=true,bY=true,bZ=true),RotationFilter=(bX=true,bY=true,bZ=true),ScaleFilter=(bX=True,bY=True,bZ=True))
    Parents=((Item=(Type=Bone,Name="dyn_Hair_Back_01"),Weight=1),(Item=(Type=Bone,Name="spine_05"),Weight=0))
    Weight=1
    ExecutePin=@SequenceAggregate.C
   )
   > &RigUnit_ParentConstraint(
    AdvancedSettings=(RotationOrderForFilter=XZY)
    Child=(Type=Bone,Name="dyn_Hair_BackR_02")
    Filter=(TranslationFilter=(bX=True,bY=True,bZ=True),RotationFilter=(bX=true,bY=true,bZ=true),ScaleFilter=(bX=True,bY=True,bZ=True))
    Parents=((Item=(Type=Bone,Name="dyn_Hair_Back_02"),Weight=1),(Item=(Type=Bone,Name="spine_05"),Weight=0))
    Weight=1
   )
   > &RigUnit_ParentConstraint(
    AdvancedSettings=(RotationOrderForFilter=XZY)
    Child=(Type=Bone,Name="dyn_Hair_BackR_03")
    Filter=(TranslationFilter=(bX=True,bY=True,bZ=True),RotationFilter=(bX=true,bY=true,bZ=true),ScaleFilter=(bX=True,bY=True,bZ=True))
    Parents=((Item=(Type=Bone,Name="dyn_Hair_Back_03"),Weight=1),(Item=(Type=Bone,Name="spine_05"),Weight=0))
    Weight=1
   )
   > &RigUnit_ParentConstraint(
    AdvancedSettings=(RotationOrderForFilter=XZY)
    Child=(Type=Bone,Name="dyn_Hair_BackR_04")
    Filter=(TranslationFilter=(bX=True,bY=True,bZ=True),RotationFilter=(bX=true,bY=true,bZ=true),ScaleFilter=(bX=True,bY=True,bZ=True))
    Parents=((Item=(Type=Bone,Name="dyn_Hair_Back_04"),Weight=1),(Item=(Type=Bone,Name="spine_05"),Weight=0))
    Weight=1
   )
   > &RigUnit_ParentConstraint(
    AdvancedSettings=(RotationOrderForFilter=XZY)
    Child=(Type=Bone,Name="dyn_Hair_BackR_05")
    Filter=(TranslationFilter=(bX=True,bY=True,bZ=True),RotationFilter=(bX=true,bY=true,bZ=true),ScaleFilter=(bX=True,bY=True,bZ=True))
    Parents=((Item=(Type=Bone,Name="dyn_Hair_Back_05"),Weight=1),(Item=(Type=Bone,Name="dyn_Hair_BackR_04"),Weight=0))
    Weight=1
   )
   > n4 = &RigUnit_SphericalPoseReader(
    ActiveRegionScaleFactors=(PositiveWidth=0.6,NegativeWidth=0.6,PositiveHeight=0.6,NegativeHeight=0.6)
    ActiveRegionSize=0.1
    Debug=(bDrawDebug=false,DebugScale=25,DebugSegments=20,DebugThickness=0.25)
    DriverAxis=(X=-1,Y=0,Z=0)
    DriverItem=(Type=Bone,Name="upperarm_r")
    FalloffRegionScaleFactors=(PositiveWidth=1,NegativeWidth=0.6,PositiveHeight=0.6,NegativeHeight=0.6)
    FalloffSize=0.6
    FlipHeightScaling=false
    FlipWidthScaling=true
    OptionalParentItem=(Type=Bone,Name="spine_05")
    RotationOffset=(X=80,Y=-30,Z=-10)
   )
   > &RigUnit_ParentConstraint(
    AdvancedSettings=(RotationOrderForFilter=XZY)
    Child=(Type=Bone,Name="dyn_Hair_BackR_03")
    Filter=(TranslationFilter=(bX=false,bY=false,bZ=false),RotationFilter=(bX=true,bY=true,bZ=true),ScaleFilter=(bX=True,bY=True,bZ=True))
    Parents=((Item=(Type=Bone,Name="dyn_Hair_Back_03"),Weight=0.3),(Item=(Type=Bone,Name="upperarm_r"),Weight=0.7))
    Weight=@n4.OutputParam
   )
   n5 = &RigVMFunction_MathQuaternionFromEuler(Euler=(X=-20,Y=10,Z=0), RotationOrder=YXZ)
   > &RigUnit_ModifyTransforms(
    ItemToModify="((Item=(Type=Bone,Name=\"dyn_Hair_BackR_02\"),Transform=(Rotation=(X=0,Y=0,Z=0,W=1),Translation=(X=2,Y=-1,Z=1),Scale3D=(X=1,Y=1,Z=1))),(Item=(Type=Bone,Name=\"dyn_Hair_BackR_04\"),Transform=(Rotation=(X=0,Y=0,Z=0,W=1),Translation=(X=0,Y=0,Z=0),Scale3D=(X=1,Y=1 ...(truncated)"
    ItemToModify.0.Transform.Rotation=@n5.Result
    ItemToModify.1.Transform.Rotation=@n5.Result
    ItemToModify.2.Transform.Rotation=&RigVMFunction_MathQuaternionFromEuler(Euler=(X=0,Y=20,Z=25)).Result
    Weight=@n4.OutputParam
   )
   n6 = &RigVMFunction_MathQuaternionFromEuler(Euler=(X=0,Y=0,Z=5))
   > &RigUnit_ModifyTransforms(
    ItemToModify="((Item=(Type=Bone,Name=\"dyn_Hair_Back_03\"),Transform=(Rotation=(X=0,Y=0,Z=0,W=1),Translation=(X=0,Y=0,Z=1),Scale3D=(X=1,Y=1,Z=1))),(Item=(Type=Bone,Name=\"dyn_Hair_Back_04\"),Transform=(Rotation=(X=0,Y=0,Z=0,W=1),Translation=(X=0,Y=0,Z=0),Scale3D=(X=1,Y=1,Z= ...(truncated)"
    ItemToModify.2.Transform.Rotation=&RigVMFunction_MathQuaternionFromEuler(Euler=(X=0,Y=0,Z=2)).Result
    ItemToModify.0.Transform.Rotation=@n6.Result
    ItemToModify.1.Transform.Rotation=@n6.Result
    Weight=@n4.OutputParam
   )
   RigVMFunctionReturnNode(ExecuteContext=@SequenceAggregate.D)
   &RigUnit_ParentConstraint(
    AdvancedSettings=(RotationOrderForFilter=XZY)
    Child=(Type=Bone,Name="dyn_Hair_BackL_01")
    ExecutePin=()
    Filter=(TranslationFilter=(bX=true,bY=true,bZ=true),RotationFilter=(bX=true,bY=true,bZ=true),ScaleFilter=(bX=True,bY=True,bZ=True))
    Parents=((Item=(Type=Bone,Name="dyn_Hair_Back_01"),Weight=1),(Item=(Type=Bone,Name="spine_05"),Weight=0))
    Weight=1
   )
  ]}
 RigVMFunctionLibrary {nodes [Hair IK = RigVMCollapseNode(ExecuteContext=())]}
 RigVMModel {nodes [
   RigUnit_BeginExecution = &RigUnit_BeginExecution()
   MathBoolOr = RigVMAggregateNode(B=RigVMVariableNode() {variable=isSkyDiving}.Value)
   RigVMVariableNode(
    Value=RigVMDispatchNode(False=0, True=1, Condition=@MathBoolOr.Result) {template="DISPATCH_RigVMDispatch_If(in Condition,in True,in False,out Result)"}.Result
    ExecuteContext=@RigUnit_BeginExecution.ExecutePin
   ) {variable=randomMotionAlpha}
   n1 = /Script/RigVM.RigVMFunction_RandomFloat(Duration=1, Maximum=1.1, Minimum=5, Seed=50) {is=struct RigVMUnitNode, template="Random::Execute(in Seed,in Minimum,in Maximum,in Duration,out Result)"}
   n2 = RigVMVariableNode() {variable=isTargeting}
   n3 = &RigVMFunction_MathBoolNot(Value=/Script/RigVM.RigVMFunction_MathDoubleGreater(
     B=0.7
     A=&RigVMFunction_RandomFloat(Duration=1, Maximum=1, Minimum=-1, Seed=20) {template="Random::Execute(in Seed,in Minimum,in Maximum,in Duration,out Result)"}.Result
    ) {is=struct RigVMUnitNode, template="Greater::Execute(in A,in B,out Result)"}.Result)
   Noise = /Script/RigVM.RigVMFunction_NoiseFloat(
    Frequency=-5
    Maximum=-10
    Minimum=2
    Value=10
    Speed=RigVMDispatchNode(
     True=0
     False=RigVMDispatchNode(True=1, False=@n1.Result, Condition=@n2.Value) {template="DISPATCH_RigVMDispatch_If(in Condition,in True,in False,out Result)"}.Result
     Condition=@n3.Result
    ) {template="DISPATCH_RigVMDispatch_If(in Condition,in True,in False,out Result)"}.Result
   ) {is=struct RigVMUnitNode, template="Noise::Execute(in Value,in Speed,in Frequency,in Minimum,in Maximum,out Result)"}
   n4 = RigVMVariableNode() {variable=randomMotionAlpha}
   Interpolate = /Script/RigVM.RigVMFunction_MathFloatLerp(B=0.5, A=@Noise.Result, T=@n4.Value) {is=struct RigVMUnitNode, template="Interpolate::Execute(in A,in B,in T,out Result)"}
   n5 = /Script/RigVM.RigVMFunction_MathQuaternionSlerp(A=(X=0,Y=0,Z=0.0087,W=1), B=(X=0,Y=0,Z=0.0262,W=0.9997), T=@Interpolate.Result) {is=struct RigVMUnitNode, template="Interpolate::Execute(in A,in B,in T,out Result)"}
   > &RigUnit_ModifyTransforms(
    ItemToModify="((Item=(Type=Bone,Name=\"dyn_Hair_L_Ear_01\"),Transform=(Rotation=(X=0,Y=0,Z=0,W=1),Translation=(X=0,Y=0,Z=0),Scale3D=(X=1,Y=1,Z=1))),(Item=(Type=Bone,Name=\"dyn_Hair_L_Ear_02\"),Transform=(Rotation=(X=0,Y=0,Z=0,W=1),Translation=(X=0,Y=0,Z=0),Scale3D=(X=1,Y=1, ...(truncated)"
    Weight=1
    ItemToModify.0.Transform.Rotation=@n5.Result
   )
   > n6 = &RigUnit_ModifyTransforms(
    ItemToModify="((Item=(Type=Bone,Name=\"dyn_Hair_R_Ear_01\"),Transform=(Rotation=(X=0,Y=0,Z=0,W=1),Translation=(X=0,Y=0,Z=0),Scale3D=(X=1,Y=1,Z=1))),(Item=(Type=Bone,Name=\"dyn_Hair_R_Ear_02\"),Transform=(Rotation=(X=0,Y=0,Z=0,W=1),Translation=(X=0,Y=0,Z=0),Scale3D=(X=1,Y=1, ...(truncated)"
    Weight=1
    ItemToModify.0.Transform.Rotation=@n5.Result
   )
   RigVMFunction_ControlFlowBranch = &RigVMFunction_ControlFlowBranch(ExecuteContext=@n6.ExecutePin, Condition=RigVMVariableNode() {variable=isWearingPackBool}.Value)
   n7 = &RigUnit_SphericalPoseReader(
    ActiveRegionScaleFactors=(PositiveWidth=3,NegativeWidth=3,PositiveHeight=3,NegativeHeight=3)
    ActiveRegionSize=0.1
    Debug=(bDrawDebug=false,DebugScale=25,DebugSegments=20,DebugThickness=0.25)
    DriverAxis=(X=1,Y=0,Z=0)
    DriverItem=(Type=Bone,Name="head")
    FalloffRegionScaleFactors=(PositiveWidth=1,NegativeWidth=1,PositiveHeight=1,NegativeHeight=0.1)
    FalloffSize=0.58
    FlipHeightScaling=true
    FlipWidthScaling=true
    OptionalParentItem=(Type=Bone,Name="spine_05")
    RotationOffset=(X=-145,Y=0,Z=-30)
    ExecutePin=@RigVMFunction_ControlFlowBranch.False
   )
   > n8 = &RigUnit_ModifyTransforms(
    ItemToModify="((Item=(Type=Bone,Name=\"dyn_Hair_BackL_01\"),Transform=(Rotation=(X=0.0872,Y=0,Z=0,W=0.9962),Translation=(X=5,Y=-1,Z=0),Scale3D=(X=1,Y=1,Z=1))),(Item=(Type=Bone,Name=\"dyn_Hair_BackL_03\"),Transform=(Rotation=(X=0.0868,Y=-0.0076,Z=-0.0868,W=0.9924),Translatio ...(truncated)"
    Weight=@n7.OutputParam
   )
   n9 = &RigVMFunction_ControlFlowBranch(ExecuteContext=@n8.ExecutePin, Condition=RigVMVariableNode() {variable=Not bIsSkyDiving_Floating}.Value)
   n10 = &RigVMFunction_ControlFlowBranch(ExecuteContext=@n9.False, Condition=RigVMVariableNode() {variable=Not bIsSkyDiving}.Value)
   &RigUnit_SphericalPoseReader(
    ActiveRegionScaleFactors=(PositiveWidth=3,NegativeWidth=3,PositiveHeight=3,NegativeHeight=3)
    ActiveRegionSize=0.1
    Debug=(bDrawDebug=false,DebugScale=25,DebugSegments=20,DebugThickness=0.25)
    DriverAxis=(X=1,Y=0,Z=0)
    DriverItem=(Type=Bone,Name="head")
    FalloffRegionScaleFactors=(PositiveWidth=1,NegativeWidth=1,PositiveHeight=1,NegativeHeight=0.1)
    FalloffSize=0.58
    FlipHeightScaling=false
    FlipWidthScaling=true
    OptionalParentItem=(Type=Bone,Name="spine_05")
    RotationOffset=(X=-35,Y=0,Z=-30)
    ExecutePin=@n10.False
   )
   > &RigUnit_ModifyTransforms(
    ItemToModify="((Item=(Type=Bone,Name=\"dyn_Hair_BackR_01\"),Transform=(Rotation=(X=0.0872,Y=0,Z=0,W=0.9962),Translation=(X=5,Y=1,Z=0),Scale3D=(X=1,Y=1,Z=1))),(Item=(Type=Bone,Name=\"dyn_Hair_BackR_03\"),Transform=(Rotation=(X=0.0868,Y=-0.0076,Z=-0.0868,W=0.9924),Translation ...(truncated)"
    Weight=1
   )
   n11 = &RigUnit_SphericalPoseReader(
    ActiveRegionScaleFactors=(PositiveWidth=3,NegativeWidth=3,PositiveHeight=3,NegativeHeight=3)
    ActiveRegionSize=0.1
    Debug=(bDrawDebug=false,DebugScale=25,DebugSegments=20,DebugThickness=0.25)
    DriverAxis=(X=1,Y=0,Z=0)
    DriverItem=(Type=Bone,Name="head")
    FalloffRegionScaleFactors=(PositiveWidth=1,NegativeWidth=1,PositiveHeight=1,NegativeHeight=0.1)
    FalloffSize=0.58
    FlipHeightScaling=false
    FlipWidthScaling=true
    OptionalParentItem=(Type=Bone,Name="clavicle_r")
    RotationOffset=(X=-35,Y=0,Z=-30)
    ExecutePin=@n10.True
   )
   > &RigUnit_ModifyTransforms(
    ItemToModify="((Item=(Type=Bone,Name=\"dyn_Hair_BackR_01\"),Transform=(Rotation=(X=0.0872,Y=0,Z=0,W=0.9962),Translation=(X=5,Y=1,Z=0),Scale3D=(X=1,Y=1,Z=1))),(Item=(Type=Bone,Name=\"dyn_Hair_BackR_03\"),Transform=(Rotation=(X=0.0868,Y=-0.0076,Z=-0.0868,W=0.9924),Translation ...(truncated)"
    Weight=@n11.OutputParam
   )
   Hair IK = RigVMFunctionReferenceNode(ExecuteContext=@RigVMFunction_ControlFlowBranch.True)
   n12 = &RigVMFunction_MathDoubleGreater(
    B=0.7
    A=&RigVMFunction_RandomFloat(Duration=1, Maximum=1, Minimum=-1, Seed=20) {template="Random::Execute(in Seed,in Minimum,in Maximum,in Duration,out Result)"}.Result
   ) {template="Greater::Execute(in A,in B,out Result)"}
   n13 = &RigVMFunction_NoiseFloat(
    Frequency=-5
    Maximum=-10
    Minimum=2
    Value=10
    Speed=RigVMDispatchNode(
     True=0
     Condition=&RigVMFunction_MathBoolNot(Value=@n12.Result).Result
     False=&RigVMFunction_RandomFloat(Duration=1, Maximum=1.1, Minimum=5, Seed=50) {template="Random::Execute(in Seed,in Minimum,in Maximum,in Duration,out Result)"}.Result
    ) {template="DISPATCH_RigVMDispatch_If(in Condition,in True,in False,out Result)"}.Result
   ) {template="Noise::Execute(in Value,in Speed,in Frequency,in Minimum,in Maximum,out Result)"}
   &RigVMFunction_MathFloatLerp(B=1, A=@n13.Result, T=@n4.Value) {template="Interpolate::Execute(in A,in B,in T,out Result)"}
   MathBoolOr_1 = RigVMAggregateNode(
    C=@n3.Result
    A=RigVMVariableNode() {variable=isJumping}.Value
    B=RigVMVariableNode() {variable=isFalling}.Value
    D=RigVMVariableNode() {variable=isEmote}.Value
   )
   &RigVMFunction_MathQuaternionSlerp(
    A=(X=0,Y=0,Z=0.0087,W=1)
    B=(X=0,Y=0,Z=0.0262,W=0.9997)
    T=&RigVMFunction_MathFloatLerp(B=0.5, A=@n13.Result, T=@n4.Value) {template="Interpolate::Execute(in A,in B,in T,out Result)"}.Result
   ) {template="Interpolate::Execute(in A,in B,in T,out Result)"}
  ]}
 RigVMModel Construction Event Graph {nodes [
   n1 = &RigUnit_PrepareForExecution()
   SequenceAggregate = RigVMAggregateNode(ExecuteContext=@n1.ExecutePin)
   RerouteNode_6_1 = RigVMDispatchNode(Values="((Type=Control,Name=\"dyn_Hair_Back_01_ctrl\"),((),Type=Control,Name=\"dyn_Hair_Back_02_ctrl\"),((),Type=Control,Name=\"dyn_Hair_Back_03_ctrl\"),((),Type=Control,Name=\"dyn_Hair_Back_04_ctrl\"),((),Type=Control,Name=\"dyn_Hair_Back_05_ctrl\"),((),Type=Control,Name=\" ...(truncated)") {template="DISPATCH_RigVMDispatch_ArrayMake(in Values,out Array)"}
   RigVMVariableNode(Value=@RerouteNode_6_1.Array, ExecuteContext=@SequenceAggregate.A) {variable=dyn_Hair_Back_ctrls}
   > RigVMVariableNode(Value=/Script/ControlRig.RigUnit_HierarchyGetPoseItemArray(ElementType=All, Initial=false, ItemsToGet=@RerouteNode_6_1.Array) {is=struct RigVMUnitNode}.Pose) {variable=dyn_Hair_Back_pose}
  ]}
]
hierarchy [
 root {loc=<0 0 0>, rot=<0 0 0 1>}
 pelvis {parent=root, loc=<0 -0.5301 98.0571>, rot=<0.0127 0.707 -0.0127 -0.707>}
 spine_01 {parent=pelvis, loc=<6.1208 -0.9328 0>, rot=<0 0 0.034 -0.9994>}
 spine_02 {parent=spine_01, loc=<6.5996 -0.2095 0>, rot=<0 0 0.0238 -0.9997>}
 spine_03 {parent=spine_02, loc=<6.4914 -0.1237 0>, rot=<0 0 0.0149 -0.9999>}
 spine_04 {parent=spine_03, loc=<6.9598 0.0794 0>, rot=<0 0 -0.0132 -0.9999>}
 spine_05 {parent=spine_04, loc=<6.9987 0 0>, rot=<0 0 -0.0015 -1>}
 neck_01 {parent=spine_05, loc=<11.3138 0.1529 0>, rot=<0 0 -0.2003 -0.9797>}
 neck_02 {parent=neck_01, loc=<4.5627 0 0>, rot=<0 0 0.0504 -0.9987>}
 head {parent=neck_02, loc=<5.3914 0 0>, rot=<0 0 0.115 -0.9934>}
 dyn_Hair_Back_01 {parent=head, loc=<8.5016 -1.2653 -0.3258>, rot=<0 0 0.0928 -0.9957>}
 dyn_Hair_Back_02 {parent=dyn_Hair_Back_01, loc=<-1.9753 -3.7984 0>, rot=<0 0 0.5593 -0.8289>}
 dyn_Hair_Back_03 {parent=dyn_Hair_Back_02, loc=<-0.2757 -6.9116 0>, rot=<0 0 -0.004 -1>}
 dyn_Hair_Back_04 {parent=dyn_Hair_Back_03, loc=<0.3289 -8.4743 -0.3554>, rot=<0 0 0 1>}
 dyn_Hair_Back_05 {parent=dyn_Hair_Back_04, loc=<2.6338 -9.663 0>, rot=<0 0 0 1>}
 dyn_Hair_Back_06 {parent=dyn_Hair_Back_05, loc=<-0.9292 -7.1818 0>, rot=<0 0 0.0617 -0.9981>}
 dyn_Hair_BackL_01 {parent=head, loc=<7.9722 -1.0353 -7.6561>, rot=<0 0 0.0928 -0.9957>}
 dyn_Hair_BackL_02 {parent=dyn_Hair_BackL_01, loc=<-1.9753 -3.7984 0>, rot=<0 0 0.5593 -0.8289>}
 dyn_Hair_BackL_03 {parent=dyn_Hair_BackL_02, loc=<-0.2757 -6.9116 0>, rot=<0 0 -0.004 -1>}
 dyn_Hair_BackL_04 {parent=dyn_Hair_BackL_03, loc=<-0.1556 -8.5965 -2.335>, rot=<0 0 0 1>}
 dyn_Hair_BackL_05 {parent=dyn_Hair_BackL_04, loc=<1.8626 -7.0493 0>, rot=<0 0 0 1>}
 dyn_Hair_BackL_06 {parent=dyn_Hair_BackL_05, loc=<0.0921 -7.9976 -0.3361>, rot=<0 0 0.0617 -0.9981>}
 dyn_Hair_BackL_03a_01 {parent=dyn_Hair_BackL_01, loc=<-9.8139 -4.1179 1.2793>, rot=<-0.0334 0.032 0.6917 -0.7207>}
 dyn_Hair_BackL_03a_02 {parent=dyn_Hair_BackL_03a_01, loc=<0 -5.2924 0>, rot=<0 0 0.1707 -0.9853>}
 dyn_Hair_BackL_03a_03 {parent=dyn_Hair_BackL_03a_02, loc=<0 -5.0994 0>, rot=<0 0 0 1>}
 dyn_Hair_BackR_01 {parent=head, loc=<8.1854 -0.439 7.1824>, rot=<0 0 0.0928 -0.9957>}
 dyn_Hair_BackR_02 {parent=dyn_Hair_BackR_01, loc=<-1.9753 -3.7984 0>, rot=<0 0 0.5593 -0.8289>}
 dyn_Hair_BackR_03 {parent=dyn_Hair_BackR_02, loc=<-0.2757 -6.9116 0>, rot=<0 0 -0.004 -1>}
 dyn_Hair_BackR_04 {parent=dyn_Hair_BackR_03, loc=<0.6626 -8.5591 0.6166>, rot=<0 0 0 1>}
 dyn_Hair_BackR_05 {parent=dyn_Hair_BackR_04, loc=<1.8626 -7.0493 0>, rot=<0 0 0 1>}
 dyn_Hair_BackR_06 {parent=dyn_Hair_BackR_05, loc=<0.1207 -7.3917 0>, rot=<0 0 0.0617 -0.9981>}
 dyn_Hair_l_base {parent=head, loc=<4.479 2.9904 -8.0242>, rot=<0.0004 -0.0217 -0.0165 -0.9996>}
 dyn_Hair_l_01 {parent=dyn_Hair_l_base, loc=<-1.8981 0.0334 0>, rot=<0 0 0.0229 -0.9997>}
 dyn_Hair_l_02 {parent=dyn_Hair_l_01, loc=<-6.3517 0.1763 0>, rot=<0 0 0.0518 -0.9987>}
 dyn_Hair_l_03 {parent=dyn_Hair_l_02, loc=<-7.3662 0.1763 0>, rot=<0 0 0.1719 -0.9851>}
 dyn_Hair_l_04 {parent=dyn_Hair_l_03, loc=<-7.1353 0.347 0>, rot=<0 0 0 1>}
 dyn_Hair_r_base {parent=head, loc=<4.4719 3.8511 8.0242>, rot=<0.0078 0.0691 0.0723 -0.995>}
 dyn_Hair_r_01 {parent=dyn_Hair_r_base, loc=<-1.9946 0 0>, rot=<0 0 0 1>}
 dyn_Hair_r_02 {parent=dyn_Hair_r_01, loc=<-6.1849 -0.5743 0.463>, rot=<0.0052 0.0497 0.1046 -0.9933>}
 dyn_Hair_r_03 {parent=dyn_Hair_r_02, loc=<-7.1136 -1.4522 1.3442>, rot=<0 0 0.1161 -0.9932>}
 dyn_Hair_r_04 {parent=dyn_Hair_r_03, loc=<-5.3949 -0.8498 0.6851>, rot=<0 0 0 1>}
 dyn_Hair_Fn_a_01 {parent=head, loc=<8.9968 5.9163 -7.5652>, rot=<-0.1806 -0.2598 0.6518 -0.6893>}
 dyn_Hair_Fn_a_02 {parent=dyn_Hair_Fn_a_01, loc=<0 -3.5742 0>, rot=<0 0 0 1>}
 dyn_Hair_Fn_b_01 {parent=head, loc=<8.6698 9.0654 -5.1512>, rot=<-0.3875 -0.4747 0.5543 -0.5632>}
 dyn_Hair_Fn_b_02 {parent=dyn_Hair_Fn_b_01, loc=<0 -3.5742 0>, rot=<0 0 0 1>}
 dyn_Hair_Fn_c_01 {parent=head, loc=<8.8465 10.69 -0.2023>, rot=<-0.5016 -0.5502 0.4578 -0.4858>}
 dyn_Hair_Fn_c_02 {parent=dyn_Hair_Fn_c_01, loc=<0 -3.5742 0>, rot=<0 0 0 1>}
 dyn_Hair_Fn_d_01 {parent=head, loc=<8.8853 9.456 4.1>, rot=<-0.5705 -0.623 0.3685 -0.388>}
 dyn_Hair_Fn_d_02 {parent=dyn_Hair_Fn_d_01, loc=<0 -3.5742 0>, rot=<0 0 0 1>}
 dyn_Hair_Fn_e_01 {parent=head, loc=<8.751 7.9718 7.0851>, rot=<-0.5799 -0.6918 0.3142 -0.294>}
 dyn_Hair_Fn_e_02 {parent=dyn_Hair_Fn_e_01, loc=<0 -3.9615 0>, rot=<0 0 0 1>}
 dyn_Hair_Fn_f_01 {parent=head, loc=<10.8154 7.9464 7.426>, rot=<-0.4949 -0.7397 0.402 -0.2154>}
 dyn_Hair_Fn_f_02 {parent=dyn_Hair_Fn_f_01, loc=<0 -5.2862 0.2967>, rot=<0.2285 0 0 -0.9735>}
 dyn_Hair_Fn_g_01 {parent=head, loc=<10.9811 10.3192 1.8215>, rot=<-0.5264 -0.4429 0.568 -0.4518>}
 dyn_Hair_Fn_g_02 {parent=dyn_Hair_Fn_g_01, loc=<0 -4.6889 0>, rot=<0.2218 -0.055 -0.2343 -0.9449>}
 dyn_Hair_Fn_h_01 {parent=head, loc=<10.8507 10.1567 -1.632>, rot=<-0.5424 -0.4232 0.551 -0.4724>}
 dyn_Hair_Fn_h_02 {parent=dyn_Hair_Fn_h_01, loc=<0 -4.6889 0>, rot=<0.2275 0.0218 0.0927 -0.9691>}
 dyn_Hair_Fn_i_01 {parent=head, loc=<16.1667 9.2857 0.2624>, rot=<-0.726 -0.1465 0.3416 -0.5785>}
 dyn_Hair_Fn_i_02 {parent=dyn_Hair_Fn_i_01, loc=<-0.3898 -5.364 0.3354>, rot=<0.222 0.054 0.2301 -0.946>}
 dyn_Hair_Fn_k_01 {parent=head, loc=<10.0132 4.0421 -8.4267>, rot=<-0.4725 -0.3645 0.4024 -0.6942>}
 dyn_Hair_Fn_k_02 {parent=dyn_Hair_Fn_k_01, loc=<0 -4.6889 0>, rot=<0 0 0.1861 -0.9825>}
 dyn_Hair_L_Ear_01 {parent=head, loc=<10.3417 -0.1165 -6.1747>, rot=<-0.5706 0.0329 -0.142 -0.8082>}
 dyn_Hair_L_Ear_02 {parent=dyn_Hair_L_Ear_01, loc=<0 -4.6889 0>, rot=<0 0 0.1861 -0.9825>}
 dyn_Hair_R_Ear_01 {parent=head, loc=<10.3421 -0.1165 6.1747>, rot=<0.0329 0.5706 0.8082 -0.142>}
 dyn_Hair_R_Ear_02 {parent=dyn_Hair_R_Ear_01, loc=<0 4.6888 0>, rot=<0 0 0.1861 -0.9825>}
 clavicle_l {parent=spine_05, loc=<5.4831 -0.1053 -3.2946>, rot=<0.0465 -0.7289 -0.009 -0.6829>}
 upperarm_l {parent=clavicle_l, loc=<9.8658 -1.2136 -0.4105>, rot=<0.0164 -0.3245 -0.0269 -0.9454>}
 clavicle_r {parent=spine_05, loc=<5.4831 -0.1053 3.2946>, rot=<0.7289 0.0465 -0.6829 0.009>}
 upperarm_r {parent=clavicle_r, loc=<-9.8658 1.2136 0.4105>, rot=<0.0164 -0.3245 -0.0269 -0.9454>}
 dyn_Hair_Back_01_Null {parent=root, type=Null, loc=<0.3258 -3.0228 160.5251>, rot=<0.0686 -0.7038 -0.0686 0.7038>}
 dyn_Hair_Back_01_ctrl {controlType=EULER_TRANSFORM, parent=dyn_Hair_Back_01_Null, type=Control, loc=<0 0 0>, rot=<0 0 0 1>}
 dyn_Hair_Back_02_Null {parent=root, type=Null, loc=<0.3258 -6.3686 157.8539>, rot=<-0.4505 0.545 0.4505 -0.545>}
 dyn_Hair_Back_02_ctrl {controlType=EULER_TRANSFORM, parent=dyn_Hair_Back_02_Null, type=Control, loc=<0 0 0>, rot=<0 0 0 1>}
 dyn_Hair_Back_03_Null {parent=root, type=Null, loc=<0.3258 -7.3993 151.0139>, rot=<0.4483 -0.5468 -0.4483 0.5468>}
 dyn_Hair_Back_03_ctrl {controlType=EULER_TRANSFORM, parent=dyn_Hair_Back_03_Null, type=Control, loc=<0 0 0>, rot=<0 0 0 1>}
 dyn_Hair_Back_04_Null {parent=root, type=Null, loc=<0.6812 -9.3834 142.7686>, rot=<0.4483 -0.5468 -0.4483 0.5468>}
 dyn_Hair_Back_04_ctrl {controlType=EULER_TRANSFORM, parent=dyn_Hair_Back_04_Null, type=Control, loc=<0 0 0>, rot=<0 0 0 1>}
 dyn_Hair_Back_05_Null {parent=root, type=Null, loc=<0.6812 -13.8608 133.8095>, rot=<0.4483 -0.5468 -0.4483 0.5468>}
 dyn_Hair_Back_05_ctrl {controlType=EULER_TRANSFORM, parent=dyn_Hair_Back_05_Null, type=Control, loc=<0 0 0>, rot=<0 0 0 1>}
 dyn_Hair_Back_06_Null {parent=root, type=Null, loc=<0.6812 -14.3578 126.585>, rot=<-0.4812 0.5181 0.4812 -0.5181>}
 dyn_Hair_Back_06_ctrl {controlType=EULER_TRANSFORM, parent=dyn_Hair_Back_06_Null, type=Control, loc=<0 0 0>, rot=<0 0 0 1>}
]
variables [
 Not bIsSkyDiving {type=bool, default=False}
 Not bIsSkyDiving_Floating {type=bool, default=False}
 dyn_Hair_Back_ctrls {
  type="TArray<FRigElementKey>"
  default=((Type=Control,Name="dyn_Hair_BackL_02"),(Type=Control,Name="dyn_Hair_BackL_03"),(Type=Control,Name="dyn_Hair_BackL_04"),(Type=Control,Name="dyn_Hair_BackL_05"),(Type=Control,Name="dyn_Hair_BackL_06"))
 }
 dyn_Hair_Back_pose {type=FRigPose, default=(Elements=,HierarchyTopologyVersion=-1,PoseHash=-1)}
 isEmote {type=bool, default=False}
 isFalling {type=bool, default=False}
 isJumping {type=bool, default=False}
 isSkyDiving {type=bool, default=False}
 isTargeting {type=bool, default=False}
 isWearingPackBool {type=bool, default=False}
 randomMotionAlpha {type=double, default=0}
]
