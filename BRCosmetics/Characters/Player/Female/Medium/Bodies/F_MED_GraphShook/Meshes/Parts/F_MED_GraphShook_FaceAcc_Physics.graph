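# Physics asset for the GraphShook female medium character's FaceAcc mesh, rigging its dynamic hair bones with capsule bodies.
# Three back chains, a branch off the back-left root and two side chains hang from roots locked to the head, with limited joints favoring one swing axis.
# Spine, neck, head, clavicle and upper-arm capsules form the body volume.
# All-free constraints link the body bones, neighboring back-hair segments, and one right-side segment to spine_05.
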PhysicsAsset /BRCosmetics/Characters/Player/Female/Medium/Bodies/F_MED_GraphShook/Meshes/Parts/F_MED_GraphShook_FaceAcc_Physics
bodies {
 clavicle_l [{
   shapeType=Capsule
   length=6.5494
   radius=5.876
   shapeName=upperarm_l_capsule
   center=<7.955 -0.6654 -1.5705>
   rotation=<rot -88.0819 0.0003 -0.0002>
  }]
 clavicle_r [{shapeType=Capsule, length=5.2395, radius=6.14, shapeName=upperarm_r_capsule, center=<-4.9697 -0.214 1.806>, rotation=<rot 87.5241 -180 180>}]
 dyn_Hair_BackL_01 [{shapeType=Capsule, length=5, radius=5, shapeName=dyn_Hair_BackL_01_capsule, center=<3.7603 0 0>}]
 dyn_Hair_BackL_02 [{
   shapeType=Capsule
   length=5.9257
   radius=4.0743
   shapeName=dyn_Hair_BackR_02_capsule
   center=<4.3174 0 0>
   rotation=<rot -0.0089 0.8275 90.6131>
  }]
 dyn_Hair_BackL_03 [{
   shapeType=Capsule
   length=3.8473
   radius=6.1527
   shapeName=dyn_Hair_BackL_03_capsule
   center=<5.7758 -6.2396 0.0668>
   rotation=<rot 0 0 90.6132>
  }]
 dyn_Hair_BackL_03a_01 [{
   shapeType=Capsule
   length=4.4242
   radius=2.483
   shapeName=dyn_Hair_BackL_04_capsule
   center=<0.4863 -2.0853 -0.0032>
   rotation=<rot 0 0 -90.0875>
  }]
 dyn_Hair_BackL_03a_02 [{
   shapeType=Capsule
   length=4.0881
   radius=2.8191
   shapeName=dyn_Hair_BackL_05_capsule
   center=<0.4686 -2.2788 -1.5784>
   rotation=<rot -0.0182 -12.0007 -90.0856>
  }]
 dyn_Hair_BackL_03a_03 [{
   shapeType=Capsule
   length=4.2474
   radius=2.6598
   shapeName=dyn_Hair_BackL_06_capsule
   center=<-1.7795 -2.1577 -0.0033>
   rotation=<rot -0.0307 -20.5094 -90.082>
  }]
 dyn_Hair_BackL_04 [{
   shapeType=Capsule
   length=4.3385
   radius=5.6615
   shapeName=dyn_Hair_BackL_04_capsule
   center=<4.3181 -6.2396 0.0668>
   rotation=<rot 0 0 90.6132>
  }]
 dyn_Hair_BackR_01 [{shapeType=Capsule, length=5, radius=5, shapeName=dyn_Hair_BackR_01_capsule, center=<3.7603 0 0>}]
 dyn_Hair_BackR_02 [{shapeType=Capsule, length=5.9257, radius=4.0743, shapeName=dyn_Hair_BackR_02_capsule, center=<4.3174 0 0>, rotation=<rot 0 0 90.6132>}]
 dyn_Hair_BackR_03 [{
   shapeType=Capsule
   length=3.8473
   radius=6.1527
   shapeName=dyn_Hair_BackL_03_capsule
   center=<4.6236 -6.2396 0.0668>
   rotation=<rot 0 0 90.6132>
  }]
 dyn_Hair_BackR_04 [{
   shapeType=Capsule
   length=3.3412
   radius=6.6588
   shapeName=dyn_Hair_BackL_04_capsule
   center=<4.3645 -6.2396 0.0668>
   rotation=<rot 0 0 90.6132>
  }]
 dyn_Hair_Back_01 [{shapeType=Capsule, length=4.0074, radius=14.7434, shapeName=dyn_Hair_Back_01_capsule, center=<-22.7694 4.5907 0>, rotation=<rot 0 10.6133 0>}]
 dyn_Hair_Back_02 [{shapeType=Capsule, length=5.9257, radius=4.0743, shapeName=dyn_Hair_Back_02_capsule, center=<4.3174 0 0>, rotation=<rot 0 0 90.6132>}]
 dyn_Hair_Back_03 [{
   shapeType=Capsule
   length=4.7996
   radius=5.2004
   shapeName=dyn_Hair_BackL_03_capsule
   center=<5.7758 -6.2396 0.0668>
   rotation=<rot 0 0 90.6132>
  }]
 dyn_Hair_Back_04 [{
   shapeType=Capsule
   length=5.6648
   radius=4.3352
   shapeName=dyn_Hair_Back_04_capsule
   center=<5.7758 -6.2396 0.0668>
   rotation=<rot 0 0 90.6132>
  }]
 dyn_Hair_l_01 [{
   shapeType=Capsule
   length=5.3762
   radius=2.6923
   shapeName=dyn_Hair_BackL_03_capsule
   center=<-3.2501 0 -1.1>
   rotation=<rot 88.9361 -180 -180>
  }]
 dyn_Hair_l_02 [{
   shapeType=Capsule
   length=5.7061
   radius=2.3624
   shapeName=dyn_Hair_BackL_04_capsule
   center=<-3.2694 0 -0.0607>
   rotation=<rot 88.9361 -180 -180>
  }]
 dyn_Hair_l_03 [{
   shapeType=Capsule
   length=5.5577
   radius=2.5108
   shapeName=dyn_Hair_BackL_05_capsule
   center=<-3.2694 0 -0.0607>
   rotation=<rot 88.9361 -180 -180>
  }]
 dyn_Hair_l_base [{shapeType=Capsule, length=0.505, radius=0.505, shapeName=dyn_Hair_l_base_capsule, rotation=<rot 0 -90 0>}]
 dyn_Hair_r_01 [{shapeType=Capsule, length=5.1065, radius=2.8736, center=<-2.6819 0 1.7574>, rotation=<rot 85.9884 0 0>}]
 dyn_Hair_r_02 [{shapeType=Capsule, length=4.3837, radius=3.5964, center=<-2.6732 -0.531 0.1875>, rotation=<rot 85.9884 0 -12.5816>}]
 dyn_Hair_r_03 [{shapeType=Capsule, length=4.255, radius=3.7251, center=<-1.4876 -0.625 -0.0553>, rotation=<rot 85.9884 0 -3.9293>}]
 dyn_Hair_r_04 [{shapeType=Capsule, length=4.9804, radius=2.9997, center=<-2.7915 0 0.1958>, rotation=<rot 85.9884 0 0>}]
 dyn_Hair_r_base [{shapeType=Capsule, length=0.505, radius=0.505, shapeName=dyn_Hair_r_base_capsule, rotation=<rot 0 -90 0>}]
 head [{shapeType=Capsule, length=4.8773, radius=6.2998, center=<1.5191 1.9108 0.0132>, rotation=<rot 89.5019 180 -180>}]
 neck_01 [{shapeType=Capsule, length=4.0074, radius=14.7434, shapeName=dyn_Hair_Back_01_capsule, center=<0.3956 10.9972 0>, rotation=<rot 0 10.6133 0>}]
 spine_04 [{shapeType=Capsule, length=12.8376, radius=11.7926, shapeName=spine_04_capsule, center=<0.6955 -0.881 0>}]
 spine_05 [{
   shapeType=Capsule
   length=18.67
   radius=7.8573
   shapeName=spine_05_capsule
   center=<0.2418 -0.0629 0.861>
   rotation=<rot -3.1797 -0.0071 0.2559>
  }]
 upperarm_l [{
   shapeType=Capsule
   length=18.5297
   radius=6.6788
   shapeName=upperarm_l_capsule
   center=<14.806 -0.6654 1.1141>
   rotation=<rot -82.4891 179.9999 -179.9999>
  }]
 upperarm_r [{shapeType=Capsule, length=21.6919, radius=6.1567, shapeName=upperarm_r_capsule, center=<-16.6121 1.0855 -2.0885>, rotation=<rot 83.0721 0 0>}]
}
constraints [
 {bone1Name=dyn_Hair_BackL_02, bone2Name=dyn_Hair_BackL_01, swing1LimitDegrees=20, swing2LimitDegrees=1, twistLimitDegrees=1}
 {bone1Name=dyn_Hair_Back_03, bone2Name=dyn_Hair_Back_02, swing1LimitDegrees=25, swing2LimitDegrees=5, twistLimitDegrees=8}
 {bone1Name=dyn_Hair_BackR_03, bone2Name=dyn_Hair_BackR_02, swing1LimitDegrees=25, swing2LimitDegrees=5, twistLimitDegrees=8}
 {
  bone1Name=dyn_Hair_BackL_01
  bone2Name=head
  swing1LimitDegrees=45
  swing1Motion=Locked
  swing2LimitDegrees=45
  swing2Motion=Locked
  twistLimitDegrees=45
  twistMotion=Locked
 }
 {bone1Name=dyn_Hair_BackL_03, bone2Name=dyn_Hair_BackL_02, swing1LimitDegrees=25, swing2LimitDegrees=5, twistLimitDegrees=8}
 {bone1Name=dyn_Hair_BackL_04, bone2Name=dyn_Hair_BackL_03, swing1LimitDegrees=60, swing2LimitDegrees=5, twistLimitDegrees=8}
 {
  bone1Name=dyn_Hair_BackR_01
  bone2Name=head
  swing1LimitDegrees=45
  swing1Motion=Locked
  swing2LimitDegrees=45
  swing2Motion=Locked
  twistLimitDegrees=45
  twistMotion=Locked
 }
 {bone1Name=dyn_Hair_BackR_02, bone2Name=dyn_Hair_BackR_01, swing1LimitDegrees=20, swing2LimitDegrees=1, twistLimitDegrees=1}
 {bone1Name=dyn_Hair_BackR_04, bone2Name=dyn_Hair_BackR_03, swing1LimitDegrees=60, swing2LimitDegrees=5, twistLimitDegrees=8}
 {
  bone1Name=dyn_Hair_Back_01
  bone2Name=head
  swing1LimitDegrees=45
  swing1Motion=Locked
  swing2LimitDegrees=45
  swing2Motion=Locked
  twistLimitDegrees=45
  twistMotion=Locked
 }
 {bone1Name=dyn_Hair_Back_02, bone2Name=dyn_Hair_Back_01, swing1LimitDegrees=20, swing2LimitDegrees=1, twistLimitDegrees=1}
 {bone1Name=dyn_Hair_Back_04, bone2Name=dyn_Hair_Back_03, swing1LimitDegrees=60, swing2LimitDegrees=5, twistLimitDegrees=8}
 {
  bone1Name=head
  bone2Name=neck_01
  swing1LimitDegrees=45
  swing1Motion=Free
  swing2LimitDegrees=45
  swing2Motion=Free
  twistLimitDegrees=45
  twistMotion=Free
 }
 {
  bone1Name=spine_05
  bone2Name=spine_04
  swing1LimitDegrees=45
  swing1Motion=Free
  swing2LimitDegrees=45
  swing2Motion=Free
  twistLimitDegrees=45
  twistMotion=Free
 }
 {
  bone1Name=upperarm_r
  bone2Name=clavicle_r
  swing1LimitDegrees=45
  swing1Motion=Free
  swing2LimitDegrees=45
  swing2Motion=Free
  twistLimitDegrees=45
  twistMotion=Free
 }
 {
  bone1Name=upperarm_l
  bone2Name=clavicle_l
  swing1LimitDegrees=45
  swing1Motion=Free
  swing2LimitDegrees=45
  swing2Motion=Free
  twistLimitDegrees=45
  twistMotion=Free
 }
 {
  bone1Name=dyn_Hair_BackR_03
  bone2Name=dyn_Hair_BackL_03
  swing1LimitDegrees=45
  swing1Motion=Free
  swing2LimitDegrees=45
  swing2Motion=Free
  twistLimitDegrees=45
  twistMotion=Free
 }
 {bone1Name=dyn_Hair_BackL_03a_01, bone2Name=dyn_Hair_BackL_01, swing1LimitDegrees=45, swing2LimitDegrees=5, twistLimitDegrees=10}
 {bone1Name=dyn_Hair_BackL_03a_02, bone2Name=dyn_Hair_BackL_03a_01, swing1LimitDegrees=45, swing2LimitDegrees=5, twistLimitDegrees=10}
 {bone1Name=dyn_Hair_BackL_03a_03, bone2Name=dyn_Hair_BackL_03a_02, swing1LimitDegrees=45, swing2LimitDegrees=5, twistLimitDegrees=10}
 {bone1Name=dyn_Hair_l_01, bone2Name=dyn_Hair_l_base, swing1LimitDegrees=45, swing2LimitDegrees=5, twistLimitDegrees=15}
 {bone1Name=dyn_Hair_l_02, bone2Name=dyn_Hair_l_01, swing1LimitDegrees=20, swing2LimitDegrees=1, twistLimitDegrees=5}
 {bone1Name=dyn_Hair_l_03, bone2Name=dyn_Hair_l_02, swing1LimitDegrees=20, swing2LimitDegrees=1, twistLimitDegrees=5}
 {bone1Name=dyn_Hair_r_01, bone2Name=dyn_Hair_r_base, swing1LimitDegrees=45, swing2LimitDegrees=5, twistLimitDegrees=10}
 {bone1Name=dyn_Hair_r_02, bone2Name=dyn_Hair_r_01, swing1LimitDegrees=45, swing2LimitDegrees=5, twistLimitDegrees=10}
 {bone1Name=dyn_Hair_r_03, bone2Name=dyn_Hair_r_02, swing1LimitDegrees=45, swing2LimitDegrees=5, twistLimitDegrees=10}
 {
  bone1Name=dyn_Hair_l_base
  bone2Name=head
  swing1LimitDegrees=45
  swing1Motion=Locked
  swing2LimitDegrees=45
  swing2Motion=Locked
  twistLimitDegrees=45
  twistMotion=Locked
 }
 {
  bone1Name=dyn_Hair_r_base
  bone2Name=head
  swing1LimitDegrees=45
  swing1Motion=Locked
  swing2LimitDegrees=45
  swing2Motion=Locked
  twistLimitDegrees=45
  twistMotion=Locked
 }
 {bone1Name=dyn_Hair_r_04, bone2Name=dyn_Hair_r_03, swing1LimitDegrees=45, swing2LimitDegrees=5, twistLimitDegrees=10}
 {
  bone1Name=dyn_Hair_BackR_04
  bone2Name=dyn_Hair_Back_04
  swing1LimitDegrees=45
  swing1Motion=Free
  swing2LimitDegrees=45
  swing2Motion=Free
  twistLimitDegrees=45
  twistMotion=Free
 }
 {
  bone1Name=dyn_Hair_BackL_04
  bone2Name=dyn_Hair_Back_04
  swing1LimitDegrees=45
  swing1Motion=Free
  swing2LimitDegrees=45
  swing2Motion=Free
  twistLimitDegrees=45
  twistMotion=Free
 }
 {
  bone1Name=dyn_Hair_BackR_03
  bone2Name=dyn_Hair_Back_03
  swing1LimitDegrees=45
  swing1Motion=Free
  swing2LimitDegrees=45
  swing2Motion=Free
  twistLimitDegrees=45
  twistMotion=Free
 }
 {
  bone1Name=dyn_Hair_BackL_03
  bone2Name=dyn_Hair_Back_03
  swing1LimitDegrees=45
  swing1Motion=Free
  swing2LimitDegrees=45
  swing2Motion=Free
  twistLimitDegrees=45
  twistMotion=Free
 }
 {
  bone1Name=clavicle_r
  bone2Name=spine_05
  swing1LimitDegrees=45
  swing1Motion=Free
  swing2LimitDegrees=45
  swing2Motion=Free
  twistLimitDegrees=45
  twistMotion=Free
 }
 {
  bone1Name=clavicle_l
  bone2Name=spine_05
  swing1LimitDegrees=45
  swing1Motion=Free
  swing2LimitDegrees=45
  swing2Motion=Free
  twistLimitDegrees=45
  twistMotion=Free
 }
 {
  bone1Name=dyn_Hair_r_03
  bone2Name=spine_05
  swing1LimitDegrees=45
  swing1Motion=Free
  swing2LimitDegrees=45
  swing2Motion=Free
  twistLimitDegrees=45
  twistMotion=Free
 }
 {
  bone1Name=dyn_Hair_Back_04
  bone2Name=dyn_Hair_BackR_04
  swing1LimitDegrees=45
  swing1Motion=Free
  swing2LimitDegrees=45
  swing2Motion=Free
  twistLimitDegrees=45
  twistMotion=Free
 }
 {
  bone1Name=dyn_Hair_Back_04
  bone2Name=dyn_Hair_BackL_04
  swing1LimitDegrees=45
  swing1Motion=Free
  swing2LimitDegrees=45
  swing2Motion=Free
  twistLimitDegrees=45
  twistMotion=Free
 }
]
