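# Animation Blueprint for the GraphShook female face accessory that adds secondary hair motion on a pose copied from the parent mesh.
# AnimDynamics springs move front and ear bones, blended by gameplay state.
# Back bling selects a backpack-specific RigidBody pass; otherwise a RigidBody pass with parameter-driven gravity feeds three back-hair AnimDynamics nodes.
# A Control Rig finishes the pose using pack and movement flags such as skydiving, jumping and emoting.
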AnimBlueprint /BRCosmetics/Characters/Player/Female/Medium/Bodies/F_MED_GraphShook/Meshes/Parts/F_MED_GraphShook_FaceAcc_AnimBP
target_skeleton=&/F_MED_GraphShook_FaceAcc_Skeleton
graphs [
 AnimGraph {nodes [
   n1 = RigidBody!(
    ComponentPose=Local To Component(LocalPose=Use cached pose 'HairPose'())
    Alpha
    OverrideWorldGravity:Vector=Get Gravity Override Cape() -> (Gravity Override Cape:Vector)
    ExternalForce
    SimSpaceSettings:Sim Space Settings Structure=(WorldAlpha=1,VelocityScaleZ=1,DampingAlpha=0.5,MaxLinearVelocity=10,MaxAngularVelocity=0.5,MaxLinearAcceleration=10,MaxAngularAcceleration=80,ExternalLinearDragV=(X=1,Y=1,Z=1),ExternalLinearVelocity=(X=0,Y=0,Z=0),ExternalAngularVelocity=(X=0,Y=0,Z=0))
   ) -> (Pose) {
    bClampLinearTranslationLimitToRefPose=true
    bOverrideWorldGravity=true
    overridePhysicsAsset=&/F_MED_GraphShook_FaceAcc_Physics
    simulationSpace=BaseBoneSpace
    alphaBoolBlend {blendOption=Linear}
    baseBoneRef {boneName=neck_02}
    componentAppliedLinearAccClamp=<0 0 0>
    overrideWorldGravity {z=-980}
    simSpaceSettings {
     dampingAlpha=0.5
     maxAngularAcceleration=80
     maxAngularVelocity=0.5
     maxLinearAcceleration=10
     maxLinearVelocity=10
     worldAlpha=1
     externalLinearDragV=<1 1 1>
    }
   }
   n2 = Anim Dynamics - Bone: dyn_Hair_BackL_05!(ComponentPose=@n1, Alpha=1, GravityOverride:Vector="0,0,500", ExternalForce="0,0,0") -> (Pose) {
    angularDampingOverride=0.99
    angularSpringConstant=5000
    bAngularSpring=true
    bOverrideAngularDamping=true
    bUseGravityOverride=true
    gravityScale=0.3
    simulationSpace=BoneRelative
    alphaBoolBlend {blendOption=Linear}
    boundBone {boneName=dyn_Hair_BackL_05}
    componentAppliedLinearAccClamp=<2500 2500 1500>
    componentLinearAccScale=<0.3 0.3 0.3>
    componentLinearVelScale=<0.35 0.35 0.35>
    gravityOverride {z=500}
    physicsBodyDefinitions [{
      collisionType=CustomSphere
      sphereCollisionRadius=2
      boundBone {boneName=dyn_Hair_BackL_05}
      boxExtents=<12 12 12>
      constraintSetup {angularLimitsMax=<0 3 20>, angularLimitsMin=<0 -3 -10>, angularTarget=<0 0 0>}
      localJointOffset=<0.5 -7 0>
     }]
    relativeSpaceBone {boneName=root}
    simSpaceSettings {maxAngularAcceleration=500, maxAngularVelocity=50}
   }
   n3 = Anim Dynamics - Bone: dyn_Hair_Back_05!(ComponentPose=@n2, Alpha=1, GravityOverride:Vector="0,0,-980", ExternalForce="0,0,0") -> (Pose) {
    angularDampingOverride=0.99
    angularSpringConstant=5000
    bAngularSpring=true
    bOverrideAngularDamping=true
    bUseGravityOverride=true
    gravityScale=0.3
    simulationSpace=BoneRelative
    alphaBoolBlend {blendOption=Linear}
    boundBone {boneName=dyn_Hair_Back_05}
    componentAppliedLinearAccClamp=<2500 2500 1500>
    componentLinearAccScale=<0.3 0.3 0.3>
    componentLinearVelScale=<0.35 0.35 0.35>
    gravityOverride {z=-980}
    physicsBodyDefinitions [{
      collisionType=CustomSphere
      sphereCollisionRadius=2
      boundBone {boneName=dyn_Hair_Back_05}
      boxExtents=<12 12 12>
      constraintSetup {angularLimitsMax=<0 3 20>, angularLimitsMin=<0 -3 -10>, angularTarget=<0 0 0>}
      localJointOffset=<0.5 -7 0>
     }]
    relativeSpaceBone {boneName=root}
    simSpaceSettings {maxAngularAcceleration=500, maxAngularVelocity=50}
   }
   n4 = Anim Dynamics - Bone: dyn_Hair_BackR_05!(ComponentPose=@n3, Alpha=1, GravityOverride:Vector="0,0,500", ExternalForce="0,0,0") -> (Pose) {
    angularDampingOverride=0.99
    angularSpringConstant=5000
    bAngularSpring=true
    bOverrideAngularDamping=true
    bUseGravityOverride=true
    gravityScale=0.3
    simulationSpace=BoneRelative
    alphaBoolBlend {blendOption=Linear}
    boundBone {boneName=dyn_Hair_BackR_05}
    componentAppliedLinearAccClamp=<2500 2500 1500>
    componentLinearAccScale=<0.3 0.3 0.3>
    componentLinearVelScale=<0.35 0.35 0.35>
    gravityOverride {z=500}
    physicsBodyDefinitions [{
      collisionType=CustomSphere
      sphereCollisionRadius=2
      boundBone {boneName=dyn_Hair_BackR_05}
      boxExtents=<12 12 12>
      constraintSetup {angularLimitsMax=<0 6 20>, angularLimitsMin=<0 -6 -10>, angularTarget=<0 0 0>}
      localJointOffset=<0.5 -7 0>
     }]
    relativeSpaceBone {boneName=root}
    simSpaceSettings {maxAngularAcceleration=500, maxAngularVelocity=50}
   }
   n5 = RigidBody(
    ComponentPose=Local To Component(LocalPose=Use cached pose 'HairDynPose'())
    OverrideWorldGravity="0,0,0"
    SimSpaceSettings=(WorldAlpha=1,VelocityScaleZ=1,DampingAlpha=0.5,MaxLinearVelocity=10,MaxAngularVelocity=0.5,MaxLinearAcceleration=10,MaxAngularAcceleration=80,ExternalLinearDragV=(X=1,Y=1,Z=1),ExternalLinearVelocity=(X=0,Y=0,Z=0),ExternalAngularVelocity=(X=0,Y=0,Z=0))
   ) {
    bClampLinearTranslationLimitToRefPose=true
    bOverrideWorldGravity=true
    overridePhysicsAsset=&/F_MED_GraphShook_FaceAcc_Backpack_Physics1
    simulationSpace=BaseBoneSpace
    alphaBoolBlend {blendOption=Linear}
    baseBoneRef {boneName=neck_02}
    componentAppliedLinearAccClamp=<0 0 0>
    simSpaceSettings {
     dampingAlpha=0.5
     maxAngularAcceleration=80
     maxAngularVelocity=0.5
     maxLinearAcceleration=10
     maxLinearVelocity=10
     worldAlpha=1
     externalLinearDragV=<1 1 1>
    }
   }
   n6 = Control Rig!(
    Source=Blend Poses by bool(
     BlendPose_0=Component To Local(ComponentPose=@n5)
     BlendPose_1=Component To Local(ComponentPose=@n4)
     bActiveValue=Get isWearingPackBool() -> (isWearingPackBool:bool)
    ) {blendType=Linear}
    Alpha
    isWearingPackBool:bool=Get isWearingPackBool()
    isTargeting:bool=Get bIsTargeting()
    isSkyDiving:bool=Get bIsSkyDiving()
    isJumping:bool=Get bIsJumping()
    isFalling:bool=Get bIsFalling()
    isEmote:bool=Get bIsPlayingEmote()
    Not bIsSkyDiving:bool=Get Not bIsSkyDiving() -> (Not bIsSkyDiving:bool)
    Not bIsSkyDiving_Floating:bool=Get Not bIsSkyDiving_Floating() -> (Not bIsSkyDiving_Floating:bool)
   ) -> (Pose) {
    alphaBoolBlend {blendOption=Linear}
    controlRigAssetReference {blueprintRigClass=&/F_MED_GraphShook_FaceAcc_CtrlRig.F_MED_GraphShook_FaceAcc_CtrlRig_C}
   }
   Output Pose(Result=@n6) {layerGroup=None}
   n7 = Get isWearingPackBool()
   n8 = Get bIsDBNO()
   n9 = Two Way Blend(
    A=Use cached pose 'HairDynPose'()
    B=Use cached pose 'HairCachePose'()
    Alpha=SelectFloat(A=1, B=0.2, bPickA=OR Boolean!(A=@n7, B=@n8, C:bool) -> (ReturnValue))
   ) {blendNode {alphaBoolBlend {blendOption=Linear}}}
   n10 = Use cached pose 'HairDynPose'()
   n11 = OR Boolean!(A=Get bIsSkydiveDiveMode(), B=Get bIsSkydiveFloating(), C=Get isWearingPackBool(), D:bool=Get bIsDBNO()) -> (ReturnValue)
   Save cached pose 'HairPose'(Pose=Blend Poses by bool(
     BlendPose_0=Blend Poses by bool(BlendPose_0=@n9, BlendPose_1=@n10, bActiveValue=@n11) {blendType=Linear}
     BlendPose_1=Use cached pose 'HairCachePose'()
     bActiveValue=Get bAnimDynamicsEnabled() -> (bAnimDynamicsEnabled:bool)
    ) {blendType=Linear}) {cacheName=HairPose}
   Save cached pose 'HairCachePose'(Pose=Copy Pose From Mesh(SourceMeshComponent=Get Skel_Mesh()) {bUseAttachedParent=true}) {cacheName=HairCachePose}
   n12 = Anim Dynamics - Bone: dyn_Hair_Fn_a_01!(
    ComponentPose=Local To Component(LocalPose=Use cached pose 'HairCachePose'())
    Alpha=1
    GravityOverride:Vector="0,0,500"
    ExternalForce="0,0,0"
   ) -> (Pose) {
    angularDampingOverride=0.9999
    angularSpringConstant=800
    bAngularSpring=true
    bOverrideAngularDamping=true
    bUseGravityOverride=true
    gravityScale=0.3
    simulationSpace=RootRelative
    alphaBoolBlend {blendOption=Linear}
    boundBone {boneName=dyn_Hair_Fn_a_01}
    componentAppliedLinearAccClamp=<2000 2000 1600>
    componentLinearAccScale=<0.4 0.4 0.4>
    componentLinearVelScale=<0.45 0.45 0.55>
    gravityOverride {z=500}
    physicsBodyDefinitions [{
      sphereCollisionRadius=2
      boundBone {boneName=dyn_Hair_Fn_a_01}
      boxExtents=<7 12 14>
      constraintSetup {twistAxis=AxisY, angularLimitsMax=<5 0 5>, angularLimitsMin=<5 0 -5>, angularTarget=<0 0 0>}
      localJointOffset=<0 -2 0>
     }]
    relativeSpaceBone {boneName=root}
   }
   n13 = Anim Dynamics - Bone: dyn_Hair_Fn_a_02!(ComponentPose=@n12, Alpha=1, GravityOverride:Vector="0,0,500", ExternalForce="0,0,0") -> (Pose) {
    angularDampingOverride=0.9999
    angularSpringConstant=800
    bAngularSpring=true
    bOverrideAngularDamping=true
    bUseGravityOverride=true
    gravityScale=0.3
    simulationSpace=RootRelative
    alphaBoolBlend {blendOption=Linear}
    boundBone {boneName=dyn_Hair_Fn_a_02}
    componentAppliedLinearAccClamp=<2000 2000 1600>
    componentLinearAccScale=<0.4 0.4 0.4>
    componentLinearVelScale=<0.45 0.45 0.55>
    gravityOverride {z=500}
    physicsBodyDefinitions [{
      sphereCollisionRadius=2
      boundBone {boneName=dyn_Hair_Fn_a_02}
      boxExtents=<5 7 10>
      constraintSetup {twistAxis=AxisY, angularLimitsMax=<5 0 5>, angularLimitsMin=<5 0 -5>, angularTarget=<0 0 0>}
      localJointOffset=<0 -2 0>
     }]
    relativeSpaceBone {boneName=root}
   }
   n14 = Anim Dynamics - Bone: dyn_Hair_Fn_b_01!(ComponentPose=@n13, Alpha=1, GravityOverride:Vector="0,0,500", ExternalForce="0,0,0") -> (Pose) {
    angularDampingOverride=0.9999
    angularSpringConstant=800
    bAngularSpring=true
    bOverrideAngularDamping=true
    bUseGravityOverride=true
    gravityScale=0.3
    simulationSpace=RootRelative
    alphaBoolBlend {blendOption=Linear}
    boundBone {boneName=dyn_Hair_Fn_b_01}
    componentAppliedLinearAccClamp=<2000 2000 1600>
    componentLinearAccScale=<0.4 0.4 0.4>
    componentLinearVelScale=<0.45 0.45 0.55>
    gravityOverride {z=500}
    physicsBodyDefinitions [{
      sphereCollisionRadius=2
      boundBone {boneName=dyn_Hair_Fn_b_01}
      boxExtents=<7 12 14>
      constraintSetup {twistAxis=AxisY, angularLimitsMax=<15 0 5>, angularLimitsMin=<-2 0 -10>, angularTarget=<0 0 0>}
      localJointOffset=<0 -2 0>
     }]
    relativeSpaceBone {boneName=root}
   }
   n15 = Anim Dynamics - Bone: dyn_Hair_Fn_b_02!(ComponentPose=@n14, Alpha=1, GravityOverride:Vector="0,0,500", ExternalForce="0,0,0") -> (Pose) {
    angularDampingOverride=0.9999
    angularSpringConstant=800
    bAngularSpring=true
    bOverrideAngularDamping=true
    bUseGravityOverride=true
    gravityScale=0.3
    simulationSpace=RootRelative
    alphaBoolBlend {blendOption=Linear}
    boundBone {boneName=dyn_Hair_Fn_b_02}
    componentAppliedLinearAccClamp=<2000 2000 1600>
    componentLinearAccScale=<0.4 0.4 0.4>
    componentLinearVelScale=<0.45 0.45 0.55>
    gravityOverride {z=500}
    physicsBodyDefinitions [{
      sphereCollisionRadius=2
      boundBone {boneName=dyn_Hair_Fn_b_02}
      boxExtents=<5 7 10>
      constraintSetup {twistAxis=AxisY, angularLimitsMax=<15 0 5>, angularLimitsMin=<-2 0 -10>, angularTarget=<0 0 0>}
      localJointOffset=<0 -2 0>
     }]
    relativeSpaceBone {boneName=root}
   }
   n16 = Anim Dynamics - Bone: dyn_Hair_Fn_c_01!(ComponentPose=@n15, Alpha=1, GravityOverride:Vector="0,0,500", ExternalForce="0,0,0") -> (Pose) {
    angularDampingOverride=0.9999
    angularSpringConstant=800
    bAngularSpring=true
    bOverrideAngularDamping=true
    bUseGravityOverride=true
    gravityScale=0.3
    simulationSpace=RootRelative
    alphaBoolBlend {blendOption=Linear}
    boundBone {boneName=dyn_Hair_Fn_c_01}
    componentAppliedLinearAccClamp=<2000 2000 1600>
    componentLinearAccScale=<0.4 0.4 0.4>
    componentLinearVelScale=<0.45 0.45 0.55>
    gravityOverride {z=500}
    physicsBodyDefinitions [{
      sphereCollisionRadius=2
      boundBone {boneName=dyn_Hair_Fn_c_01}
      boxExtents=<7 12 14>
      constraintSetup {twistAxis=AxisY, angularLimitsMax=<15 0 7>, angularLimitsMin=<-2 0 -2>, angularTarget=<0 0 0>}
      localJointOffset=<0 -2 0>
     }]
    relativeSpaceBone {boneName=root}
   }
   n17 = Anim Dynamics - Bone: dyn_Hair_Fn_c_02!(ComponentPose=@n16, Alpha=1, GravityOverride:Vector="0,0,500", ExternalForce="0,0,0") -> (Pose) {
    angularDampingOverride=0.9999
    angularSpringConstant=800
    bAngularSpring=true
    bOverrideAngularDamping=true
    bUseGravityOverride=true
    gravityScale=0.3
    simulationSpace=RootRelative
    alphaBoolBlend {blendOption=Linear}
    boundBone {boneName=dyn_Hair_Fn_c_02}
    componentAppliedLinearAccClamp=<2000 2000 1600>
    componentLinearAccScale=<0.4 0.4 0.4>
    componentLinearVelScale=<0.45 0.45 0.55>
    gravityOverride {z=500}
    physicsBodyDefinitions [{
      sphereCollisionRadius=2
      boundBone {boneName=dyn_Hair_Fn_c_02}
      boxExtents=<5 7 10>
      constraintSetup {twistAxis=AxisY, angularLimitsMax=<15 0 7>, angularLimitsMin=<-2 0 -2>, angularTarget=<0 0 0>}
      localJointOffset=<0 -2 0>
     }]
    relativeSpaceBone {boneName=root}
   }
   n18 = Anim Dynamics - Bone: dyn_Hair_Fn_d_01!(ComponentPose=@n17, Alpha=1, GravityOverride:Vector="0,0,500", ExternalForce="0,0,0") -> (Pose) {
    angularDampingOverride=0.9999
    angularSpringConstant=800
    bAngularSpring=true
    bOverrideAngularDamping=true
    bUseGravityOverride=true
    gravityScale=0.3
    simulationSpace=RootRelative
    alphaBoolBlend {blendOption=Linear}
    boundBone {boneName=dyn_Hair_Fn_d_01}
    componentAppliedLinearAccClamp=<2000 2000 1600>
    componentLinearAccScale=<0.4 0.4 0.4>
    componentLinearVelScale=<0.45 0.45 0.55>
    gravityOverride {z=500}
    physicsBodyDefinitions [{
      sphereCollisionRadius=2
      boundBone {boneName=dyn_Hair_Fn_d_01}
      boxExtents=<7 12 14>
      constraintSetup {twistAxis=AxisY, angularLimitsMax=<15 0 3>, angularLimitsMin=<-2 0 -5>, angularTarget=<0 0 0>}
      localJointOffset=<0 -2 0>
     }]
    relativeSpaceBone {boneName=root}
   }
   n19 = Anim Dynamics - Bone: dyn_Hair_Fn_d_02!(ComponentPose=@n18, Alpha=1, GravityOverride:Vector="0,0,500", ExternalForce="0,0,0") -> (Pose) {
    angularDampingOverride=0.9999
    angularSpringConstant=800
    bAngularSpring=true
    bOverrideAngularDamping=true
    bUseGravityOverride=true
    gravityScale=0.3
    simulationSpace=RootRelative
    alphaBoolBlend {blendOption=Linear}
    boundBone {boneName=dyn_Hair_Fn_d_02}
    componentAppliedLinearAccClamp=<2000 2000 1600>
    componentLinearAccScale=<0.4 0.4 0.4>
    componentLinearVelScale=<0.45 0.45 0.55>
    gravityOverride {z=500}
    physicsBodyDefinitions [{
      sphereCollisionRadius=2
      boundBone {boneName=dyn_Hair_Fn_d_02}
      boxExtents=<5 7 10>
      constraintSetup {twistAxis=AxisY, angularLimitsMax=<15 0 3>, angularLimitsMin=<-2 0 -5>, angularTarget=<0 0 0>}
      localJointOffset=<0 -2 0>
     }]
    relativeSpaceBone {boneName=root}
   }
   n20 = Anim Dynamics - Bone: dyn_Hair_Fn_e_01!(ComponentPose=@n19, Alpha=1, GravityOverride:Vector="0,0,500", ExternalForce="0,0,0") -> (Pose) {
    angularDampingOverride=0.9999
    angularSpringConstant=800
    bAngularSpring=true
    bOverrideAngularDamping=true
    bUseGravityOverride=true
    gravityScale=0.3
    simulationSpace=RootRelative
    alphaBoolBlend {blendOption=Linear}
    boundBone {boneName=dyn_Hair_Fn_e_01}
    componentAppliedLinearAccClamp=<2000 2000 1600>
    componentLinearAccScale=<0.4 0.4 0.4>
    componentLinearVelScale=<0.45 0.45 0.55>
    gravityOverride {z=500}
    physicsBodyDefinitions [{
      sphereCollisionRadius=2
      boundBone {boneName=dyn_Hair_Fn_e_01}
      boxExtents=<7 12 14>
      constraintSetup {twistAxis=AxisY, angularLimitsMax=<15 0 3>, angularLimitsMin=<-2 0 -5>, angularTarget=<0 0 0>}
      localJointOffset=<0 -2 0>
     }]
    relativeSpaceBone {boneName=root}
   }
   n21 = Anim Dynamics - Bone: dyn_Hair_Fn_e_02!(ComponentPose=@n20, Alpha=1, GravityOverride:Vector="0,0,500", ExternalForce="0,0,0") -> (Pose) {
    angularDampingOverride=0.9999
    angularSpringConstant=800
    bAngularSpring=true
    bOverrideAngularDamping=true
    bUseGravityOverride=true
    gravityScale=0.3
    simulationSpace=RootRelative
    alphaBoolBlend {blendOption=Linear}
    boundBone {boneName=dyn_Hair_Fn_e_02}
    componentAppliedLinearAccClamp=<2000 2000 1600>
    componentLinearAccScale=<0.4 0.4 0.4>
    componentLinearVelScale=<0.45 0.45 0.55>
    gravityOverride {z=500}
    physicsBodyDefinitions [{
      sphereCollisionRadius=2
      boundBone {boneName=dyn_Hair_Fn_e_02}
      boxExtents=<5 7 10>
      constraintSetup {twistAxis=AxisY, angularLimitsMax=<15 0 3>, angularLimitsMin=<-2 0 -5>, angularTarget=<0 0 0>}
      localJointOffset=<0 -2 0>
     }]
    relativeSpaceBone {boneName=root}
   }
   n22 = Anim Dynamics - Bone: dyn_Hair_Fn_f_01!(ComponentPose=@n21, Alpha=1, GravityOverride:Vector="0,0,500", ExternalForce="0,0,0") -> (Pose) {
    angularDampingOverride=0.9999
    angularSpringConstant=800
    bAngularSpring=true
    bOverrideAngularDamping=true
    bUseGravityOverride=true
    gravityScale=0.3
    simulationSpace=RootRelative
    alphaBoolBlend {blendOption=Linear}
    boundBone {boneName=dyn_Hair_Fn_f_01}
    componentAppliedLinearAccClamp=<2000 2000 1600>
    componentLinearAccScale=<0.4 0.4 0.4>
    componentLinearVelScale=<0.45 0.45 0.55>
    gravityOverride {z=500}
    physicsBodyDefinitions [{
      sphereCollisionRadius=2
      boundBone {boneName=dyn_Hair_Fn_f_01}
      boxExtents=<7 12 14>
      constraintSetup {twistAxis=AxisY, angularLimitsMax=<15 0 10>, angularLimitsMin=<-2 0 -3>, angularTarget=<0 0 0>}
      localJointOffset=<0 -2 0>
     }]
    relativeSpaceBone {boneName=root}
    retargetingSettings {bEnabled=true, rotationComponent=QuaternionAngle, source {rotation {pitch=5}}}
   }
   n23 = Anim Dynamics - Bone: dyn_Hair_Fn_f_02!(ComponentPose=@n22, Alpha=1, GravityOverride:Vector="0,0,500", ExternalForce="0,0,0") -> (Pose) {
    angularDampingOverride=0.9999
    angularSpringConstant=800
    bAngularSpring=true
    bOverrideAngularDamping=true
    bUseGravityOverride=true
    gravityScale=0.3
    simulationSpace=RootRelative
    alphaBoolBlend {blendOption=Linear}
    boundBone {boneName=dyn_Hair_Fn_f_02}
    componentAppliedLinearAccClamp=<2000 2000 1600>
    componentLinearAccScale=<0.4 0.4 0.4>
    componentLinearVelScale=<0.45 0.45 0.55>
    gravityOverride {z=500}
    physicsBodyDefinitions [{
      sphereCollisionRadius=2
      boundBone {boneName=dyn_Hair_Fn_f_02}
      boxExtents=<5 7 10>
      constraintSetup {twistAxis=AxisY, angularLimitsMax=<15 0 10>, angularLimitsMin=<-2 0 -3>, angularTarget=<0 0 0>}
      localJointOffset=<0 -2 0>
     }]
    relativeSpaceBone {boneName=root}
    retargetingSettings {bEnabled=true, rotationComponent=QuaternionAngle, source {rotation {pitch=5}}}
   }
   n24 = Anim Dynamics - Bone: dyn_Hair_Fn_g_01!(ComponentPose=@n23, Alpha=1, GravityOverride:Vector="0,0,500", ExternalForce="0,0,0") -> (Pose) {
    angularDampingOverride=0.9999
    angularSpringConstant=800
    bAngularSpring=true
    bOverrideAngularDamping=true
    bUseGravityOverride=true
    gravityScale=0.3
    simulationSpace=RootRelative
    alphaBoolBlend {blendOption=Linear}
    boundBone {boneName=dyn_Hair_Fn_g_01}
    componentAppliedLinearAccClamp=<2000 2000 1600>
    componentLinearAccScale=<0.4 0.4 0.4>
    componentLinearVelScale=<0.45 0.45 0.55>
    gravityOverride {z=500}
    physicsBodyDefinitions [{
      sphereCollisionRadius=2
      boundBone {boneName=dyn_Hair_Fn_g_01}
      boxExtents=<7 12 14>
      constraintSetup {twistAxis=AxisY, angularLimitsMax=<15 0 3>, angularLimitsMin=<1 0 -3>, angularTarget=<0 0 0>}
      localJointOffset=<0 -2 0>
     }]
    relativeSpaceBone {boneName=root}
   }
   n25 = Anim Dynamics - Bone: dyn_Hair_Fn_g_02!(ComponentPose=@n24, Alpha=1, GravityOverride:Vector="0,0,500", ExternalForce="0,0,0") -> (Pose) {
    angularDampingOverride=0.9999
    angularSpringConstant=800
    bAngularSpring=true
    bOverrideAngularDamping=true
    bUseGravityOverride=true
    gravityScale=0.3
    simulationSpace=RootRelative
    alphaBoolBlend {blendOption=Linear}
    boundBone {boneName=dyn_Hair_Fn_g_02}
    componentAppliedLinearAccClamp=<2000 2000 1600>
    componentLinearAccScale=<0.4 0.4 0.4>
    componentLinearVelScale=<0.45 0.45 0.55>
    gravityOverride {z=500}
    physicsBodyDefinitions [{
      sphereCollisionRadius=2
      boundBone {boneName=dyn_Hair_Fn_g_02}
      boxExtents=<5 7 10>
      constraintSetup {twistAxis=AxisY, angularLimitsMax=<15 0 3>, angularLimitsMin=<1 0 -3>, angularTarget=<0 0 0>}
      localJointOffset=<0 -2 0>
     }]
    relativeSpaceBone {boneName=root}
   }
   n26 = Anim Dynamics - Bone: dyn_Hair_Fn_h_01!(ComponentPose=@n25, Alpha=1, GravityOverride:Vector="0,0,500", ExternalForce="0,0,0") -> (Pose) {
    angularDampingOverride=0.9999
    angularSpringConstant=800
    bAngularSpring=true
    bOverrideAngularDamping=true
    bUseGravityOverride=true
    gravityScale=0.3
    simulationSpace=RootRelative
    alphaBoolBlend {blendOption=Linear}
    boundBone {boneName=dyn_Hair_Fn_h_01}
    componentAppliedLinearAccClamp=<2000 2000 1600>
    componentLinearAccScale=<0.4 0.4 0.4>
    componentLinearVelScale=<0.45 0.45 0.55>
    gravityOverride {z=500}
    physicsBodyDefinitions [{
      sphereCollisionRadius=2
      boundBone {boneName=dyn_Hair_Fn_h_01}
      boxExtents=<7 12 14>
      constraintSetup {twistAxis=AxisY, angularLimitsMax=<15 0 3>, angularLimitsMin=<1 0 -3>, angularTarget=<0 0 0>}
      localJointOffset=<0 -2 0>
     }]
    relativeSpaceBone {boneName=root}
   }
   n27 = Anim Dynamics - Bone: dyn_Hair_Fn_h_02!(ComponentPose=@n26, Alpha=1, GravityOverride:Vector="0,0,500", ExternalForce="0,0,0") -> (Pose) {
    angularDampingOverride=0.9999
    angularSpringConstant=800
    bAngularSpring=true
    bOverrideAngularDamping=true
    bUseGravityOverride=true
    gravityScale=0.3
    simulationSpace=RootRelative
    alphaBoolBlend {blendOption=Linear}
    boundBone {boneName=dyn_Hair_Fn_h_02}
    componentAppliedLinearAccClamp=<2000 2000 1600>
    componentLinearAccScale=<0.4 0.4 0.4>
    componentLinearVelScale=<0.45 0.45 0.55>
    gravityOverride {z=500}
    physicsBodyDefinitions [{
      sphereCollisionRadius=2
      boundBone {boneName=dyn_Hair_Fn_h_02}
      boxExtents=<5 7 10>
      constraintSetup {twistAxis=AxisY, angularLimitsMax=<15 0 3>, angularLimitsMin=<1 0 -3>, angularTarget=<0 0 0>}
      localJointOffset=<0 -2 0>
     }]
    relativeSpaceBone {boneName=root}
   }
   n28 = Anim Dynamics - Bone: dyn_Hair_Fn_i_01!(ComponentPose=@n27, Alpha=1, GravityOverride:Vector="0,0,500", ExternalForce="0,0,0") -> (Pose) {
    angularDampingOverride=0.9999
    angularSpringConstant=800
    bAngularSpring=true
    bOverrideAngularDamping=true
    bUseGravityOverride=true
    gravityScale=0.3
    simulationSpace=RootRelative
    alphaBoolBlend {blendOption=Linear}
    boundBone {boneName=dyn_Hair_Fn_i_01}
    componentAppliedLinearAccClamp=<2000 2000 1600>
    componentLinearAccScale=<0.4 0.4 0.4>
    componentLinearVelScale=<0.45 0.45 0.55>
    gravityOverride {z=500}
    physicsBodyDefinitions [{
      sphereCollisionRadius=2
      boundBone {boneName=dyn_Hair_Fn_i_01}
      boxExtents=<7 12 14>
      constraintSetup {twistAxis=AxisY, angularLimitsMax=<15 0 10>, angularLimitsMin=<-2 0 -3>, angularTarget=<0 0 0>}
      localJointOffset=<0 -2 0>
     }]
    relativeSpaceBone {boneName=root}
   }
   n29 = Anim Dynamics - Bone: dyn_Hair_Fn_i_02!(ComponentPose=@n28, Alpha=1, GravityOverride:Vector="0,0,500", ExternalForce="0,0,0") -> (Pose) {
    angularDampingOverride=0.9999
    angularSpringConstant=800
    bAngularSpring=true
    bOverrideAngularDamping=true
    bUseGravityOverride=true
    gravityScale=0.3
    simulationSpace=RootRelative
    alphaBoolBlend {blendOption=Linear}
    boundBone {boneName=dyn_Hair_Fn_i_02}
    componentAppliedLinearAccClamp=<2000 2000 1600>
    componentLinearAccScale=<0.4 0.4 0.4>
    componentLinearVelScale=<0.45 0.45 0.55>
    gravityOverride {z=500}
    physicsBodyDefinitions [{
      sphereCollisionRadius=2
      boundBone {boneName=dyn_Hair_Fn_i_02}
      boxExtents=<5 7 10>
      constraintSetup {twistAxis=AxisY, angularLimitsMax=<15 0 10>, angularLimitsMin=<-2 0 -3>, angularTarget=<0 0 0>}
      localJointOffset=<0 -2 0>
     }]
    relativeSpaceBone {boneName=root}
   }
   n30 = Anim Dynamics - Bone: dyn_Hair_Fn_k_01!(ComponentPose=@n29, Alpha=1, GravityOverride:Vector="0,0,500", ExternalForce="0,0,0") -> (Pose) {
    angularDampingOverride=0.9999
    angularSpringConstant=800
    bAngularSpring=true
    bOverrideAngularDamping=true
    bUseGravityOverride=true
    gravityScale=0.3
    simulationSpace=RootRelative
    alphaBoolBlend {blendOption=Linear}
    boundBone {boneName=dyn_Hair_Fn_k_01}
    componentAppliedLinearAccClamp=<2000 2000 1600>
    componentLinearAccScale=<0.4 0.4 0.4>
    componentLinearVelScale=<0.45 0.45 0.55>
    gravityOverride {z=500}
    physicsBodyDefinitions [{
      sphereCollisionRadius=2
      boundBone {boneName=dyn_Hair_Fn_k_01}
      boxExtents=<7 12 14>
      constraintSetup {twistAxis=AxisY, angularLimitsMax=<5 0 8>, angularLimitsMin=<-10 0 -8>, angularTarget=<0 0 0>}
      localJointOffset=<0 -2 0>
     }]
    relativeSpaceBone {boneName=root}
    retargetingSettings {bEnabled=true, rotationComponent=QuaternionAngle, source {rotation {pitch=-8}}}
   }
   n31 = Anim Dynamics - Bone: dyn_Hair_Fn_k_02!(ComponentPose=@n30, Alpha=1, GravityOverride:Vector="0,0,500", ExternalForce="0,0,0") -> (Pose) {
    angularDampingOverride=0.9999
    angularSpringConstant=800
    bAngularSpring=true
    bOverrideAngularDamping=true
    bUseGravityOverride=true
    gravityScale=0.3
    simulationSpace=RootRelative
    alphaBoolBlend {blendOption=Linear}
    boundBone {boneName=dyn_Hair_Fn_k_02}
    componentAppliedLinearAccClamp=<2000 2000 1600>
    componentLinearAccScale=<0.4 0.4 0.4>
    componentLinearVelScale=<0.45 0.45 0.55>
    gravityOverride {z=500}
    physicsBodyDefinitions [{
      sphereCollisionRadius=2
      boundBone {boneName=dyn_Hair_Fn_k_02}
      boxExtents=<5 7 10>
      constraintSetup {twistAxis=AxisY, angularLimitsMax=<5 0 8>, angularLimitsMin=<-10 0 -8>, angularTarget=<0 0 0>}
      localJointOffset=<0 -2 0>
     }]
    relativeSpaceBone {boneName=root}
    retargetingSettings {bEnabled=true, rotationComponent=QuaternionAngle, source {rotation {pitch=-8}}}
   }
   n32 = Anim Dynamics - Bone: dyn_Hair_L_Ear_01!(ComponentPose=@n31, Alpha=1, GravityOverride:Vector="0,0,500", ExternalForce="0,0,0") -> (Pose) {
    angularDampingOverride=0.9999
    angularSpringConstant=800
    bAngularSpring=true
    bOverrideAngularDamping=true
    bUseGravityOverride=true
    gravityScale=0.3
    simulationSpace=RootRelative
    alphaBoolBlend {blendOption=Linear}
    boundBone {boneName=dyn_Hair_L_Ear_01}
    componentAppliedLinearAccClamp=<2000 2000 1600>
    componentLinearAccScale=<0.4 0.4 0.4>
    componentLinearVelScale=<0.45 0.45 0.55>
    gravityOverride {z=500}
    physicsBodyDefinitions [{
      sphereCollisionRadius=2
      boundBone {boneName=dyn_Hair_L_Ear_01}
      boxExtents=<25 30 15>
      constraintSetup {twistAxis=AxisY, angularLimitsMax=<3 0 5>, angularLimitsMin=<-3 0 -2>, angularTarget=<0 0 0>}
      localJointOffset=<0 -2 0>
     }]
    relativeSpaceBone {boneName=root}
   }
   n33 = Anim Dynamics - Bone: dyn_Hair_L_Ear_02!(ComponentPose=@n32, Alpha=1, GravityOverride:Vector="0,0,500", ExternalForce="0,0,0") -> (Pose) {
    angularDampingOverride=0.9999
    angularSpringConstant=800
    bAngularSpring=true
    bOverrideAngularDamping=true
    bUseGravityOverride=true
    gravityScale=0.3
    simulationSpace=RootRelative
    alphaBoolBlend {blendOption=Linear}
    boundBone {boneName=dyn_Hair_L_Ear_02}
    componentAppliedLinearAccClamp=<2000 2000 1600>
    componentLinearAccScale=<0.4 0.4 0.4>
    componentLinearVelScale=<0.45 0.45 0.55>
    gravityOverride {z=500}
    physicsBodyDefinitions [{
      sphereCollisionRadius=2
      boundBone {boneName=dyn_Hair_L_Ear_02}
      boxExtents=<10 12.5 8>
      constraintSetup {twistAxis=AxisY, angularLimitsMax=<5 0 7>, angularLimitsMin=<-5 0 -2>, angularTarget=<0 0 0>}
      localJointOffset=<0 -2 0>
     }]
    relativeSpaceBone {boneName=root}
   }
   n34 = Anim Dynamics - Bone: dyn_Hair_R_Ear_01!(ComponentPose=@n33, Alpha=1, GravityOverride:Vector="0,0,500", ExternalForce="0,0,0") -> (Pose) {
    angularDampingOverride=0.9999
    angularSpringConstant=800
    bAngularSpring=true
    bOverrideAngularDamping=true
    bUseGravityOverride=true
    gravityScale=0.3
    simulationSpace=RootRelative
    alphaBoolBlend {blendOption=Linear}
    boundBone {boneName=dyn_Hair_R_Ear_01}
    componentAppliedLinearAccClamp=<2000 2000 1600>
    componentLinearAccScale=<0.4 0.4 0.4>
    componentLinearVelScale=<0.45 0.45 0.55>
    gravityOverride {z=500}
    physicsBodyDefinitions [{
      sphereCollisionRadius=2
      boundBone {boneName=dyn_Hair_R_Ear_01}
      boxExtents=<25 30 15>
      constraintSetup {twistAxis=AxisY, angularLimitsMax=<5 0 2>, angularLimitsMin=<-5 0 -5>, angularTarget=<0 0 0>}
      localJointOffset=<0 2 0>
     }]
    relativeSpaceBone {boneName=root}
   }
   n35 = Anim Dynamics - Bone: dyn_Hair_R_Ear_02!(ComponentPose=@n34, Alpha=1, GravityOverride:Vector="0,0,500", ExternalForce="0,0,0") -> (Pose) {
    angularDampingOverride=0.9999
    angularSpringConstant=800
    bAngularSpring=true
    bOverrideAngularDamping=true
    bUseGravityOverride=true
    gravityScale=0.3
    simulationSpace=RootRelative
    alphaBoolBlend {blendOption=Linear}
    boundBone {boneName=dyn_Hair_R_Ear_02}
    componentAppliedLinearAccClamp=<2000 2000 1600>
    componentLinearAccScale=<0.4 0.4 0.4>
    componentLinearVelScale=<0.45 0.45 0.55>
    gravityOverride {z=500}
    physicsBodyDefinitions [{
      sphereCollisionRadius=2
      boundBone {boneName=dyn_Hair_R_Ear_02}
      boxExtents=<10 12.5 8>
      constraintSetup {twistAxis=AxisY, angularLimitsMax=<10 0 2>, angularLimitsMin=<-10 0 -5>, angularTarget=<0 0 0>}
      localJointOffset=<0 2 0>
     }]
    relativeSpaceBone {boneName=root}
   }
   Save cached pose 'HairDynPose'(Pose=Component To Local(ComponentPose=@n35)) {cacheName=HairDynPose}
  ]}
 EventGraph {nodes [
   Event BlueprintUpdateAnimation()
   > n1 = Sequence()
   > n2 = "Gravity Override State Function (Parameter Asset) (Nativized)"(
    CharacterPartAnimInstance=Self-Reference()
    GameTimeInSeconds=GetGameTimeInSeconds()
    Parameters=&/F_MED_GraphShook_FaceAcc_Parameters
   )
   > Set Gravity Override Cape(Gravity Override Cape:Vector=@n2) -> (Output_Get:Vector)
   > Set RBAN Settings Out(RBAN Settings Out=@n2.OutRigidBodyAnimNodeParameters)
   Set Not bIsSkyDiving(execute=@n1.then_1, Not bIsSkyDiving:bool=NOT Boolean(A=Get bIsSkyDiving())) -> (Output_Get:bool)
   > Set Not bIsSkyDiving_Floating(Not bIsSkyDiving_Floating:bool=AND Boolean(A=Get bIsSkyDiving(), B=Get bIsParachuteOpened())) -> (Output_Get:bool)
   Event OnCharacterCustomizationLoaded()
   n3 = CheckForBackbling()
   > Set isWearingPackBool(isWearingPackBool:bool=OR Boolean!(A=@n3, B=@n3.bIsBackblingACape, C=@n3.bIsBackblingLong) -> (ReturnValue)) -> (Output_Get:bool)
   > Set is_WearingCape(is_WearingCape:bool=@n3.bIsBackblingLong) -> (Output_Get:bool)
   Event BlueprintInitializeAnimation()
   > Set Skel_Mesh(Skel_Mesh=GetOwnerSkeletalMeshComponent())
   AND Boolean(A=@n3, B=NOT Boolean(A=@n3.bIsBackblingLong))
  ]}
]
variables [
 Gravity Override Cape {type=Vector}
 is_WearingCape {type=boolean}
 isWearingPackBool {type=boolean}
 Not bIsSkyDiving {type=boolean}
 Not bIsSkyDiving_Floating {type=boolean}
 RBAN Settings Out {type=FortRigidBodyAnimNodeParameters}
 Skel_Mesh {type=SkeletalMeshComponent}
]
